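# Editor utility widget for the prop browser that presents a "Please Select a Project" dialog of animation browser projects.
# Set Projects loops over the given projects and, for each with at least one animatable prop, adds a named EUW_SelectProjectButton to a vertical box.
# Each button's click event binds to a handler that rebroadcasts the project through the Project Selected dispatcher.
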WidgetBlueprint /AnimInEdTools/PropBrowser/EUW_SelectPropProjectDialog
graphs [
 EventGraph {links [], nodes []}
 Project Button Clicked {nodes [
   n1 = Project Button Clicked!() -> (then, Project:BP Animation Browser Project Object Reference) {is=title K2Node_FunctionEntry}
   > Call Project Selected!(execute:Exec, self:EUW Select Prop Project Dialog Object Reference, Project:BP Animation Browser Project Object Reference=@n1) -> (then:Exec) {is=title K2Node_CallDelegate}
  ]}
 Project Selected {nodes [Project Selected!() -> (then, Project:BP Animation Browser Project Object Reference) {is=title K2Node_FunctionEntry}]}
 Set Projects {nodes [
   n1 = Set Projects!() -> (then, Projects:Array of BP Animation Browser Project Object References) {is=title K2Node_FunctionEntry}
   > n2 = For Each Loop(Array:Array of BP Animation Browser Project Object References=@n1) -> (Array Element:BP Animation Browser Project Object Reference)
   n3 = Get Animatable Props!(Project:BP Animation Browser Project Object Reference=@n2, __WorldContext:Object Reference) -> (Animatable Creatures:Array of Animatable Creature Object References) {is=title K2Node_CallFunction}
   > Branch(Condition="integer > integer"(A=Length(TargetArray:Array of Animatable Creature Object References=@n3)))
   > n4 = Create Widget!(
    execute:Exec
    WorldContextObject:Object Reference
    Class:User Widget Class Reference=/AnimInEdTools/AnimationBrowser/EUW_SelectProjectButton.EUW_SelectProjectButton_C
    OwningPlayer:Player Controller Object Reference
   ) -> (then:Exec, ReturnValue:EUW Select Project Button Object Reference)
   > Set Text(
    Text="To Text (String)"(InString=Get Name!(self:BP Animation Browser Project Object Reference=@n2) -> (Name:String))
    self=Get TextBlock!(self:EUW Select Project Button Object Reference=@n4) -> (TextBlock:Text Object Reference)
   )
   > Set Project!(execute, Project:BP Animation Browser Project Object Reference=@n2, self:EUW Select Project Button Object Reference=@n4) -> (then, Output_Get:BP Animation Browser Project Object Reference)
   > Bind Event to Project Button Clicked(self:EUW Select Project Button Object Reference=@n4, Delegate=Create Event()) {is=title K2Node_AddDelegate}
   > AddChild(self=Get ProjectVerticalBox() -> (ProjectVerticalBox:Vertical Box Object Reference), Content=@n4)
  ]}
]
variables [Project Selected {}]
widgets {
 info {inheritedWidgetCount=0, parentClass=&EditorUtilityWidget, rootWidgetClass=&VerticalBox, widgetCount=4}
 widgets [
  VerticalBox_1 {widgetClassPath=&VerticalBox, widget=VerticalBox_1}
  Border_274 {
   widgetClassPath=&Border
   parent=VerticalBox_1
   widget=Border_274
   properties {verticalAlignment=VAlign_Center, brushColor=<rgba 0.0275 0.0275 0.0275 1>, padding=<ltrb 5 5 5 5>}
   slot {class=VerticalBoxSlot, properties {padding=<ltrb 0 0 0 5>}}
  }
  TitleTextBlock {
   widgetClassPath=&TextBlock
   bIsVariable=true
   parent=Border_274
   widget=TitleTextBlock
   properties {
    text=Please Select a Project
    colorAndOpacity {colorUseRule=UseColor_Specified, specifiedColor=<rgba 0.5255 0.5255 0.5255 1>}
    font {
     bForceMonospaced=false
     bMaterialIsStencil=false
     fontMaterial=null
     fontObject=&Roboto
     letterSpacing=0
     monospacedWidth=1
     size=10
     skewAmount=0
     typefaceFontName=Bold
     outlineSettings {
      bApplyOutlineToDropShadows=false
      bMiteredCorners=false
      bSeparateFillAlpha=false
      outlineMaterial=null
      outlineSize=0
      outlineColor=<rgba 0 0 0 1>
     }
    }
   }
   slot {class=BorderSlot, properties {verticalAlignment=VAlign_Center, padding=<ltrb 5 5 5 5>}}
  }
  ProjectVerticalBox {
   widgetClassPath=&VerticalBox
   bIsVariable=true
   parent=VerticalBox_1
   widget=ProjectVerticalBox
   slot {class=VerticalBoxSlot, properties {verticalAlignment=VAlign_Center, padding=<ltrb 5 5 5 5>}}
  }
 ]
}
