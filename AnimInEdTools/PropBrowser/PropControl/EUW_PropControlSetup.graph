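# Editor Utility Widget that lets animators set up prop control on Control Rig characters in Sequencer.
# It finds the rig's Prop Global control and hand attach target, then enables Activate, Apply and Select.
# Activate resets spaces and keys, then snaps the prop to the attach at the start frame.
# Apply bakes it over the full or selected-key range, then space-switches attach and checked IK arms to it, using Python snapping and a temporary locator.
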EditorUtilityWidgetBlueprint /AnimInEdTools/PropBrowser/PropControl/EUW_PropControlSetup
graphs [
 Activate Prop Control {nodes [
   Activate Prop Control() {is=title K2Node_FunctionEntry}
   > n1 = Sequence!(execute) -> (then_0, then_1, then_2:Exec, then_3:Exec, then_4:Exec)
   > n2 = Get Frames!(execute, self) -> (then, valid frames:bool, start frame:int=0, end frame:int=0) {is=title K2Node_CallFunction}
   > Set valid frames(valid frames:bool=@n2) -> (Output_Get:bool)
   > Set Local Start(Local Start:int=@n2.start frame) -> (Output_Get:int=0)
   > Set Local End(Local End:int=@n2.end frame) -> (Output_Get:int=0)
   n3 = Branch(execute=@n1.then_1, Condition=Get valid frames() -> (valid frames:bool))
   > Arm and Attach Default Space!(execute, self, start frame:int=Get Local Start() -> (Local Start:int=0)) -> (then)
   Return Node(execute=@n3.else)
   n4 = GetControlRigs!(execute=@n1.then_2, LevelSequence:Level Sequence Object Reference=Get Level Sequence() -> (Level Sequence:Level Sequence Object Reference)) -> (then, ReturnValue:Array of Control Rig Sequencer Binding Proxy Structures)
   > n5 = For Each Loop!(Exec, Array:Array of Control Rig Sequencer Binding Proxy Structures=@n4) -> (
    LoopBody
    Array Element_Proxy:Movie Scene Binding Proxy Structure
    Array Element_ControlRig:Control Rig Object Reference
    Array Element_Track:Movie Scene Control Rig Parameter Track Object Reference
    Array Index
    Completed
   )
   > Branch(Condition="Equal (Object)"(A=Get Selected Rig for Prop() -> (Selected Rig for Prop:Control Rig Object Reference), B=@n5.Array Element_ControlRig))
   > n6 = GetSections!(execute, Track:Movie Scene Track Object Reference=@n5.Array Element_Track) -> (then, ReturnValue:Array of Movie Scene Section Object References)
   > n7 = For Each Loop!(Exec, Array:Array of Movie Scene Section Object References=@n6) -> (LoopBody, Array Element:Movie Scene Section Object Reference, Array Index, Completed)
   > n8 = GetAllChannels!(execute, Section:Movie Scene Section Object Reference=@n7) -> (then, ReturnValue:Array of Movie Scene Scripting Channel Object References)
   > n9 = For Each Loop(Array:Array of Movie Scene Scripting Channel Object References=@n8) -> (Array Element:Movie Scene Scripting Channel Object Reference)
   n10 = Get!() -> (Prop Control_Type:ERigElementType Enum, Prop Control_Name:Name)
   > Branch(Condition=Contains(SearchIn=GetDisplayName(Object=@n9), Substring="To String (Name)"(InName=@n10.Prop Control_Name)))
   > Execute Python Script!(
    execute
    PythonScript="import unreal\r\n\r\nfor key in channel.get_keys():\r\n    channel.remove_key(key)"
    channel:Movie Scene Scripting Channel Object Reference=@n9
   ) -> (then, ReturnValue) {inputs [channel]}
   n11 = For Each Loop(
    Exec=@n1.then_4
    Array=Get Prop Global control Channels() -> (Prop Global control Channels:Array of Movie Scene Scripting Channel Object References)
   )
   > n12 = Cast To MovieSceneScriptingFloatChannel!(execute, Object=@n11) -> (then, CastFailed, AsMovie Scene Scripting Float Channel:Movie Scene Scripting Float Channel Object Reference) {is=title K2Node_DynamicCast}
   > Execute Python Script!(
    execute
    PythonScript="import unreal\r\n\r\n\r\nkeys = []\r\n\r\nfor key in channel.get_keys():\r\n    if key.get_time().frame_number.value > start:\r\n        keys.append(key)\r\n\r\nkeys = list(set(keys))\r\n\r\nfor key in keys:\r\n    channel.remove_key(key)"
    channel:Movie Scene Scripting Float Channel Object Reference=@n12
    start:int=Get Local Start() -> (Local Start=0)
   ) -> (then, ReturnValue) {inputs [channel,start]}
   Execute Python Script!(
    execute=@n1.then_3
    PythonScript="import unreal\r\n\r\n## Create ControlRigSnapperSelection objects for the parent and the children\r\nparent = unreal.ControlRigSnapperSelection()\r\nchildren = unreal.ControlRigSnapperSelection()\r\n\r\n## Create a ControlRigForWorldTransforms object\r\n## Set to the proper control rig and the left hand control as the control\r\n## It is possible to have multiple control names in here\r\nchild_control_rig = unreal.ControlRigForWorldTransforms()\r\nchild_control_rig.control_rig = control_rig\r\nchild_control_rig.control_names = [prop_control]\r\nchildren.control_rigs = [child_control_rig]\r\n\r\n\r\nparent_control_rig = unreal.ControlRigForWorldTransforms()\r\nparent_control_rig.control_rig = control_rig\r\nparent_control_rig.control_names = [target_control]\r\nparent.control_rigs = [parent_control_rig]\r\n\r\n## Create and set snap settings\r\nsnap_settings = unreal.ControlRigSnapSettings()\r\nsnap_settings.keep_offset = False\r\nsnap_settings.snap_position = True\r\nsnap_settings.snap_rotation = True\r\nsnap_settings.snap_scale = False\r\nsnap_settings.baking_key_settings = unreal.BakingKeySettings.ALL_FRAMES\r\n\r\n## Take the ControlRigForWorldTransforms object, set as the child ControlRigSnapperSelection\r\nchildren.control_rigs = [child_control_rig]\r\n\r\nunreal.ControlRigSequencerLibrary.snap_control_rig(level_sequence, unreal.FrameNumber(frame), unreal.FrameNumber(frame), children, parent, snap_settings)\r\n"
    control_rig:Control Rig Object Reference=Get Selected Rig for Prop()
    target_control:String=Get Attach Target Name() -> (Attach Target Name:String)
    prop_control:String="Get Prop Global Control "() -> ("Prop Global Control ":String)
    level_sequence:Level Sequence Object Reference=Get Level Sequence()
    frame:int=Get Local Start() -> (Local Start=0)
   ) -> (then, ReturnValue) {inputs [control_rig, target_control, prop_control, level_sequence, frame]}
   > SetLocalControlRigBool!(
    execute
    LevelSequence:Level Sequence Object Reference=Get Level Sequence()
    ControlRig:Control Rig Object Reference=Get Selected Rig for Prop()
    ControlName:Name=Get().Prop Control_Name
    Frame_Value:int=0
    Value:bool=true
    TimeUnit:EMovieSceneTimeUnit Enum=DisplayRate
    bSetKey:bool=true
   ) -> (then)
   Execute Python Script!(
    execute
    PythonScript="import unreal\r\n\r\nfor key in channel.get_keys():\r\n    print(key.get_time())\r\n    ##channel.remove_key(key)"
    channel:Movie Scene Scripting Float Channel Object Reference=@n12
   ) -> (then, ReturnValue) {inputs [channel]}
  ]}
 Apply Prop Control {nodes [
   Apply Prop Control() {is=title K2Node_FunctionEntry}
   > n1 = Sequence()
   > n2 = Get Frames() -> (start frame=0, end frame=0)
   > n3 = Set valid frames(valid frames=@n2)
   > n4 = Branch(Condition=@n3)
   > Set Local Start(Local Start=@n2.start frame) -> (Output_Get=0)
   > Set Local End(Local End=@n2.end frame) -> (Output_Get=0)
   > n5 = Execute Python Script!(
    execute
    PythonScript="str_output =f\"This will bake the prop control on keyed frames [start, end]:\\n[{start},{end}]\\nContinue?\""
    start:int=Get Local Start() -> (Local Start=0)
    end:int=Get Local End() -> (Local End:int=0)
   ) -> (then, ReturnValue, str_output:Text) {inputs [start,end], outputs [str_output]}
   > n6 = Show Message Dialog(
    Title="To Text (String)"(InString=Prop Activation)
    Message=Format Text(Format=@n5.str_output)
    MessageType=YesNo
    DefaultValue=No
    MessageCategory=Info
   )
   > n7 = Branch(Condition="Equal (Enum)"(A:EAppReturnType Enum=@n6, B:EAppReturnType Enum=Yes))
   > Set valid frames(valid frames=true)
   > Execute Python Script!(
    execute
    PythonScript="import unreal\r\n\r\n## Create ControlRigSnapperSelection objects for the parent and the children\r\nparent = unreal.ControlRigSnapperSelection()\r\nchildren = unreal.ControlRigSnapperSelection()\r\n\r\n## Create a ControlRigForWorldTransforms object\r\n## Set to the proper control rig and the left hand control as the control\r\n## It is possible to have multiple control names in here\r\nchild_control_rig = unreal.ControlRigForWorldTransforms()\r\nchild_control_rig.control_rig = control_rig\r\nchild_control_rig.control_names = [prop_control]\r\nchildren.control_rigs = [child_control_rig]\r\n\r\n\r\nparent_control_rig = unreal.ControlRigForWorldTransforms()\r\nparent_control_rig.control_rig = control_rig\r\nparent_control_rig.control_names = [target_control]\r\nparent.control_rigs = [parent_control_rig]\r\n\r\n## Create and set snap settings\r\nsnap_settings = unreal.ControlRigSnapSettings()\r\nsnap_settings.keep_offset = True\r\nsnap_settings.snap_position = True\r\nsnap_settings.snap_rotation = True\r\nsnap_settings.snap_scale = False\r\nsnap_settings.baking_key_settings = unreal.BakingKeySettings.ALL_FRAMES\r\n\r\n## Take the ActorForWorldTransforms object, set it as the parent ControlRigSnapperSelection\r\n## Take the ControlRigForWorldTransforms object, set as the child ControlRigSnapperSelection\r\nchildren.control_rigs = [child_control_rig]\r\n\r\nunreal.ControlRigSequencerLibrary.snap_control_rig(level_sequence, unreal.FrameNumber(start_frame), unreal.FrameNumber(end_frame), children, parent, snap_settings)\r\n"
    control_rig:Control Rig Object Reference=Get Selected Rig for Prop()
    target_control:String=Get Attach Target Name()
    prop_control:String="Get Prop Global Control "()
    level_sequence:Level Sequence Object Reference=Get Level Sequence()
    start_frame:int=Get Local Start() -> (Local Start=0)
    end_frame:int=Get Local End() -> (Local End=0)
   ) -> (then, ReturnValue) {inputs [control_rig, target_control, prop_control, level_sequence, start_frame, end_frame]}
   > PrintString(InString=Snapped control)
   Set valid frames(execute=@n7.else, valid frames=false)
   Show Message Dialog(
    execute=@n4.else
    Title="To Text (String)"(InString=Prop Control Setup)
    Message="To Text (String)"(InString=No keys selected in sequencer)
    MessageType=Ok
    DefaultValue=No
    MessageCategory=Error
   )
   n8 = Branch(execute=@n1.then_1, Condition=Get valid frames())
   > n9 = Branch(Condition=Contains(SearchIn=Get Attach Target Name(), Substring=Attach and Weapon))
   > n10 = Execute Python Script!(
    execute
    PythonScript="\r\noutput_string = f\"{input_string.split('/')[0]}/{input_string.split('/')[-1].lower().replace(' ', '_')}\""
    input_string:String=Get Attach Target Name()
   ) -> (then, ReturnValue, output_string:String) {inputs [input_string], outputs [output_string]}
   > n11 = Set Local Space Key Name(Local Space Key Name:String=Append(A=@n10.output_string, B=" Prop Space")) -> (Output_Get:String)
   n12 = Spawn and Key Locator!(
    execute=[@n11.then, @n17.then]
    self
    start time:int=Get Local Start() -> (Local Start=0)
    Sequence:Movie Scene Sequence Object Reference=Get Level Sequence()
    end time:int=int + int(A=Get Local End() -> (Local End=0))
   ) -> (then, output actor:Actor Object Reference, output binding:Movie Scene Binding Proxy Structure)
   > Set locator(locator:Actor Object Reference=@n12) -> (Output_Get:Actor Object Reference)
   > Set locator binding(locator binding:Movie Scene Binding Proxy Structure=@n12.output binding) -> (Output_Get:Movie Scene Binding Proxy Structure)
   n13 = Get Attach Target Name()
   > Execute Python Script!(
    execute
    PythonScript="import unreal\r\n\r\n## Create ControlRigSnapperSelection objects for the parent and the children\r\nparent = unreal.ControlRigSnapperSelection()\r\nchildren = unreal.ControlRigSnapperSelection()\r\n\r\n## Create a ControlRigForWorldTransforms object\r\n## Set to the proper control rig and the left hand control as the control\r\n## It is possible to have multiple control names in here\r\nchild_actor = unreal.ActorForWorldTransforms()\r\nchild_actor.actor = actor\r\nchildren.actors = [child_actor]\r\n\r\nparent_control_rig = unreal.ControlRigForWorldTransforms()\r\nparent_control_rig.control_rig = control_rig\r\nparent_control_rig.control_names = [target_control]\r\nparent.control_rigs = [parent_control_rig]\r\n\r\n## Create and set snap settings\r\nsnap_settings = unreal.ControlRigSnapSettings()\r\nsnap_settings.keep_offset = False\r\nsnap_settings.snap_position = True\r\nsnap_settings.snap_rotation = True\r\nsnap_settings.snap_scale = False\r\nsnap_settings.baking_key_settings = unreal.BakingKeySettings.ALL_FRAMES\r\n\r\n\r\nunreal.ControlRigSequencerLibrary.snap_control_rig(level_sequence, unreal.FrameNumber(0), unreal.FrameNumber(end_frame), children, parent, snap_settings)"
    control_rig:Control Rig Object Reference=Get Selected Rig for Prop()
    target_control:String=@n13
    prop_control:String=@n13
    level_sequence:Level Sequence Object Reference=Get Level Sequence()
    start_frame:int=Get Local Start() -> (Local Start=0)
    end_frame:int=Get Local End() -> (Local End=0)
    actor:Actor Object Reference=Get locator() -> (locator:Actor Object Reference)
   ) -> (then, ReturnValue) {inputs [control_rig, target_control, prop_control, level_sequence, start_frame, end_frame, actor]}
   > n14 = GetWorldSpaceReferenceKey!(execute) -> (then, ReturnValue_Type:ERigElementType Enum, ReturnValue_Name:Name)
   > SetControlRigSpace!(
    execute
    InSequence:Level Sequence Object Reference=Get Level Sequence()
    InControlRig:Control Rig Object Reference=Get Selected Rig for Prop()
    InControlName:Name=String To Name(InString=Get Attach Target Name())
    InSpaceKey_Type:ERigElementType Enum=@n14
    InSpaceKey_Name:Name=@n14.ReturnValue_Name
    InTime_Value:int=int - int(A=Get Local Start() -> (Local Start=0))
    TimeUnit:EMovieSceneTimeUnit Enum=DisplayRate
   ) -> (then, ReturnValue:bool)
   > SetControlRigSpace(
    InSequence=Get Level Sequence()
    InControlRig=Get Selected Rig for Prop()
    InControlName=String To Name(InString=Get Attach Target Name())
    InSpaceKey_Type=Null
    InSpaceKey_Name=String To Name(InString=Get Local Space Key Name() -> (Local Space Key Name:String))
    InTime_Value=Get Local Start() -> (Local Start=0)
    TimeUnit=DisplayRate
   )
   n15 = Get Attach Target Name()
   > Execute Python Script(
    PythonScript="import unreal\r\n\r\n## Create ControlRigSnapperSelection objects for the parent and the children\r\nparent = unreal.ControlRigSnapperSelection()\r\nchildren = unreal.ControlRigSnapperSelection()\r\n\r\n## Create a ControlRigForWorldTransforms object\r\n## Set to the proper control rig and the left hand control as the control\r\n## It is possible to have multiple control names in here\r\nchild_control_rig = unreal.ControlRigForWorldTransforms()\r\nchild_control_rig.control_rig = control_rig\r\nchild_control_rig.control_names = [prop_control]\r\nchildren.control_rigs = [child_control_rig]\r\n\r\nparent_actor = unreal.ActorForWorldTransforms()\r\nparent_actor.actor = actor\r\nparent.actors = [parent_actor]\r\n\r\n## Create and set snap settings\r\nsnap_settings = unreal.ControlRigSnapSettings()\r\nsnap_settings.keep_offset = False\r\nsnap_settings.snap_position = True\r\nsnap_settings.snap_rotation = True\r\nsnap_settings.snap_scale = False\r\nsnap_settings.baking_key_settings = unreal.BakingKeySettings.KEYS_ONLY\r\n\r\n## Take the ActorForWorldTransforms object, set it as the parent ControlRigSnapperSelection\r\n## Take the ControlRigForWorldTransforms object, set as the child ControlRigSnapperSelection\r\n\r\nunreal.ControlRigSequencerLibrary.snap_control_rig(level_sequence, unreal.FrameNumber(start_frame),unreal.FrameNumber(end_frame), children, parent, snap_settings)\r\n"
    control_rig=Get Selected Rig for Prop()
    target_control=@n15
    prop_control=@n15
    level_sequence=Get Level Sequence()
    start_frame=Get Local Start() -> (Local Start=0)
    end_frame=Get Local End() -> (Local End=0)
    actor=Get locator()
   ) {inputs [control_rig, target_control, prop_control, level_sequence, start_frame, end_frame, actor]}
   > Destroy Actor(self=Get locator())
   > Remove!(
    execute
    InBinding:"Movie Scene Binding Proxy Structure (by ref)"=Get locator binding() -> (locator binding:Movie Scene Binding Proxy Structure)
   ) -> (then)
   > n16 = GetDefaultParentKey!(execute) -> (then, ReturnValue_Type:ERigElementType Enum, ReturnValue_Name:Name)
   > SetControlRigSpace(
    InSequence=Get Level Sequence()
    InControlRig=Get Selected Rig for Prop()
    InControlName=String To Name(InString=Get Attach Target Name())
    InSpaceKey_Type=@n16
    InSpaceKey_Name=@n16.ReturnValue_Name
    InTime_Value=int - int(A=Get Local Start() -> (Local Start=0))
    TimeUnit=DisplayRate
   )
   n17 = Set Local Space Key Name(execute=@n9.else, Local Space Key Name=Append(A=Get Attach Target Name(), B=" Prop Space"))
   Return Node(execute=@n8.else)
   n18 = Branch(execute=@n1.then_3, Condition=Get local is biped v2() -> (local is biped v2:bool))
   Apply Prop Space Switches(execute=@n18.else)
   n19 = GetControlRigs(execute=@n1.then_2, LevelSequence=Get Level Sequence())
   > n20 = For Each Loop(Array:Array of Control Rig Sequencer Binding Proxy Structures=@n19) -> (Array Element:Control Rig Sequencer Binding Proxy Structure)
   Branch(execute=@n20.Completed, Condition=Get local is biped v2())
   > IKFK_V2_Match_Event()
   n21 = Branch(execute=@n20.LoopBody, Condition=Get local is biped v2())
   n22 = Set local control rig sequencer proxy!(execute=@n21.else, local control rig sequencer proxy:Control Rig Sequencer Binding Proxy Structure=@n20) -> (
    then
    Output_Get_Proxy:Movie Scene Binding Proxy Structure
    Output_Get_ControlRig:Control Rig Object Reference
    Output_Get_Track:Movie Scene Control Rig Parameter Track Object Reference
   )
   > n23 = Branch(Condition="Equal (Object)"(A=Get Selected Rig for Prop(), B=@n22.Output_Get_ControlRig))
   > Set control rig sequencer proxy(control rig sequencer proxy:Control Rig Sequencer Binding Proxy Structure=Get local control rig sequencer proxy() -> (local control rig sequencer proxy:Control Rig Sequencer Binding Proxy Structure)) -> (Output_Get:Control Rig Sequencer Binding Proxy Structure)
   > n24 = GetSections(Track=@n22.Output_Get_Track)
   > n25 = For Each Loop(Array:Array of Movie Scene Section Object References=@n24) -> (Array Element:Movie Scene Section Object Reference)
   > n26 = GetAllChannels(Section=@n25)
   > n27 = For Each Loop(Array:Array of Movie Scene Scripting Channel Object References=@n26) -> (Array Element:Movie Scene Scripting Channel Object Reference)
   n28 = "To String (Name)"(InName=Get ChannelName!(self:Movie Scene Scripting Channel Object Reference=@n27) -> (ChannelName:Name=None))
   > Branch(Condition=AND Boolean(A=Contains(SearchIn=@n28, Substring=/Ik Fk Switch), B=Contains(SearchIn=@n28, Substring=arm)))
   n29 = Get l_ik_apply_cbox() -> (l_ik_apply_cbox:Editor Utility Check Box Object Reference)
   n30 = Split(SourceString=@n28, InStr=/, SearchDir=FromStart)
   > n31 = Branch(Condition=AND Boolean(A=IsChecked(self=@n29), B=EndsWith(SourceString=@n30, InSuffix=L, SearchCase=CaseSensitive)))
   > n32 = Set local left side(local left side:bool=true) -> (Output_Get:bool)
   n33 = Cast To MovieSceneScriptingBoolChannel!(execute=[@n37.then, @n32.then], Object=@n27) -> (then, CastFailed, AsMovie Scene Scripting Bool Channel:Movie Scene Scripting Bool Channel Object Reference) {is=title K2Node_DynamicCast}
   > Set!(
    execute
    Range_bHasStart:bool=true
    Range_bHasEnd:bool=true
    Range_InclusiveStart:int=int - int(A=Get Local Start() -> (Local Start=0))
    Range_ExclusiveEnd:int=Get Local End() -> (Local End=0)
   ) -> (then, Output_Get:Sequencer Scripting Range Structure)
   > n34 = Cast To MR_Biped_v02!(execute, Object=Get Selected Rig for Prop()) -> (then, CastFailed, AsMR Biped V 02:MR Biped V 02 Object Reference) {is=title K2Node_DynamicCast}
   > Set local is biped v2(local is biped v2:bool=true) -> (Output_Get:bool)
   n35 = For Loop(execute=@n34.CastFailed, FirstIndex=Get Local Start() -> (Local Start=0), LastIndex=Get Local End() -> (Local End=0))
   > "Add Key (Bool)"!(
    execute
    self:Movie Scene Scripting Bool Channel Object Reference=@n33
    InTime_Value:int=@n35
    NewValue:bool=true
    SubFrame:float=0
    TimeUnit:EMovieSceneTimeUnit Enum=DisplayRate
   ) -> (then, ReturnValue:Movie Scene Scripting Bool Key Object Reference)
   > Set local is biped v2(local is biped v2=false)
   n36 = Get r_ik_apply_cbox() -> (r_ik_apply_cbox:Editor Utility Check Box Object Reference)
   Branch(execute=@n31.else, Condition=AND Boolean(A=EndsWith(SourceString=@n30, InSuffix=R, SearchCase=CaseSensitive), B=IsChecked(self=@n36)))
   > n37 = Set local left side(local left side=false)
   Set local is biped v2(execute=@n23.else, local is biped v2=false)
  ]}
 Apply Prop Space Switches {nodes [
   Apply Prop Space Switches() {is=title K2Node_FunctionEntry}
   n1 = Get!() -> (Range_bHasStart:bool=~, Range_bHasEnd:bool=~, Range_InclusiveStart:int, Range_ExclusiveEnd:int)
   > Set Local Start(Local Start=int + int(A=@n1.Range_InclusiveStart)) -> (Output_Get=0)
   > Set Local End(Local End=@n1.Range_ExclusiveEnd) -> (Output_Get=0)
   > n2 = Set Local Rig Heirarchy(Local Rig Heirarchy:Rig Hierarchy Object Reference=GetHierarchy!(self:Control Rig Object Reference=Get Selected Rig for Prop()) -> (ReturnValue:Rig Hierarchy Object Reference)) -> (Output_Get:Rig Hierarchy Object Reference)
   > n3 = For Each Loop(Array:Array of Rig Element Key Structures=Get Controls!(self:Rig Hierarchy Object Reference=@n2, bTraverse:bool=true) -> (ReturnValue:Array of Rig Element Key Structures) {is=title K2Node_CallFunction}) -> (Array Element:Rig Element Key Structure)
   > Set Local Rig Element(Local Rig Element:Rig Element Key Structure=@n3) -> (Output_Get:Rig Element Key Structure)
   n4 = "To String (Name)"(InName=Get!() -> (Local Rig Element_Type:ERigElementType Enum, Local Rig Element_Name:Name).Local Rig Element_Name)
   > Branch(Condition=AND Boolean(A=Contains(SearchIn=@n4, Substring=arm), B=EndsWith(SourceString=@n4, InSuffix=/PV)))
   > n5 = Cast To ModularRig!(execute, Object=Get Selected Rig for Prop()) -> (then, CastFailed, AsModular Rig:Modular Rig Object Reference) {is=title K2Node_DynamicCast}
   > n6 = Sequence!(execute) -> (then_0, then_1)
   n7 = GetModuleFName!(
    self:Rig Hierarchy Object Reference=Get Local Rig Heirarchy() -> (Local Rig Heirarchy:Rig Hierarchy Object Reference)
    InItem:Rig Element Key Structure=Get Local Rig Element() -> (Local Rig Element:Rig Element Key Structure)
   ) -> (ReturnValue:Name=None)
   n8 = Get Parent Module Name!(self:Modular Rig Object Reference=@n5, InModuleName:Name=@n7) -> (ReturnValue:Name=None) {is=title K2Node_CallFunction}
   > n9 = Execute Python Script!(
    execute
    PythonScript="\r\noutput_string = f\"{input_string.split('/')[0]}/{input_string.split('/')[-1].replace(' ', '_', 1)}\""
    input_string:String=Append!(A="To String (Name)"(InName=@n7), B=/, C:String="To String (Name)"(InName=@n8), D:String=" PV Null") -> (ReturnValue)
   ) -> (then, ReturnValue, output_string:String) {inputs [input_string], outputs [output_string]}
   > n10 = GetWorldSpaceReferenceKey()
   > SetControlRigSpace(
    InSequence=Get Level Sequence()
    InControlRig=Get Selected Rig for Prop()
    InControlName=Get().Local Rig Element_Name
    InSpaceKey_Type=@n10
    InSpaceKey_Name=@n10.ReturnValue_Name
    InTime_Value=int - int(A=Get Local Start() -> (Local Start=0))
    TimeUnit=DisplayRate
   )
   > n11 = GetDefaultParentKey()
   > SetControlRigSpace(
    InSequence=Get Level Sequence()
    InControlRig=Get Selected Rig for Prop()
    InControlName=Get().Local Rig Element_Name
    InSpaceKey_Type=@n11
    InSpaceKey_Name=@n11.ReturnValue_Name
    InTime_Value=int - int(A=Get Local Start() -> (Local Start=0))
    TimeUnit=DisplayRate
   )
   > SetControlRigSpace(
    InSequence=Get Level Sequence()
    InControlRig=Get Selected Rig for Prop()
    InControlName=Get().Local Rig Element_Name
    InSpaceKey_Type=Null
    InSpaceKey_Name=String To Name(InString=@n9.output_string)
    InTime_Value=Get Local Start() -> (Local Start=0)
    TimeUnit=DisplayRate
   )
   n12 = Get l_ik_apply_cbox()
   n13 = "To String (Name)"(InName=GetModuleFName(self=Get Local Rig Heirarchy(), InItem=Get Local Rig Element()) -> (ReturnValue=None))
   n14 = Branch(execute=@n6.then_1, Condition=AND Boolean(A=IsChecked(self=@n12), B=EndsWith(SourceString=@n13, InSuffix=L, SearchCase=CaseSensitive)))
   n15 = Branch(execute=[@n14.then, @n30.then], Condition=Contains(SearchIn="Get Prop Global Control "(), Substring=v2))
   n16 = Get Local Rig Heirarchy()
   n17 = Get Local Rig Element()
   n18 = Append!(A="To String (Name)"(InName=GetModuleFName(self=@n16, InItem=@n17) -> (ReturnValue=None)), B=/Prop) -> (ReturnValue)
   > n19 = Set Local Arm Prop Space(Local Arm Prop Space:String=Append(A=@n18, B=" IK Null_2")) -> (Output_Get:String)
   n20 = Get Local Rig Heirarchy()
   n21 = Get Local Rig Element()
   n22 = Append(A="To String (Name)"(InName=GetModuleFName(self=@n20, InItem=@n21) -> (ReturnValue=None)), B=/IK)
   Execute Python Script!(
    execute=[@n28.then, @n19.then]
    PythonScript="import unreal\r\n\r\n## Create ControlRigSnapperSelection objects for the parent and the children\r\nparent = unreal.ControlRigSnapperSelection()\r\nchildren = unreal.ControlRigSnapperSelection()\r\n\r\n## Create a ControlRigForWorldTransforms object\r\n## Set to the proper control rig and the left hand control as the control\r\n## It is possible to have multiple control names in here\r\nchild_control_rig = unreal.ControlRigForWorldTransforms()\r\nchild_control_rig.control_rig = control_rig\r\nchild_control_rig.control_names = [prop_control]\r\nchildren.control_rigs = [child_control_rig]\r\n\r\n\r\nparent_control_rig = unreal.ControlRigForWorldTransforms()\r\nparent_control_rig.control_rig = control_rig\r\nparent_control_rig.control_names = [target_control]\r\nparent.control_rigs = [parent_control_rig]\r\n\r\n## Create and set snap settings\r\nsnap_settings = unreal.ControlRigSnapSettings()\r\nsnap_settings.keep_offset = True\r\nsnap_settings.snap_position = True\r\nsnap_settings.snap_rotation = True\r\nsnap_settings.snap_scale = False\r\nsnap_settings.baking_key_settings = unreal.BakingKeySettings.ALL_FRAMES\r\n\r\n## Take the ActorForWorldTransforms object, set it as the parent ControlRigSnapperSelection\r\n## Take the ControlRigForWorldTransforms object, set as the child ControlRigSnapperSelection\r\nchildren.control_rigs = [child_control_rig]\r\n\r\nunreal.ControlRigSequencerLibrary.snap_control_rig(level_sequence, unreal.FrameNumber(start_frame), unreal.FrameNumber(start_frame), children, parent, snap_settings)\r\n"
    control_rig:Control Rig Object Reference=Get Selected Rig for Prop()
    target_control:String=@n22
    prop_control:String=@n22
    level_sequence:Level Sequence Object Reference=Get Level Sequence()
    start_frame:int=Get Local Start() -> (Local Start=0)
    end_frame:int=Get Local End() -> (Local End=0)
   ) -> (then, ReturnValue) {inputs [control_rig, target_control, prop_control, level_sequence, start_frame, end_frame]}
   > n23 = GetWorldSpaceReferenceKey()
   n24 = "To String (Name)"(InName=GetModuleFName(self=Get Local Rig Heirarchy(), InItem=Get Local Rig Element()) -> (ReturnValue=None))
   > SetControlRigSpace(
    InSequence=Get Level Sequence()
    InControlRig=Get Selected Rig for Prop()
    InControlName=String To Name(InString=Append(A=@n24, B=/IK))
    InSpaceKey_Type=@n23
    InSpaceKey_Name=@n23.ReturnValue_Name
    InTime_Value=int - int(A=Get Local Start() -> (Local Start=0))
    TimeUnit=DisplayRate
   )
   n25 = "To String (Name)"(InName=GetModuleFName(self=Get Local Rig Heirarchy(), InItem=Get Local Rig Element()) -> (ReturnValue=None))
   > SetControlRigSpace(
    InSequence=Get Level Sequence()
    InControlRig=Get Selected Rig for Prop()
    InControlName=String To Name(InString=Append(A=@n25, B=/IK))
    InSpaceKey_Type=Null
    InSpaceKey_Name=String To Name(InString=Get Local Arm Prop Space() -> (Local Arm Prop Space:String))
    InTime_Value=Get Local Start() -> (Local Start=0)
    TimeUnit=DisplayRate
   )
   > n26 = GetDefaultParentKey()
   n27 = "To String (Name)"(InName=GetModuleFName(self=Get Local Rig Heirarchy(), InItem=Get Local Rig Element()) -> (ReturnValue=None))
   > SetControlRigSpace(
    InSequence=Get Level Sequence()
    InControlRig=Get Selected Rig for Prop()
    InControlName=String To Name(InString=Append(A=@n27, B=/IK))
    InSpaceKey_Type=@n26
    InSpaceKey_Name=@n26.ReturnValue_Name
    InTime_Value=int - int(A=Get Local Start() -> (Local Start=0))
    TimeUnit=DisplayRate
   )
   n28 = Set Local Arm Prop Space(execute=@n15.else, Local Arm Prop Space=Append(A=@n18, B=" IK Null"))
   n29 = Get r_ik_apply_cbox()
   n30 = Branch(execute=@n14.else, Condition=AND Boolean(A=EndsWith(SourceString=@n13, InSuffix=R, SearchCase=CaseSensitive), B=IsChecked(self=@n29)))
  ]}
 Arm and Attach Default Space {nodes [
   n1 = Arm and Attach Default Space!() -> (then, start frame:int) {is=title K2Node_FunctionEntry}
   > Set Local Start(Local Start=@n1) -> (Output_Get=0)
   > n2 = Set Local Rig Heirarchy(Local Rig Heirarchy=GetHierarchy(self=Get Selected Rig for Prop()))
   > n3 = For Each Loop(Array=Get Controls(self=@n2, bTraverse=true))
   > Set Local Rig Element(Local Rig Element=@n3)
   n4 = "To String (Name)"(InName=Get().Local Rig Element_Name)
   > n5 = Branch(Condition=AND Boolean(A=Contains(SearchIn=@n4, Substring=arm), B=EndsWith(SourceString=@n4, InSuffix=/PV)))
   n6 = "To String (Name)"(InName=GetModuleFName(self=Get Local Rig Heirarchy(), InItem=Get Local Rig Element()) -> (ReturnValue=None))
   > n7 = Set Local Target Control(Local Target Control:Name=String To Name(InString=Append(A=@n6, B=/IK))) -> (Output_Get:Name=None)
   Cast To ModularRig(execute=[@n7.then, @n10.then], Object=Get Selected Rig for Prop())
   > n8 = GetDefaultParentKey()
   > SetControlRigSpace(
    InSequence=Get Level Sequence()
    InControlRig=Get Selected Rig for Prop()
    InControlName=Get Local Target Control() -> (Local Target Control:Name=None)
    InSpaceKey_Type=@n8
    InSpaceKey_Name=@n8.ReturnValue_Name
    InTime_Value=Get Local Start() -> (Local Start=0)
    TimeUnit=DisplayRate
   )
   Branch(execute=@n5.else, Condition="Equal Exactly (String)"(A=@n4, B=Get Attach Target Name()))
   n9 = Get()
   > n10 = Set Local Target Control(Local Target Control=@n9.Local Rig Element_Name) -> (Output_Get=None)
   "To String (Name)"(InName=@n9.Local Rig Element_Name)
  ]}
 EventGraph {nodes [
   "On Clicked (ActivatePropButton)"() {is=title K2Node_ComponentBoundEvent}
   > Get Prop Control Info() {is=title K2Node_CallFunction}
   > Activate Prop Control() {is=title K2Node_CallFunction}
   "On Clicked (ApplyPropControlButton)"() {is=title K2Node_ComponentBoundEvent}
   > Get Prop Control Info()
   > Apply Prop Control() {is=title K2Node_CallFunction}
   "On Pressed (SelectPropButton)"() {is=title K2Node_ComponentBoundEvent}
   > ClearControlSelection!(execute, self:Control Rig Object Reference=Get Selected Rig for Prop(), bSetupUndo:bool=false) -> (then, ReturnValue:bool)
   > SelectControl!(
    execute
    self:Control Rig Object Reference=Get Selected Rig for Prop()
    InControlName:"Name (by ref)"=String To Name(InString="Get Prop Global Control "())
    bSelect:bool=true
    bSetupUndo:bool=false
   ) -> (then)
   "On Check State Changed (PropFullRangeCbox)"!() -> (OutputDelegate, then, bIsChecked:bool) {is=title K2Node_ComponentBoundEvent}
   n1 = Get PropSelectedKeyRangeCbox() -> (PropSelectedKeyRangeCbox:Editor Utility Check Box Object Reference)
   > SetIsChecked!(execute, self:Check Box Object Reference=@n1, InIsChecked:bool=InvertBool!(bValue:bool=IsChecked(self=@n1)) -> (ReturnValue:bool)) -> (then)
   "On Check State Changed (PropSelectedKeyRangeCbox)"!() -> (OutputDelegate, then, bIsChecked:bool) {is=title K2Node_ComponentBoundEvent}
   n2 = Get PropFullRangeCbox() -> (PropFullRangeCbox:Editor Utility Check Box Object Reference)
   > SetIsChecked(self=@n2, InIsChecked=InvertBool(bValue=IsChecked(self=@n2)))
   IKFK_V2_Match_Event() {is=title K2Node_CustomEvent}
   > DelayUntilNextTick!(execute, WorldContextObject:Object Reference) -> (then)
   > Set Current Local Time!(
    execute
    PlaybackParams_Frame_FrameNumber_Value:int=Get!() -> (Range_bHasStart:bool=~, Range_bHasEnd:bool=~, Range_InclusiveStart:int, Range_ExclusiveEnd:int).Range_InclusiveStart
    PlaybackParams_Frame_SubFrame:float=0
    PlaybackParams_Time:float=0
    PlaybackParams_MarkedFrame:String
    PlaybackParams_Timecode:Timecode Structure=(Hours=0,Minutes=0,Seconds=0,Frames=0,Subframe=0,bDropFrameFormat=False)
    PlaybackParams_PositionType:EMovieScenePositionType Enum=Frame
    PlaybackParams_UpdateMethod:EUpdatePositionMethod Enum=Play
    PlaybackParams_bHasJumped:bool=False
    TimeUnit:EMovieSceneTimeUnit Enum=DisplayRate
   ) -> (then) {is=title K2Node_CallFunction}
   > IKFKMatch_BipedV2()
  ]}
 Get Frames {nodes [
   Get Frames() {is=title K2Node_FunctionEntry}
   > n1 = Sequence()
   > n2 = Branch(Condition=IsChecked(self=Get PropFullRangeCbox()))
   > n3 = Execute Python Script!(
    execute
    PythonScript="start = level_sequence.get_playback_start() \r\nend = level_sequence.get_playback_end()"
    level_sequence=Get Level Sequence()
   ) -> (then, ReturnValue, start:int, end:int) {inputs [level_sequence], outputs [start,end]}
   > Set local start(local start:int=@n3.start) -> (Output_Get:int=0)
   > Set local end(local end:int=@n3.end) -> (Output_Get:int=0)
   > Set local valid frames(local valid frames:bool=true) -> (Output_Get:bool)
   n4 = Execute Python Script!(
    execute=@n2.else
    PythonScript="from AnimLibrary.core.utils import misc\r\n\r\nvalid_frames = True\r\n\r\nstart, end = misc.get_selected_keys_range()\r\n\r\nif not start:\r\n   valid_frames = False\r\n   print('Start Frame is invalid')\r\n   start = 0\r\n\r\nif not end:\r\n    valid_frames = False\r\n    end = 0\r\n    print('End Frame is invalid')\r\n"
   ) -> (then, ReturnValue, start, end, valid_frames:bool=~) {outputs [start,end,valid_frames]}
   > n5 = Set local valid frames(local valid frames=@n4.valid_frames)
   > Branch(Condition=@n5)
   > Set local start(local start=@n4.start) -> (Output_Get=0)
   > Set local end(local end=@n4.end) -> (Output_Get=0)
   Return Node!(
    execute=@n1.then_1
    valid frames:bool=Get local valid frames() -> (local valid frames:bool)
    start frame:int=Get local start() -> (local start:int=0)
    end frame:int=Get local end() -> (local end:int=0)
   )
  ]}
 Get Prop Control Info {nodes [
   Get Prop Control Info() {is=title K2Node_FunctionEntry}
   > Clear(TargetArray:Array of Movie Scene Scripting Channel Object References=Get Prop Global control Channels())
   > n1 = Is Valid(InputObject=Get Selected Parent Rig() -> (Selected Parent Rig:Control Rig Object Reference))
   Set Text(execute=@n1.Is Not Valid, self=Get PropControlText() -> (PropControlText:Text Object Reference))
   > SetIsEnabled(self=Get ActivatePropButton() -> (ActivatePropButton:Editor Utility Button Object Reference), bInIsEnabled=false)
   > Set Selected Rig for Prop(Selected Rig for Prop:Control Rig Object Reference) -> (Output_Get:Control Rig Object Reference)
   > SetIsEnabled(self=Get ApplyPropControlButton() -> (ApplyPropControlButton:Editor Utility Button Object Reference), bInIsEnabled=false)
   n2 = Sequence!(execute=@n1.Is Valid) -> (then_0, then_1, then_2:Exec)
   > n3 = Execute Python Script!(
    execute
    PythonScript="import unreal\r\n\r\nold_prop_location = ['MR_Biped_v01'.lower(), 'MR_Biped_FN1_v01'.lower()]\r\ncontrol = ''\r\nfound_item=''\r\nfound=False\r\nattach_control_target  = \"Hand Attach R\"\r\n\r\n## figure out the control base name\r\nif control_rig.get_name().lower() in old_prop_location:\r\n    search = \"Prop v2\"\r\n    attach_mod = \"Hand Attach and Weapon R\"\r\n\r\nelse:\r\n    search = 'Prop'\r\n    attach_mod = \"Attach\"\r\n\r\nattach_target_name = f'{attach_mod}/{attach_control_target}'\r\n\r\ntarget_control = f'{search}/Prop Global Vis'\r\n## comb through the list of controls for our target. The channel name can be different (ex: _2, _4, etc.)\r\nall_controls = control_rig.get_hierarchy().get_controls(traverse=True)\r\nfor control in all_controls:\r\n    if target_control.lower() in str(control.name).lower():\r\n        print('Found,', control)\r\n        found=True,\r\n        found_item=control\r\n   \r\nif found:\r\n     control = found_item\r\nelse:\r\n    control = ''\r\n"
    control_rig:Control Rig Object Reference=Get Selected Parent Rig()
   ) -> (then, ReturnValue, control:Rig Element Key Structure, attach_target_name:String) {inputs [control_rig], outputs [control, attach_target_name]}
   > n4 = Set local control check!(execute, local control check:Rig Element Key Structure=@n3.control) -> (then, Output_Get_Type:ERigElementType Enum, Output_Get_Name:Name)
   > n5 = Branch(Condition="Not Equal (Name)"(A=@n4.Output_Get_Name) -> (ReturnValue=~))
   > Set Attach Target Name(Attach Target Name:String=@n3.attach_target_name) -> (Output_Get:String)
   > n6 = Set Prop Control!(execute, Prop Control:Rig Element Key Structure=@n3.control) -> (then, Output_Get_Type:ERigElementType Enum, Output_Get_Name:Name)
   > n7 = Set Selected Rig for Prop(Selected Rig for Prop=Get Selected Parent Rig())
   n8 = "To String (Name)"(InName=@n6.Output_Get_Name)
   "Set Prop Global Control "(execute=[@n7.then, @n10.then], "Prop Global Control ":String=Append(A=Split(SourceString=@n8, InStr=/, SearchDir=FromStart), B=/Prop Global)) -> (Output_Get:String)
   n9 = "Get Prop Global Control "()
   > Set Text(Text="To Text (String)"(InString=Append(A=@n9, B=":  ")), self=Get PropControlText())
   > SetIsEnabled(self=Get ActivatePropButton(), bInIsEnabled=true)
   > SetIsEnabled(self=Get ApplyPropControlButton(), bInIsEnabled=true)
   > SetIsEnabled(self=Get SelectPropButton() -> (SelectPropButton:Editor Utility Button Object Reference), bInIsEnabled=true)
   > Set Valid For Prop(Valid For Prop:bool=InvertBool(bValue=Get Valid For Prop() -> (Valid For Prop:bool))) -> (Output_Get:bool)
   n10 = Branch(execute=@n5.else, Condition=Get Valid For Prop())
   SetIsEnabled(execute=@n10.else, self=Get ActivatePropButton(), bInIsEnabled=false)
   > Set Text(Text="LOCTEXT(\"No Prop!\")", self=Get PropControlText())
   > SetIsEnabled(self=Get ApplyPropControlButton(), bInIsEnabled=false)
   > SetIsEnabled(self=Get SelectPropButton(), bInIsEnabled=false)
   > Set Valid For Prop(Valid For Prop=false)
   n11 = GetControlRigs(execute=@n2.then_1, LevelSequence=Get Level Sequence())
   > n12 = For Each Loop!(Exec, Array:Array of Control Rig Sequencer Binding Proxy Structures=@n11) -> (
    LoopBody
    Array Element_Proxy:Movie Scene Binding Proxy Structure
    Array Element_ControlRig:Control Rig Object Reference
    Array Element_Track:Movie Scene Control Rig Parameter Track Object Reference
    Array Index
    Completed
   )
   > Branch(Condition="Equal (Object)"(A=Get Selected Rig for Prop(), B=@n12.Array Element_ControlRig))
   > n13 = GetSections(Track=@n12.Array Element_Track)
   > n14 = For Each Loop!(Exec, Array:Array of Movie Scene Section Object References=@n13) -> (LoopBody, Array Element:Movie Scene Section Object Reference, Array Index, Completed)
   > n15 = GetAllChannels(Section=@n14)
   > n16 = For Each Loop(Array:Array of Movie Scene Scripting Channel Object References=@n15) -> (Array Element:Movie Scene Scripting Channel Object Reference)
   > Set Local Current Channel(Local Current Channel:Movie Scene Scripting Channel Object Reference=@n16) -> (Output_Get:Movie Scene Scripting Channel Object Reference)
   n17 = Get Local Current Channel() -> (Local Current Channel:Movie Scene Scripting Channel Object Reference)
   > Branch(Condition=Contains(SearchIn=GetDisplayName(Object=@n17), Substring="Get Prop Global Control "()))
   > Add(
    TargetArray:Array of Movie Scene Scripting Channel Object References=Get Prop Global control Channels()
    NewItem:"Movie Scene Scripting Channel Object Reference (by ref)"=Get Local Current Channel()
   ) {is=title K2Node_CallArrayFunction}
   "To String (Name)"(InName=@n4.Output_Get_Name)
   "Not Equal Exactly (String)"(A=GetDisplayName(Object=Get Selected Rig for Prop()))
  ]}
 IKFKMatch_BipedV2 {nodes [
   IKFKMatch_BipedV2() {is=title K2Node_FunctionEntry}
   > n1 = LoadAsset_Blocking(Asset=/AnimInEdTools/EpicControlRig/DA_IkFkMatchingData_Biped_v02)
   > n2 = Cast To PDA_IkFkMatchingBakeData!(execute, Object=@n1) -> (then, CastFailed, AsPDA Ik Fk Matching Bake Data:PDA Ik Fk Matching Bake Data Object Reference) {is=title K2Node_DynamicCast}
   > Set Local IKFK Data Array(Local IKFK Data Array:Array of Ik Fk Matching Data Structures=Get Ik Fk Matching Data!(self:PDA Ik Fk Matching Bake Data Object Reference=@n2) -> (Ik Fk Matching Data:Array of Ik Fk Matching Data Structures)) -> (Output_Get:Array of Ik Fk Matching Data Structures)
   > n3 = Create Widget!(
    execute:Exec
    WorldContextObject:Object Reference
    Class:User Widget Class Reference=/AnimInEdTools/IkFkMatchingTool/EUW_IkFkMatching.EUW_IkFkMatching_C
    OwningPlayer:Player Controller Object Reference
    Primary Control Rig:Control Rig Object Reference
    Animatable Creature:Animatable Creature Object Reference
   ) -> (then:Exec, ReturnValue:EUW Ik Fk Matching Object Reference)
   > n4 = Set Local EUW IKFK Matcher(Local EUW IKFK Matcher:EUW Ik Fk Matching Object Reference=@n3) -> (Output_Get:EUW Ik Fk Matching Object Reference)
   > Bind Event to IKFK Match Finished E Dispatcher(self:EUW Ik Fk Matching Object Reference=@n4, Delegate=Create Event()) {is=title K2Node_AddDelegate}
   > n5 = Cast To ModularRig(Object=Get Selected Rig for Prop())
   > n6 = Sequence()
   Run Match(
    execute=@n6.then_1
    self:EUW Ik Fk Matching Object Reference=Get Local EUW IKFK Matcher() -> (Local EUW IKFK Matcher:EUW Ik Fk Matching Object Reference)
   )
   Return Node!(execute=@n6.then_2)
   Set Modular Rig!(execute=@n6.then_0, Modular Rig:Modular Rig Object Reference=@n5, self:EUW Ik Fk Matching Object Reference=Get Local EUW IKFK Matcher()) -> (then, Output_Get:Modular Rig Object Reference)
   > Set!(execute, Frame Rate_Numerator:int=30, Frame Rate_Denominator:int=1, self:EUW Ik Fk Matching Object Reference=Get Local EUW IKFK Matcher()) -> (then, Output_Get:Frame Rate Structure)
   > Set Rig Hierarchy!(
    execute
    Rig Hierarchy:Rig Hierarchy Object Reference=GetHierarchy(self=Get Selected Rig for Prop())
    self:EUW Ik Fk Matching Object Reference=Get Local EUW IKFK Matcher()
   ) -> (then, Output_Get:Rig Hierarchy Object Reference)
   > SetIsChecked(
    self=Get CurrentFrameCheckBox!(self:EUW Ik Fk Matching Object Reference=Get Local EUW IKFK Matcher()) -> (CurrentFrameCheckBox:Editor Utility Check Box Object Reference)
    InIsChecked=false
   )
   > Set Playback Range!(
    execute
    Playback Range:Sequencer Scripting Range Structure=Get Range() -> (Range:Sequencer Scripting Range Structure)
    self:EUW Ik Fk Matching Object Reference=Get Local EUW IKFK Matcher()
   ) -> (then, Output_Get:Sequencer Scripting Range Structure)
   > Set Current Matching Frame!(
    execute
    Current Matching Frame:int=Get() -> (Range_bHasStart=~, Range_bHasEnd=~).Range_InclusiveStart
    self:EUW Ik Fk Matching Object Reference=Get Local EUW IKFK Matcher()
   ) -> (then, Output_Get:int=0)
   > SetIsChecked(
    self=Get CurrentKeysOnlyCheckBox!(self:EUW Ik Fk Matching Object Reference=Get Local EUW IKFK Matcher()) -> (CurrentKeysOnlyCheckBox:Editor Utility Check Box Object Reference)
    InIsChecked=false
   )
   > Set IsModular!(execute, IsModular:bool=true, self:EUW Ik Fk Matching Object Reference=Get Local EUW IKFK Matcher()) -> (then, Output_Get:bool)
   > n7 = For Each Loop(Array:Array of Ik Fk Matching Data Structures=Get Local IKFK Data Array() -> (Local IKFK Data Array:Array of Ik Fk Matching Data Structures)) -> (Array Element:Ik Fk Matching Data Structure)
   > Set Local IKFK Data Struct(Local IKFK Data Struct:Ik Fk Matching Data Structure=@n7) -> (Output_Get:Ik Fk Matching Data Structure)
   > Branch(Condition="integer < integer"(A=@n7.Array Index, B=2))
   > n8 = GetSections(Track=Break Control Rig Sequencer Binding Proxy!(ControlRigSequencerBindingProxy:"Control Rig Sequencer Binding Proxy Structure (by ref)"=Get control rig sequencer proxy() -> (control rig sequencer proxy:Control Rig Sequencer Binding Proxy Structure)) -> (
     Proxy:Movie Scene Binding Proxy Structure
     ControlRig:Control Rig Object Reference
     Track:Movie Scene Control Rig Parameter Track Object Reference
    ) {is=title K2Node_BreakStruct}.Track)
   > n9 = For Each Loop(Array:Array of Movie Scene Section Object References=@n8) -> (Array Element:Movie Scene Section Object Reference)
   > n10 = GetAllChannels(Section=@n9)
   > n11 = For Each Loop(Array:Array of Movie Scene Scripting Channel Object References=@n10) -> (Array Element:Movie Scene Scripting Channel Object Reference)
   n12 = Break Ik Fk Matching Data!(IkFkMatchingData:"Ik Fk Matching Data Structure (by ref)"=Get Local IKFK Data Struct() -> (Local IKFK Data Struct:Ik Fk Matching Data Structure)) -> (
    Name_19_#1:Name
    IkFkSwitchAttr_20_#2:Name
    MatchIKEvent_21_#3:Name
    MatchFKEvent_22_#4:Name
    IKControls_10_#5:Array of Names
    FKControls_12_#6:Array of Names
   ) {is=title K2Node_BreakStruct}
   n13 = "To String (Name)"(InName=@n12)
   n14 = "To String (Name)"(InName=@n12.IkFkSwitchAttr_20_#2)
   > Branch(Condition=Contains(SearchIn=GetDisplayName(Object=@n11), Substring=Append!(A=@n13, B=/, C:String=@n14) -> (ReturnValue)))
   > n15 = Cast To MovieSceneScriptingBoolChannel(Object=@n11)
   > Add!(
    execute
    TargetMap:Map of Movie Scene Scripting Bool Channel Object References to Booleans=Get IK FK Switch Channels!(self:EUW Ik Fk Matching Object Reference=Get Local EUW IKFK Matcher()) -> (IK FK Switch Channels:Map of Movie Scene Scripting Bool Channel Object References to Booleans)
    Key:"Movie Scene Scripting Bool Channel Object Reference (by ref)"=@n15
    Value:"Boolean (by ref)"=true
   ) -> (then) {is=title K2Node_CallFunction}
   > Add(
    TargetArray:Array of Ik Fk Matching Data Structures=Get Ik Fk Matching Data(self:EUW Ik Fk Matching Object Reference=Get Local EUW IKFK Matcher())
    NewItem:"Ik Fk Matching Data Structure (by ref)"=Get Local IKFK Data Struct()
   ) {is=title K2Node_CallArrayFunction}
   > Add(
    TargetArray:Array of Strings=Get Module Paths!(self:EUW Ik Fk Matching Object Reference=Get Local EUW IKFK Matcher()) -> (Module Paths:Array of Strings)
    NewItem:"String (by ref)"=@n13
   )
  ]}
 Set Context {nodes [
   n1 = Set Context!() -> (
    then
    In Level Sequence:Level Sequence Object Reference
    In Sequencer Proxy:Control Rig Sequencer Binding Proxy Structure
    In Parent Rig:Control Rig Object Reference
   ) {is=title K2Node_FunctionEntry}
   > Set Level Sequence(Level Sequence:Level Sequence Object Reference=@n1) -> (Output_Get:Level Sequence Object Reference)
   > Set control rig sequencer proxy(control rig sequencer proxy=@n1.In Sequencer Proxy)
   > Set Selected Parent Rig(Selected Parent Rig:Control Rig Object Reference=@n1.In Parent Rig) -> (Output_Get:Control Rig Object Reference)
  ]}
 Spawn and Key Locator {nodes [
   Spawn and Key Locator!() -> (then, start time:int, Sequence:Movie Scene Sequence Object Reference, end time:int) {is=title K2Node_FunctionEntry}
   > n1 = Sequence!(execute) -> (then_0, then_1)
   > n2 = SpawnActorFromObject!(execute, ObjectToUse:Object Reference=/AnimatorKit/Locator/Locator, Location:Vector="0, 0, 0", Rotation:Rotator, bTransient:bool=false) -> (then, ReturnValue:Actor Object Reference)
   > n3 = Set local actor(local actor:Actor Object Reference=@n2) -> (Output_Get:Actor Object Reference)
   > n4 = AddPossessable!(
    execute
    Sequence:Movie Scene Sequence Object Reference=Get Sequence() -> (Sequence:Movie Scene Sequence Object Reference)
    ObjectToPossess:Object Reference=@n3
   ) -> (then, ReturnValue:Movie Scene Binding Proxy Structure)
   > n5 = Set local binding(local binding:Movie Scene Binding Proxy Structure=@n4) -> (Output_Get:Movie Scene Binding Proxy Structure)
   > n6 = AddTrack!(
    execute
    InBinding:"Movie Scene Binding Proxy Structure (by ref)"=@n5
    TrackType:Movie Scene Track Class Reference=/Script/MovieSceneTracks.MovieScene3DTransformTrack
   ) -> (then, ReturnValue:Movie Scene Track Object Reference)
   > n7 = AddSection!(execute, Track:Movie Scene Track Object Reference=@n6) -> (then, ReturnValue:Movie Scene Section Object Reference)
   > SetRange!(execute, Section:Movie Scene Section Object Reference=@n7, StartFrame:int=0, EndFrame:int=Get end time() -> (end time:int=0)) -> (then)
   > n8 = GetAllChannels(Section=@n7)
   > n9 = For Each Loop(Array=@n8)
   > n10 = Cast To MovieSceneScriptingDoubleChannel!(execute, Object=@n9) -> (then, CastFailed, AsMovie Scene Scripting Double Channel:Movie Scene Scripting Double Channel Object Reference) {is=title K2Node_DynamicCast}
   > n11 = For Loop(LastIndex=Get end time() -> (end time=0))
   > "Add Key (Double)"!(
    execute
    self:Movie Scene Scripting Double Channel Object Reference=@n10
    InTime_Value:int=@n11
    NewValue:double=0
    SubFrame:float=0
    TimeUnit:EMovieSceneTimeUnit Enum=DisplayRate
    InInterpolation:EMovieSceneKeyInterpolation Enum=Auto
   ) -> (then, ReturnValue:Movie Scene Scripting Double Key Object Reference)
   Return Node!(
    execute=@n1.then_1
    output actor:Actor Object Reference=Get local actor() -> (local actor:Actor Object Reference)
    output binding:Movie Scene Binding Proxy Structure=Get local binding() -> (local binding:Movie Scene Binding Proxy Structure)
   )
  ]}
]
variables [
 Attach Target Name {type=string, default=""}
 control rig sequencer proxy {
  type=ControlRigSequencerBindingProxy
  default {controlRig=null, track=null, proxy {bindingId=#7, sequence=null}}
 }
 Level Sequence {type=LevelSequence, default=null}
 Prop Control {type=RigElementKey, default {name=None, type=None}}
 "Prop Global Control " {type=string, default=""}
 Prop Global control Channels {type="array<MovieSceneScriptingChannel>", default []}
 Range {
  type=SequencerScriptingRange
  default {bHasEnd=false, bHasStart=false, exclusiveEnd=0, inclusiveStart=0, internalRate {denominator=1, numerator=60000}}
 }
 Selected Parent Rig {type=ControlRig, default=null}
 Selected Rig for Prop {type=ControlRig, default=null}
 Valid For Prop {type=boolean}
]
widgets {
 info {namedSlotCount=2, rootWidgetClass=/Script/Blutility.EditorUtilityExpandableArea, widgetCount=32}
 widgets [
  EditorUtilityExpandableArea_167 {widgetClassPath=&EditorUtilityExpandableArea, bIsVariable=true, widget=EditorUtilityExpandableArea_167}
  Border_100 {
   widgetClassPath=&Border
   namedSlotHost=EditorUtilityExpandableArea_167
   widget=Border_100
   properties {brushColor=<rgba 0.0467 0.0467 0.0497 1>, padding=<ltrb 0 0 0 0>}
  }
  PropControlSetupText {
   widgetClassPath=&TextBlock
   parent=Border_100
   widget=PropControlSetupText
   properties {
    text=Prop Control Setup
    colorAndOpacity {colorUseRule=UseColor_Foreground, specifiedColor=<rgba 1 1 1 1>}
    font {
     bForceMonospaced=false
     bMaterialIsStencil=false
     fontMaterial=null
     fontObject=&Roboto
     letterSpacing=0
     monospacedWidth=1
     size=13.5
     skewAmount=0
     typefaceFontName=Bold
     outlineSettings {
      bApplyOutlineToDropShadows=false
      bMiteredCorners=false
      bSeparateFillAlpha=false
      outlineMaterial=null
      outlineSize=0
      outlineColor=<rgba 0 0 0 1>
     }
    }
   }
   slot {class=BorderSlot, properties {padding=<ltrb 0 0 0 0>}}
  }
  Border_102 {
   widgetClassPath=&Border
   namedSlotHost=EditorUtilityExpandableArea_167
   widget=Border_102
   properties {
    background {
     bIsDynamicallyLoaded=false
     drawAs=Box
     imageType=NoImage
     mirroring=NoMirror
     resourceName=None
     resourceObject=null
     tiling=NoTile
     imageSize=<32 32>
     margin=<ltrb 0 0 0 0>
     outlineSettings {
      bUseBrushTransparency=false
      roundingType=HalfHeightRadius
      width=0
      color {colorUseRule=UseColor_Specified, specifiedColor=<rgba 0 0 0 0>}
      cornerRadii=<0 0 0 0>
     }
     tintColor {colorUseRule=UseColor_Specified, specifiedColor=<rgba 1 1 1 1>}
     uVRegion {bIsValid=false, max=<0 0>, min=<0 0>}
    }
    brushColor=<rgba 0.0185 0.0185 0.0194 1>
    padding=<ltrb 0 2 4 2>
   }
  }
  VerticalBox_8 {
   widgetClassPath=&VerticalBox
   parent=Border_102
   widget=VerticalBox_8
   slot {class=BorderSlot, properties {padding=<ltrb 0 2 4 2>}}
  }
  HorizontalBox_388 {
   widgetClassPath=&HorizontalBox
   parent=VerticalBox_8
   widget=HorizontalBox_388
   slot {class=VerticalBoxSlot, properties {horizontalAlignment=HAlign_Left, padding=<ltrb 0 5 0 0>}}
  }
  PropControlText {
   widgetClassPath=&TextBlock
   bIsVariable=true
   parent=HorizontalBox_388
   widget=PropControlText
   properties {
    visibility=HitTestInvisible
    font {
     bForceMonospaced=false
     bMaterialIsStencil=false
     fontMaterial=null
     fontObject=&Roboto
     letterSpacing=0
     monospacedWidth=1
     size=12
     skewAmount=0
     typefaceFontName=Regular
     outlineSettings {
      bApplyOutlineToDropShadows=false
      bMiteredCorners=false
      bSeparateFillAlpha=false
      outlineMaterial=null
      outlineSize=0
      outlineColor=<rgba 0 0 0 1>
     }
    }
   }
   slot {
    class=HorizontalBoxSlot
    properties {horizontalAlignment=HAlign_Left, verticalAlignment=VAlign_Center, padding=<ltrb 4 2 3 2>, size {sizeRule=Fill, value=1}}
   }
  }
  ActivatePropButton {
   widgetClassPath=&EditorUtilityButton
   parent=HorizontalBox_388
   widget=ActivatePropButton
   properties {toolTipText="Snap the Prop Control to the Hand Attach R. Will clear prop control keys AFTER the start frame!"}
   slot {class=HorizontalBoxSlot, properties {horizontalAlignment=HAlign_Center, verticalAlignment=VAlign_Center}}
  }
  ActivatePropText {
   widgetClassPath=&TextBlock
   bIsVariable=true
   parent=ActivatePropButton
   widget=ActivatePropText
   properties {
    text=Activate
    visibility=HitTestInvisible
    font {
     bForceMonospaced=false
     bMaterialIsStencil=false
     fontMaterial=null
     fontObject=&Roboto
     letterSpacing=0
     monospacedWidth=1
     size=12
     skewAmount=0
     typefaceFontName=Regular
     outlineSettings {
      bApplyOutlineToDropShadows=false
      bMiteredCorners=false
      bSeparateFillAlpha=false
      outlineMaterial=null
      outlineSize=0
      outlineColor=<rgba 0 0 0 1>
     }
    }
   }
   slot {class=ButtonSlot, properties {horizontalAlignment=HAlign_Left}}
  }
  ApplyPropControlButton {
   widgetClassPath=&EditorUtilityButton
   parent=HorizontalBox_388
   widget=ApplyPropControlButton
   slot {class=HorizontalBoxSlot, properties {horizontalAlignment=HAlign_Center, verticalAlignment=VAlign_Center, padding=<ltrb 2 0 0 0>}}
  }
  ApplyPropText {
   widgetClassPath=&TextBlock
   bIsVariable=true
   parent=ApplyPropControlButton
   widget=ApplyPropText
   properties {
    text=Apply Prop Control setup
    toolTipText="Bake the Prop Control Setup, set the Arms to IK and Prop Space, and set the Arm PVs to Clavicle space."
    visibility=HitTestInvisible
    font {
     bForceMonospaced=false
     bMaterialIsStencil=false
     fontMaterial=null
     fontObject=&Roboto
     letterSpacing=0
     monospacedWidth=1
     size=12
     skewAmount=0
     typefaceFontName=Regular
     outlineSettings {
      bApplyOutlineToDropShadows=false
      bMiteredCorners=false
      bSeparateFillAlpha=false
      outlineMaterial=null
      outlineSize=0
      outlineColor=<rgba 0 0 0 1>
     }
    }
   }
   slot {class=ButtonSlot, properties {horizontalAlignment=HAlign_Left}}
  }
  SelectPropButton {
   widgetClassPath=&EditorUtilityButton
   parent=HorizontalBox_388
   widget=SelectPropButton
   slot {class=HorizontalBoxSlot, properties {horizontalAlignment=HAlign_Center, verticalAlignment=VAlign_Center, padding=<ltrb 2 0 0 0>}}
  }
  SelectPropText {
   widgetClassPath=&TextBlock
   bIsVariable=true
   parent=SelectPropButton
   widget=SelectPropText
   properties {
    text=Select Prop Control
    toolTipText="Bake the Prop Control Setup, set the Arms to IK and Prop Space, and set the Arm PVs to Clavicle space."
    visibility=HitTestInvisible
    font {
     bForceMonospaced=false
     bMaterialIsStencil=false
     fontMaterial=null
     fontObject=&Roboto
     letterSpacing=0
     monospacedWidth=1
     size=12
     skewAmount=0
     typefaceFontName=Regular
     outlineSettings {
      bApplyOutlineToDropShadows=false
      bMiteredCorners=false
      bSeparateFillAlpha=false
      outlineMaterial=null
      outlineSize=0
      outlineColor=<rgba 0 0 0 1>
     }
    }
   }
   slot {class=ButtonSlot, properties {horizontalAlignment=HAlign_Left}}
  }
  HorizontalBox_2 {
   widgetClassPath=&HorizontalBox
   parent=VerticalBox_8
   widget=HorizontalBox_2
   slot {class=VerticalBoxSlot, properties {padding=<ltrb 8 8 0 0>}}
  }
  VerticalBox_9 {
   widgetClassPath=&VerticalBox
   parent=HorizontalBox_2
   widget=VerticalBox_9
   slot {class=HorizontalBoxSlot, properties {size {sizeRule=Fill, value=1}}}
  }
  TextBlock_364 {
   widgetClassPath=&TextBlock
   parent=VerticalBox_9
   widget=TextBlock_364
   properties {
    text=Range
    font {
     bForceMonospaced=false
     bMaterialIsStencil=false
     fontMaterial=null
     fontObject=&Roboto
     letterSpacing=0
     monospacedWidth=1
     size=12
     skewAmount=0
     typefaceFontName=Regular
     outlineSettings {
      bApplyOutlineToDropShadows=false
      bMiteredCorners=false
      bSeparateFillAlpha=false
      outlineMaterial=null
      outlineSize=0
      outlineColor=<rgba 0 0 0 1>
     }
    }
   }
   slot {class=VerticalBoxSlot}
  }
  Border {
   widgetClassPath=&Border
   parent=VerticalBox_9
   widget=Border
   properties {
    background {
     bIsDynamicallyLoaded=false
     drawAs=Image
     imageType=NoImage
     mirroring=NoMirror
     resourceName=None
     resourceObject=null
     tiling=NoTile
     imageSize=<32 32>
     margin=<ltrb 0 0 0 0>
     outlineSettings {
      bUseBrushTransparency=false
      roundingType=HalfHeightRadius
      width=0
      color {colorUseRule=UseColor_Specified, specifiedColor=<rgba 0 0 0 0>}
      cornerRadii=<0 0 0 0>
     }
     tintColor {colorUseRule=UseColor_Specified, specifiedColor=<rgba 0.0103 0.0103 0.0103 1>}
     uVRegion {bIsValid=false, max=<0 0>, min=<0 0>}
    }
    desiredSizeScale=<0.5 0.5>
   }
   slot {class=VerticalBoxSlot, properties {padding=<ltrb 0 5 0 5>}}
  }
  HorizontalBox_0 {widgetClassPath=&HorizontalBox, parent=VerticalBox_9, widget=HorizontalBox_0, slot {class=VerticalBoxSlot}}
  VerticalBox_3 {widgetClassPath=&VerticalBox, parent=HorizontalBox_0, widget=VerticalBox_3, slot {class=HorizontalBoxSlot}}
  PropFullRangeCbox {
   widgetClassPath=&EditorUtilityCheckBox
   bIsVariable=true
   parent=VerticalBox_3
   widget=PropFullRangeCbox
   properties {checkedState=Checked}
   slot {class=VerticalBoxSlot}
  }
  FullRangeText {
   widgetClassPath=&TextBlock
   parent=PropFullRangeCbox
   widget=FullRangeText
   properties {
    text=Use Start/End Range
    font {
     bForceMonospaced=false
     bMaterialIsStencil=false
     fontMaterial=null
     fontObject=&Roboto
     letterSpacing=0
     monospacedWidth=1
     size=12
     skewAmount=0
     typefaceFontName=Regular
     outlineSettings {
      bApplyOutlineToDropShadows=false
      bMiteredCorners=false
      bSeparateFillAlpha=false
      outlineMaterial=null
      outlineSize=0
      outlineColor=<rgba 0 0 0 1>
     }
    }
   }
   slot {class=PanelSlot}
  }
  PropSelectedKeyRangeCbox {
   widgetClassPath=&EditorUtilityCheckBox
   bIsVariable=true
   parent=VerticalBox_3
   widget=PropSelectedKeyRangeCbox
   slot {class=VerticalBoxSlot}
  }
  SelectedRangeText {
   widgetClassPath=&TextBlock
   parent=PropSelectedKeyRangeCbox
   widget=SelectedRangeText
   properties {
    text=Use Selected Key Range
    font {
     bForceMonospaced=false
     bMaterialIsStencil=false
     fontMaterial=null
     fontObject=&Roboto
     letterSpacing=0
     monospacedWidth=1
     size=12
     skewAmount=0
     typefaceFontName=Regular
     outlineSettings {
      bApplyOutlineToDropShadows=false
      bMiteredCorners=false
      bSeparateFillAlpha=false
      outlineMaterial=null
      outlineSize=0
      outlineColor=<rgba 0 0 0 1>
     }
    }
   }
   slot {class=PanelSlot}
  }
  VerticalBox {
   widgetClassPath=&VerticalBox
   parent=HorizontalBox_2
   widget=VerticalBox
   slot {class=HorizontalBoxSlot, properties {size {sizeRule=Fill, value=1}}}
  }
  TextBlock {
   widgetClassPath=&TextBlock
   parent=VerticalBox
   widget=TextBlock
   properties {
    text=Apply To
    font {
     bForceMonospaced=false
     bMaterialIsStencil=false
     fontMaterial=null
     fontObject=&Roboto
     letterSpacing=0
     monospacedWidth=1
     size=12
     skewAmount=0
     typefaceFontName=Regular
     outlineSettings {
      bApplyOutlineToDropShadows=false
      bMiteredCorners=false
      bSeparateFillAlpha=false
      outlineMaterial=null
      outlineSize=0
      outlineColor=<rgba 0 0 0 1>
     }
    }
   }
   slot {class=VerticalBoxSlot}
  }
  Border_1 {
   widgetClassPath=&Border
   parent=VerticalBox
   widget=Border_1
   properties {
    background {
     bIsDynamicallyLoaded=false
     drawAs=Image
     imageType=NoImage
     mirroring=NoMirror
     resourceName=None
     resourceObject=null
     tiling=NoTile
     imageSize=<32 32>
     margin=<ltrb 0 0 0 0>
     outlineSettings {
      bUseBrushTransparency=false
      roundingType=HalfHeightRadius
      width=0
      color {colorUseRule=UseColor_Specified, specifiedColor=<rgba 0 0 0 0>}
      cornerRadii=<0 0 0 0>
     }
     tintColor {colorUseRule=UseColor_Specified, specifiedColor=<rgba 0.0103 0.0103 0.0103 1>}
     uVRegion {bIsValid=false, max=<0 0>, min=<0 0>}
    }
    desiredSizeScale=<0.5 0.5>
   }
   slot {class=VerticalBoxSlot, properties {padding=<ltrb 0 5 0 5>}}
  }
  HorizontalBox_6 {widgetClassPath=&HorizontalBox, parent=VerticalBox, widget=HorizontalBox_6, slot {class=VerticalBoxSlot}}
  VerticalBox_59 {widgetClassPath=&VerticalBox, parent=HorizontalBox_6, widget=VerticalBox_59, slot {class=HorizontalBoxSlot}}
  r_ik_apply_cbox {
   widgetClassPath=&EditorUtilityCheckBox
   bIsVariable=true
   parent=VerticalBox_59
   widget=r_ik_apply_cbox
   properties {checkedState=Checked}
   slot {class=VerticalBoxSlot}
  }
  RightHandIK {
   widgetClassPath=&TextBlock
   parent=r_ik_apply_cbox
   widget=RightHandIK
   properties {
    text=Apply to R IK Hand
    font {
     bForceMonospaced=false
     bMaterialIsStencil=false
     fontMaterial=null
     fontObject=&Roboto
     letterSpacing=0
     monospacedWidth=1
     size=12
     skewAmount=0
     typefaceFontName=Regular
     outlineSettings {
      bApplyOutlineToDropShadows=false
      bMiteredCorners=false
      bSeparateFillAlpha=false
      outlineMaterial=null
      outlineSize=0
      outlineColor=<rgba 0 0 0 1>
     }
    }
   }
   slot {class=PanelSlot}
  }
  l_ik_apply_cbox {
   widgetClassPath=&EditorUtilityCheckBox
   bIsVariable=true
   parent=VerticalBox_59
   widget=l_ik_apply_cbox
   properties {checkedState=Checked}
   slot {class=VerticalBoxSlot}
  }
  LeftHandIK {
   widgetClassPath=&TextBlock
   parent=l_ik_apply_cbox
   widget=LeftHandIK
   properties {
    text=Apply to L IK Hand
    font {
     bForceMonospaced=false
     bMaterialIsStencil=false
     fontMaterial=null
     fontObject=&Roboto
     letterSpacing=0
     monospacedWidth=1
     size=12
     skewAmount=0
     typefaceFontName=Regular
     outlineSettings {
      bApplyOutlineToDropShadows=false
      bMiteredCorners=false
      bSeparateFillAlpha=false
      outlineMaterial=null
      outlineSize=0
      outlineColor=<rgba 0 0 0 1>
     }
    }
   }
   slot {class=PanelSlot}
  }
 ]
}
