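# Editor utility SOS icon button that lets a tool user ping a developer for help.
# Clicking opens a Yes/No dialog; Yes runs inline Python calling ErrorReporter.request_help, then shows a confirmation.
# The incoming_toolname variable feeds the Python node's toolName input, though the script reads tool_name.
# The button disables itself after the dialog flow, with no re-enable in this graph.
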EditorUtilityWidgetBlueprint /AnimInEdTools/PropBrowser/SOSButton/EUW_SosButton
graphs [
 EventGraph {nodes [
   "On Clicked (SOS_Button)"() {is=title K2Node_ComponentBoundEvent}
   > Sos Button()
   > Set bIsEnabled!(execute, bIsEnabled:bool=false, self:Widget Object Reference=Get SOS_Button() -> (SOS_Button:Button Object Reference)) -> (then, Output_Get:bool)
  ]}
 Sos Button {nodes [
   Sos Button() {is=title K2Node_FunctionEntry}
   > n1 = Show Message Dialog(
    Title="To Text (String)"(InString=Help Message)
    Message="To Text (String)"(InString=Do you want to ping an SOS message to a developer?)
    MessageType=YesNo
    DefaultValue=No
    MessageCategory=Info
   )
   > Branch(Condition="Equal (Enum)"(A:EAppReturnType Enum=@n1, B:EAppReturnType Enum=Yes))
   > Execute Python Script!(
    execute
    PythonScript="import unreal\r\nfrom utilities.error_reporter import ErrorReporter\r\nErrorReporter.request_help(tool_name=tool_name)   \r\nunreal.log(\"Help request sent!\")"
    toolName:String=Get incoming_toolname() -> (incoming_toolname:String)
   ) -> (then, ReturnValue) {inputs [toolName]}
   > Show Message Dialog(
    Title="To Text (String)"(InString=Help Message)
    Message="To Text (String)"(InString="Your help signal has been sent. A developer should reach out to you soon. Hang in there friend, help is on the way!. ")
    MessageType=Ok
    DefaultValue=Ok
    MessageCategory=Success
   )
  ]}
]
variables [incoming_toolname {type=string, default=Unknown Tool Name}]
widgets {
 info {rootWidgetClass=&SizeBox, widgetCount=3}
 widgets [
  SizeBox_0 {
   widgetClassPath=&SizeBox
   widget=SizeBox_0
   properties {bOverride_HeightOverride=true, bOverride_WidthOverride=true, heightOverride=28, widthOverride=28}
  }
  SOS_Button {
   widgetClassPath=&Button
   parent=SizeBox_0
   widget=SOS_Button
   properties {
    toolTipText="SOS, I need help. "
    widgetStyle {
     clickedSlateSound {resourceObject=null}
     disabled {
      bIsDynamicallyLoaded=false
      drawAs=RoundedBox
      imageType=NoImage
      mirroring=NoMirror
      resourceName=None
      resourceObject=null
      tiling=NoTile
      imageSize=<0 0>
      margin=<ltrb 0 0 0 0>
      outlineSettings {
       bUseBrushTransparency=false
       roundingType=FixedRadius
       width=1
       color {colorUseRule=UseColor_ColorTable, specifiedColor=<rgba 0 0 0 0>}
       cornerRadii=<4 4 4 4>
      }
      tintColor {colorUseRule=UseColor_Specified, specifiedColor=<rgba 0.0395 0.0395 0.0395 1>}
      uVRegion {bIsValid=false, max=<0 0>, min=<0 0>}
     }
     disabledForeground {colorUseRule=UseColor_ColorTable, specifiedColor=<rgba 0 0 0 0>}
     hovered {
      bIsDynamicallyLoaded=false
      drawAs=RoundedBox
      imageType=FullColor
      mirroring=NoMirror
      resourceName=None
      resourceObject=null
      tiling=NoTile
      imageSize=<28 28>
      margin=<ltrb 0 0 0 0>
      outlineSettings {
       bUseBrushTransparency=false
       roundingType=FixedRadius
       width=0
       color {colorUseRule=UseColor_ColorTable, specifiedColor=<rgba 0 0 0 0>}
       cornerRadii=<4 4 4 4>
      }
      tintColor {colorUseRule=UseColor_Specified, specifiedColor=<rgba 0.0941 0.0941 0.0941 1>}
      uVRegion {bIsValid=false, max=<0 0>, min=<0 0>}
     }
     hoveredForeground {colorUseRule=UseColor_ColorTable, specifiedColor=<rgba 0 0 0 0>}
     hoveredSlateSound {resourceObject=null}
     normal {
      bIsDynamicallyLoaded=false
      drawAs=NoDrawType
      imageType=FullColor
      mirroring=NoMirror
      resourceName=None
      resourceObject=null
      tiling=NoTile
      imageSize=<28 28>
      margin=<ltrb 0 0 0 0>
      outlineSettings {
       bUseBrushTransparency=false
       roundingType=FixedRadius
       width=0
       color {colorUseRule=UseColor_ColorTable, specifiedColor=<rgba 0 0 0 0>}
       cornerRadii=<4 4 4 4>
      }
      tintColor {colorUseRule=UseColor_Specified, specifiedColor=<rgba 0.0176 0.0176 0.0176 1>}
      uVRegion {bIsValid=false, max=<0 0>, min=<0 0>}
     }
     normalForeground {colorUseRule=UseColor_ColorTable, specifiedColor=<rgba 0 0 0 0>}
     normalPadding=<ltrb 0 0 0 0>
     pressed {
      bIsDynamicallyLoaded=false
      drawAs=RoundedBox
      imageType=FullColor
      mirroring=NoMirror
      resourceName=None
      resourceObject=null
      tiling=NoTile
      imageSize=<28 28>
      margin=<ltrb 0 0 0 0>
      outlineSettings {
       bUseBrushTransparency=false
       roundingType=FixedRadius
       width=0
       color {colorUseRule=UseColor_ColorTable, specifiedColor=<rgba 0 0 0 0>}
       cornerRadii=<4 4 4 4>
      }
      tintColor {colorUseRule=UseColor_Specified, specifiedColor=<rgba 0.0392 0.0392 0.0392 1>}
      uVRegion {bIsValid=false, max=<0 0>, min=<0 0>}
     }
     pressedForeground {colorUseRule=UseColor_ColorTable, specifiedColor=<rgba 0 0 0 0>}
     pressedPadding=<ltrb 0 0 0 0>
     pressedSlateSound {resourceObject=null}
    }
   }
   slot {class=SizeBoxSlot}
  }
  Sos_Button_Icon {
   widgetClassPath=&Image
   bIsVariable=true
   parent=SOS_Button
   widget=Sos_Button_Icon
   properties {
    toolTipText=SOS I need help from a dev
    brush {
     bIsDynamicallyLoaded=false
     drawAs=Image
     imageType=FullColor
     mirroring=NoMirror
     resourceName=None
     resourceObject=&/SOS_Icon
     tiling=NoTile
     imageSize=<28 28>
     margin=<ltrb 0 0 0 0>
     outlineSettings {
      bUseBrushTransparency=false
      roundingType=HalfHeightRadius
      width=0
      color {colorUseRule=UseColor_Specified, specifiedColor=<rgba 0 0 0 0>}
      cornerRadii=<0 0 0 0>
     }
     tintColor {colorUseRule=UseColor_Specified, specifiedColor=<rgba 1 1 1 1>}
     uVRegion {bIsValid=false, max=<0 0>, min=<0 0>}
    }
   }
   slot {class=ButtonSlot, properties {padding=<ltrb 0 0 0 0>}}
  }
 ]
}
