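# Editor utility widget for browsing a project's Animatable Props and importing them into a Level Sequence.
# Asset, plugin, subfolder and search filters drive the prop list.
# Import can create a linked AnimSequence, copying project compression settings, and add a 3D Attach track to a bone or socket on a chosen parent Control Rig, which can be switched later.
# Embedded Python handles sorting, asset creation and error reporting, alongside usage telemetry.
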EditorUtilityWidgetBlueprint /AnimInEdTools/PropBrowser/PropBrowser
graphs [
 Animatable Prop Get Variants {nodes [
   n1 = Animatable Prop Get Variants!() -> (then, Asset:String, Add All option:bool=~) {is=title K2Node_FunctionEntry}
   > Clear(TargetArray=Get Variant Array() -> (Variant Array:Array of Strings))
   > n2 = For Each Loop(Array:Array of Animatable Creature Object References=Get Animatable Props!(
     Project:BP Animation Browser Project Object Reference=Get Current Project() -> (Current Project:BP Animation Browser Project Object Reference)
     __WorldContext:Object Reference
    ) -> (Animatable Creatures:Array of Animatable Creature Object References) {is=title K2Node_CallFunction}) -> (Array Element:Animatable Creature Object Reference)
   n3 = Execute Python Script!(
    execute=@n2.Completed
    PythonScript="sorted_creatures = sorted(creatures)\r\n\r\nif 'Shared' in sorted_creatures:\r\n    sorted_creatures.insert(0, sorted_creatures.pop(sorted_creatures.index('Shared')))\r\n\r\nif len(creatures) > 1 and add_all:\r\n    sorted_creatures.insert(0, 'All')"
    creatures:Array of Strings=Get Variant Array()
    add_all:bool=@n1.Add All option
   ) -> (then, ReturnValue, sorted_creatures:Array of Strings) {inputs [creatures,add_all], outputs [sorted_creatures]}
   > n4 = Set Variant Array(Variant Array:Array of Strings=@n3.sorted_creatures) -> (Output_Get:Array of Strings)
   > Return Node!(execute, CollectedVariants:Array of Strings=@n4)
   Branch(
    execute=@n2.LoopBody
    Condition="Equal Exactly (String)"(A=Get Asset!(self:Animatable Creature Object Reference=@n2) -> (Asset:String), B=@n1)
   )
   > Add Unique(TargetArray=Get Variant Array(), NewItem=Get Variant!(self:Animatable Creature Object Reference=@n2) -> (Variant:String))
  ]}
 AnimatableProps Get Assets {nodes [
   n1 = AnimatableProps Get Assets!() -> (then, AnimatableProps:Array of Animatable Creature Object References) {is=title K2Node_FunctionEntry}
   n2 = Get Local Categories() -> (Local Categories:Array of Strings)
   > Clear(TargetArray=@n2)
   > Add Unique(TargetArray=@n2, NewItem=All)
   > n3 = For Each Loop(Array=@n1)
   n4 = Execute Python Script!(execute=@n3.Completed, PythonScript="sorted_controls=sorted(controls)", controls:Array of Strings=Get Local Categories()) -> (then, ReturnValue, sorted_controls:Array of Strings) {inputs [controls], outputs [sorted_controls]}
   > n5 = Set Local Categories(Local Categories:Array of Strings=@n4.sorted_controls) -> (Output_Get:Array of Strings)
   > Return Node!(execute, Categories:Array of Strings=@n5)
   n6 = Get Asset(self=@n3)
   n7 = Branch(execute=@n3.LoopBody, Condition=IsEmpty(InString=@n6))
   Add Unique(execute=@n7.else, TargetArray=Get Local Categories(), NewItem=@n6)
  ]}
 Attach Track Function {nodes [
   n1 = Attach Track Function!() -> (then, Control Rig:Movie Scene Track Object Reference) {is=title K2Node_FunctionEntry}
   > n2 = Is Valid(InputObject=@n1)
   n3 = Set Attachment Component Name(execute=@n2.Is Not Valid, Attachment Component Name:Name=None) -> (Output_Get:Name=None)
   Branch(execute=[@n8.then, @n3.then], Condition=Get Attach CheckBox() -> (Attach CheckBox:bool))
   > Branch(Condition=IsValid!(InBinding:"Movie Scene Binding Proxy Structure (by ref)"=Get Parent Rig Binding() -> (Parent Rig Binding:Movie Scene Binding Proxy Structure)) -> (ReturnValue))
   > n4 = AddTrack!(
    execute
    InBinding:"Movie Scene Binding Proxy Structure (by ref)"=Get Spawned Binding for Asset() -> (Spawned Binding for Asset:Movie Scene Binding Proxy Structure)
    TrackType:Movie Scene Track Class Reference=/Script/MovieSceneTracks.MovieScene3DAttachTrack
   ) -> (then, ReturnValue:Movie Scene Track Object Reference)
   > n5 = AddSection!(execute, Track:Movie Scene Track Object Reference=@n4) -> (then, ReturnValue:Movie Scene Section Object Reference)
   > SetStartFrameBounded!(execute, Section:Movie Scene Section Object Reference=@n5, bIsBounded:bool=false) -> (then)
   > SetEndFrameBounded!(execute, Section:Movie Scene Section Object Reference=@n5, bIsBounded:bool=false) -> (then)
   > n6 = Cast To MovieScene3DAttachSection!(execute, Object=@n5) -> (then, CastFailed, AsMovie Scene 3DAttach Section:Movie Scene 3DAttach Section Object Reference) {is=title K2Node_DynamicCast}
   > n7 = GetBindingID!(
    execute
    Sequence:Movie Scene Sequence Object Reference=Get Level Sequence() -> (Level Sequence:Level Sequence Object Reference)
    InBinding:"Movie Scene Binding Proxy Structure (by ref)"=Get Parent Rig Binding()
   ) -> (then, ReturnValue:Movie Scene Object Binding ID Structure)
   > SetConstraintBindingID!(
    execute
    self:Movie Scene 3DConstraint Section Object Reference=@n6
    InConstraintBindingID:"Movie Scene Object Binding ID Structure (by ref)"=@n7
   ) -> (then)
   > Set AttachSocketName!(
    execute
    AttachSocketName:Name=Get Attach Bone Name() -> (Attach Bone Name:Name=None)
    self:Movie Scene 3DAttach Section Object Reference=@n6
   ) -> (then, Output_Get:Name=None)
   > Set AttachComponentName!(
    execute
    AttachComponentName:Name=Get Attachment Component Name() -> (Attachment Component Name:Name=None)
    self:Movie Scene 3DAttach Section Object Reference=@n6
   ) -> (then, Output_Get:Name=None)
   > Set bIsEnabled!(
    execute
    bIsEnabled:bool=true
    self:Widget Object Reference=Get SwitchDriveCtrl_Btn() -> (SwitchDriveCtrl_Btn:Editor Utility Button Object Reference)
   ) -> (then, Output_Get:bool)
   n8 = Set Attachment Component Name(execute=@n2.Is Valid, Attachment Component Name=SkeletalMesh) -> (Output_Get=None)
  ]}
 Browse To Project {nodes [
   Browse To Project!() -> (then, Context Item:Editor Utility Widget Object Reference, Slot Widget:Editor Utility Widget Object Reference) {is=title K2Node_FunctionEntry}
   n1 = Get Current Project()
   > SyncBrowserToObjects(AssetPaths=Make Array("[0]":String=GetSoftObjectPath!(Object=@n1) -> (ReturnValue_PathString:String)) -> (Array:Array of Strings))
  ]}
 Callback Populate Season Box From Project Selection {nodes [
   n1 = Callback Populate Season Box From Project Selection!() -> (then, Context Item:Editor Utility Widget Object Reference, Slot Widget:Editor Utility Widget Object Reference) {is=title K2Node_FunctionEntry}
   > n2 = Sequence!(execute) -> (then_0, then_1, then_2:Exec)
   > Set Selected Project Name(Selected Project Name:String) -> (Output_Get:String)
   > n3 = Cast To EUW_ContextMenu_Icon!(execute, Object=Get Slot Widget() -> (Slot Widget:Editor Utility Widget Object Reference)) -> (then, CastFailed, AsEUW Context Menu Icon:EUW Context Menu Icon Object Reference) {is=title K2Node_DynamicCast}
   n4 = Get ButtonText!(self:EUW Context Menu Icon Object Reference=@n3) -> (ButtonText:Text Object Reference)
   > Set Selected Project Name(Selected Project Name="To String (Text)"(InText=Get Text(self:Text Object Reference=@n4)))
   n5 = Cast To EUW_ContextMenu_Icon(execute=@n2.then_1, Object=@n1.Slot Widget)
   n6 = Find Project!(Project Name:String=Get String Data!(self:EUW Context Menu Icon Object Reference=@n5) -> (String Data:String)) -> (AsBP Animation Browser Project:BP Animation Browser Project Object Reference)
   > Is Valid(InputObject=@n6)
   > Set Current Project(Current Project:BP Animation Browser Project Object Reference=@n6) -> (Output_Get:BP Animation Browser Project Object Reference)
   > Save UI State()
   > Telemetry!(
    execute
    self
    feature:String=ui
    action:String=project_selected
    metadata:String=project
    metadata_function:String=Get Selected Project Name() -> (Selected Project Name:String)
   ) -> (then)
   > RefreshTypeCache()
   n7 = AnimatableProps Get Assets!(execute=@n2.then_2, self, AnimatableProps:Array of Animatable Creature Object References=Get Animatable Props(Project=Get Current Project())) -> (then, Categories:Array of Strings) {is=title K2Node_CallFunction}
   > Update Combo Box!(
    execute
    self
    Combo Box:"Searchable Combo Box (String) Object Reference"=Get SearchableAssetComboBox() -> (SearchableAssetComboBox:"Searchable Combo Box (String) Object Reference")
    Items:Array of Strings=@n7
    Default Item:String="Get (a ref)"(Array:Array of Strings=@n7, Dimension 1=0) -> (Output:"String (by ref)")
   ) -> (then)
   > Populating Extra Slot()
  ]}
 Create Linked Anim Sequence {nodes [
   n1 = Create Linked Anim Sequence!() -> (then, Level Sequence:Level Sequence Object Reference, New AnimSequence Path:String, Binding:Movie Scene Binding Proxy Structure) {is=title K2Node_FunctionEntry}
   > n2 = Is Valid(InputObject=@n1)
   Execute Python Script!(
    execute=@n2.Is Not Valid
    PythonScript="from utilities.error_reporter import ErrorReporter\r\n\r\nErrorReporter.send(\r\n    tool_name=\"PropBrowser\",\r\n    feature=\"create_linked_anim_sequence/level_sequence_null\",\r\n    error=Exception(\"Level Sequence was null when Create Linked Anim Sequence was called\"),\r\n    traceback_str=\"No Python traceback — failure detected in Blueprint via IsValid(Level Sequence) check\"\r\n)"
   ) -> (then, ReturnValue)
   > Return Node!(execute, Anim Sequence:Anim Sequence Object Reference)
   n3 = Get Current Time!(execute=@n2.Is Valid, TimeUnit:EMovieSceneTimeUnit Enum=DisplayRate) -> (then, ReturnValue:Movie Scene Sequence Playback Params Structure) {is=title K2Node_CallFunction}
   n4 = Break Movie Scene Sequence Playback Params!(MovieSceneSequencePlaybackParams:"Movie Scene Sequence Playback Params Structure (by ref)"=@n3) -> (
    Frame_FrameNumber_Value:int
    Frame_SubFrame:float=~
    Time:float=~
    MarkedFrame:String
    Timecode:Timecode Structure
    PositionType:EMovieScenePositionType Enum
    UpdateMethod:EUpdatePositionMethod Enum
    bHasJumped:bool=~
   ) {is=title K2Node_BreakStruct}
   > Set Current Time!(
    execute
    PlaybackParams:Movie Scene Sequence Playback Params Structure=Make Movie Scene Sequence Playback Params!(
     Frame_FrameNumber_Value:int=int + int(A=@n4)
     Frame_SubFrame:float=0
     Time:float=0
     MarkedFrame:String
     Timecode:Timecode Structure=(Hours=0,Minutes=0,Seconds=0,Frames=0,Subframe=0,bDropFrameFormat=False)
     PositionType:EMovieScenePositionType Enum=Frame
     UpdateMethod:EUpdatePositionMethod Enum=Play
     bHasJumped:bool=False
    ) -> (MovieSceneSequencePlaybackParams:Movie Scene Sequence Playback Params Structure) {is=title K2Node_MakeStruct}
    TimeUnit:EMovieSceneTimeUnit Enum=DisplayRate
   ) -> (then) {is=title K2Node_CallFunction}
   > Set Current Time(PlaybackParams=@n3, TimeUnit=DisplayRate)
   > ForceUpdate!(execute) -> (then)
   > n5 = GetBindingID(Sequence=Get Level Sequence(), InBinding=Get Binding() -> (Binding:Movie Scene Binding Proxy Structure))
   > n6 = For Each Loop with Break!(
    Exec:Exec
    Array:Array of Object References=GetBoundObjects!(ObjectBinding:Movie Scene Object Binding ID Structure=@n5) -> (ReturnValue:Array of Object References)
    Break:Exec=@n21.then
   ) -> (LoopBody:Exec, Array Element:Object Reference, Array Index:int, Completed:Exec)
   n7 = Is Valid(exec=@n6.Completed, InputObject=Get Skeleton() -> (Skeleton:Skeleton Object Reference))
   Execute Python Script(
    execute=@n7.Is Not Valid
    PythonScript="from utilities.error_reporter import ErrorReporter\r\n\r\nErrorReporter.send(\r\n    tool_name=\"PropBrowser\",\r\n    feature=\"create_linked_anim_sequence/skeleton_null\",\r\n    error=Exception(\"Skeleton was null — bound actor may have no SkeletalMeshComponent or GetBoundObjects returned empty\"),\r\n    traceback_str=\"No Python traceback — failure detected in Blueprint via IsValid(Skeleton) check\"\r\n)"
   )
   > Return Node()
   n8 = Get Binding()
   n9 = Get Level Sequence()
   n10 = Execute Python Script!(
    execute=@n7.Is Valid
    PythonScript="import unreal\r\nimport traceback\r\nfrom utilities.error_reporter import ErrorReporter\r\n\r\nlog_path, log_start = ErrorReporter.snapshot_log()\r\n\r\ntry:\r\n    editor_subsystem = unreal.get_editor_subsystem(unreal.UnrealEditorSubsystem)\r\n    world = editor_subsystem.get_editor_world()\r\n\r\n    anim_seq_export_options = unreal.AnimSeqExportOption()\r\n    anim_seq_export_options.export_transforms = True\r\n    anim_seq_export_options.export_morph_targets = True\r\n\r\n    asset_tools = unreal.AssetToolsHelpers.get_asset_tools()\r\n    anim_sequence_factory = unreal.AnimSequenceFactory()\r\n    anim_sequence_factory.target_skeleton = skeleton\r\n\r\n    anim_sequence = unreal.AssetTools.create_asset(asset_tools, asset_name=asset_name, package_path=package_path, asset_class=unreal.AnimSequence, factory=anim_sequence_factory)\r\n    if anim_sequence is None:\r\n        raise Exception(\"create_asset returned None — asset may already exist or path is invalid\")\r\n\r\n    ErrorReporter.check_log_for_errors(log_path, log_start, tool_name=\"PropBrowser\", feature=\"create_linked_anim_sequence\")\r\n\r\nexcept Exception as e:\r\n    ErrorReporter.send_with_log_deferred(\r\n        tool_name=\"PropBrowser\",\r\n        feature=\"create_linked_anim_sequence\",\r\n        error=e,\r\n        traceback_str=traceback.format_exc(),\r\n        log_path=log_path,\r\n        log_start=log_start\r\n    )\r\n    raise"
    package_path:String=GetPath!(InPath:String=@n1.New AnimSequence Path) -> (ReturnValue:String)
    asset_name:String=GetBaseFilename(InPath=@n1.New AnimSequence Path, bRemovePath=true)
    level_sequence:Level Sequence Object Reference=@n9
    binding:Movie Scene Binding Proxy Structure=@n8
    skeleton:Skeleton Object Reference=Get Skeleton()
   ) -> (then, ReturnValue, anim_seq_export_options:Anim Seq Export Option Object Reference) {inputs [package_path, asset_name, level_sequence, binding, skeleton], outputs [anim_seq_export_options]}
   > n11 = LoadAsset(AssetPath=Get New AnimSequence Path() -> (New AnimSequence Path:String))
   > n12 = Cast To AnimSequence!(execute, Object=@n11) -> (then, CastFailed, AsAnim Sequence:Anim Sequence Object Reference)
   > n13 = Sequence!(execute) -> (then_0, then_1)
   n14 = Get Compression Settings!(self:BP Animation Browser Project Object Reference=Get Current Project()) -> (Compression Settings:Anim Sequence Soft Object Reference)
   > Branch(Condition=IsValidSoftObjectReference(SoftObjectReference=@n14))
   > n15 = LoadAsset_Blocking(Asset=@n14)
   > n16 = Cast To AnimSequence(Object=@n15)
   > SetBoneCompressionSettings!(
    execute
    AnimationSequence:Anim Sequence Object Reference=@n12
    CompressionSettings:Anim Bone Compression Settings Object Reference=GetBoneCompressionSettings!(AnimationSequence:Anim Sequence Object Reference=@n16) -> (CompressionSettings:Anim Bone Compression Settings Object Reference)
   ) -> (then)
   > SetCurveCompressionSettings!(
    execute
    AnimationSequence:Anim Sequence Object Reference=@n12
    CompressionSettings:Anim Curve Compression Settings Object Reference=GetCurveCompressionSettings!(AnimationSequence:Anim Sequence Object Reference=@n16) -> (CompressionSettings:Anim Curve Compression Settings Object Reference)
   ) -> (then)
   > SetAnimationInterpolationType!(
    execute
    AnimationSequence:Anim Sequence Object Reference=@n12
    InterpolationType:EAnimInterpolationType Enum=GetAnimationInterpolationType!(AnimationSequence:Anim Sequence Object Reference=@n16) -> (InterpolationType:EAnimInterpolationType Enum=Linear)
   ) -> (then)
   n17 = LinkAnimSequence!(
    execute=@n13.then_1
    Sequence:Level Sequence Object Reference=@n9
    AnimSequence:Anim Sequence Object Reference=@n12
    ExportOptions:Anim Seq Export Option Object Reference=@n10.anim_seq_export_options
    Binding:"Movie Scene Binding Proxy Structure (by ref)"=@n8
   ) -> (then, ReturnValue:bool)
   > n18 = Branch(Condition=@n17)
   > Return Node(Anim Sequence=@n12)
   PrintString(execute=@n18.else, InString=PropBrowser/CreateLinkedAnimSequence: Link Anim Sequence Failed.)
   > Return Node()
   n19 = Cast To Actor!(execute=@n6.LoopBody, Object=@n6) -> (then, CastFailed, AsActor:Actor Object Reference) {is=title K2Node_DynamicCast}
   n20 = GetComponentByClass!(self:Actor Object Reference=@n19, ComponentClass:Actor Component Class Reference=/Script/Engine.SkeletalMeshComponent) -> (ReturnValue:Skeletal Mesh Component Object Reference)
   > n21 = Set Skeleton(Skeleton:Skeleton Object Reference=Get Skeleton!(self:Skeletal Mesh Object Reference=Get SkeletalMesh!(self:Skinned Mesh Component Object Reference=@n20) -> (SkeletalMesh:Skeletal Mesh Object Reference)) -> (Skeleton)) -> (Output_Get:Skeleton Object Reference)
  ]}
 Disable Import Linked Prop Button {nodes [
   n1 = Disable Import Linked Prop Button!() -> (then, Animatable Prop:Animatable Prop Object Reference) {is=title K2Node_FunctionEntry}
   n2 = Get Skeletal Mesh Asset!(self:Animatable Prop Object Reference=@n1) -> (Skeletal Mesh Asset:Skeletal Mesh Soft Object Reference)
   n3 = Get Sequencer Actor!(self:Animatable Creature Object Reference=@n1) -> (Sequencer Actor:Object Soft Class Reference)
   > n4 = Branch(Condition=OR Boolean(
     A=IsValidSoftObjectReference(SoftObjectReference=@n2)
     B=IsValidSoftClassReference!(SoftClassReference:"Object Soft Class Reference (by ref)"=@n3) -> (ReturnValue:bool)
    ))
   > Set bIsEnabled(bIsEnabled=true, self=Get ImportPropCreateLinkBtn() -> (ImportPropCreateLinkBtn:Editor Utility Button Object Reference))
   n5 = Branch(
    execute=@n4.else
    Condition=IsValidSoftObjectReference(SoftObjectReference=Get Static Mesh Asset!(self:Animatable Prop Object Reference=@n1) -> (Static Mesh Asset:Static Mesh Soft Object Reference))
   )
   > Set bIsEnabled(bIsEnabled=false, self=Get ImportPropCreateLinkBtn())
   Set bIsEnabled(execute=@n5.else, bIsEnabled=true, self=Get ImportPropCreateLinkBtn())
  ]}
 Display Modal Dialog {nodes [
   n1 = Display Modal Dialog!() -> (then, Class:User Widget Class Reference) {is=title K2Node_FunctionEntry}
   > n2 = Create Widget!(
    execute:Exec
    WorldContextObject:Object Reference
    Class:User Widget Class Reference=/AnimInEdTools/AnimationBrowser/EUW_ModalDialog.EUW_ModalDialog_C
    OwningPlayer:Player Controller Object Reference
   ) -> (then:Exec, ReturnValue:EUW Modal Dialog Object Reference)
   > n3 = Create Widget(Class=@n1) -> (ReturnValue:User Widget Object Reference)
   > AddChild(self=Get NamedSlot!(self:EUW Modal Dialog Object Reference=@n2) -> (NamedSlot:Named Slot Object Reference), Content=@n3)
   > AddChild(self=Get ModalLayer() -> (ModalLayer:Named Slot Object Reference), Content=@n2)
   > Return Node!(execute, Modal Widget:EUW Modal Dialog Object Reference=@n2, New Widget:User Widget Object Reference=@n3)
  ]}
 EventGraph {nodes [
   Telemetry(feature=prop_control, action=active)
   n1 = Get ParentControlRigComboBox() -> (ParentControlRigComboBox:Editor Utility Combo Box String Object Reference)
   > Set CurrentIndex(CurrentIndex:int=GetSelectedIndex!(self:"Combo Box (String) Object Reference"=@n1) -> (ReturnValue:int=0)) -> (Output_Get:int=0)
   > Refresh()
   > SetSelectedIndex!(execute, self:"Combo Box (String) Object Reference"=@n1, Index:int=Get CurrentIndex() -> (CurrentIndex:int=0)) -> (then)
   n2 = "On Selection Changed (ParentControlRigComboBox)"!() -> (OutputDelegate, then, SelectedItem:String, SelectionType:ESelectInfo Enum) {is=title K2Node_ComponentBoundEvent}
   > Set Selected Parent Asset(Selected Parent Asset:String=@n2.SelectedItem) -> (Output_Get:String)
   > n3 = GetControlRigs!(execute, LevelSequence:Level Sequence Object Reference=Get Level Sequence()) -> (then, ReturnValue:Array of Control Rig Sequencer Binding Proxy Structures)
   > n4 = For Each Loop!(Exec, Array:Array of Control Rig Sequencer Binding Proxy Structures=@n3) -> (
    LoopBody
    Array Element_Proxy:Movie Scene Binding Proxy Structure
    Array Element_ControlRig:Control Rig Object Reference
    Array Element_Track:Movie Scene Control Rig Parameter Track Object Reference
    Array Index
    Completed
   )
   Set Attach Bone selection!(
    execute=@n4.Completed
    self
    Default Attach Bone:String=Get Default Attach Bone!(self:Animatable Prop Object Reference=Get AnimatableProp() -> (AnimatableProp:Animatable Prop Object Reference)) -> (Default Attach Bone:String)
   ) -> (then) {is=title K2Node_CallFunction}
   n5 = Sequence(execute=@n4.LoopBody)
   n6 = GetSelectedIndex(self=Get ParentControlRigComboBox()) -> (ReturnValue=0)
   > Branch(Condition="Equal (Integer)"(A=@n4.Array Index, B=@n6))
   > Set Parent Rig Binding(Parent Rig Binding:Movie Scene Binding Proxy Structure=@n4) -> (Output_Get:Movie Scene Binding Proxy Structure)
   n7 = Get ParentRigs() -> (ParentRigs:Array of Control Rig Object References)
   Branch(
    execute=@n5.then_1
    Condition="Equal (Object)"(
     A="Get (a copy)"(Array:Array of Control Rig Object References=@n7, Dimension 1=@n6) -> (Output:Control Rig Object Reference)
     B=@n4.Array Element_ControlRig
    )
   )
   > Set Selected Parent Rig(Selected Parent Rig:Control Rig Object Reference=@n4.Array Element_ControlRig) -> (Output_Get:Control Rig Object Reference)
   > Prop Control Setup()
   > n8 = Set Control Rig Controls for Combo Box!(
    execute
    self
    Selected Control Rig:Control Rig Object Reference=Get Selected Parent Rig() -> (Selected Parent Rig:Control Rig Object Reference)
    Combo Box:"Searchable Combo Box (String) Object Reference"=Get AttachmentPointSearchableComboBox() -> (AttachmentPointSearchableComboBox:"Searchable Combo Box (String) Object Reference")
    Socket attachment:ECheckBoxState Enum=Get CheckedState!(self:Check Box Object Reference=Get attach_socket_chkbx() -> (attach_socket_chkbx:Editor Utility Check Box Object Reference)) -> (CheckedState:ECheckBoxState Enum=Unchecked)
   ) -> (then, Attach Point Array:Array of Strings) {is=title K2Node_CallFunction}
   > Set the Attach Point selection!(execute, self, Array:Array of Strings=@n8) -> (then) {is=title K2Node_CallFunction}
   "On Clicked (ImportPropCreateLinkBtn)"() {is=title K2Node_ComponentBoundEvent}
   > Telemetry(feature=import, action=with_anim_sequence, metadata=prop, metadata_function=GetDisplayName(Object=Get AnimatableProp()))
   > n9 = Validate Parent and AttachBone!(execute, self) -> (then, "Validation Succeeded ":bool)
   > Branch(Condition=@n9)
   > n10 = Sequence()
   > n11 = Create Widget(Class=/AnimInEdTools/AnimationBrowser/EUW_Animation_Dialog.EUW_Animation_Dialog_C) -> (ReturnValue:EUW Animation Dialog Object Reference)
   > Bind Event to Generate Animation Paths(self:EUW Animation Dialog Object Reference=@n11, Delegate=Create Event()) {is=title K2Node_AddDelegate}
   > n12 = Set Animation Dialog Widget(Animation Dialog Widget:EUW Animation Dialog Object Reference=@n11) -> (Output_Get:EUW Animation Dialog Object Reference)
   > AddChild(self=Get ModalLayer(), Content=@n12)
   > Set Project!(
    execute
    Project:BP Animation Browser Project Object Reference=Get Current Project()
    self:EUW Animation Dialog Object Reference=Get Animation Dialog Widget() -> (Animation Dialog Widget:EUW Animation Dialog Object Reference)
   ) -> (then, Output_Get:BP Animation Browser Project Object Reference)
   n13 = Get Animation Dialog Widget()
   > Set Options from Animatable Creature!(
    execute
    self:EUW Animation Dialog Object Reference=@n13
    Selected Creature:Animatable Creature Object Reference=Get AnimatableProp()
    All Creatures:Array of Animatable Creature Object References=Get Animatable Props(Project=Get Current Project())
    Update Assets:bool=true
    Update Plugins:bool=true
    Set Plugin:bool=true
   ) -> (then) {is=title K2Node_CallFunction}
   > Validate UI Input!(execute, self:EUW Animation Dialog Object Reference=@n13) -> (then, Is Valid:bool)
   Show!(
    execute=@n10.then_1
    self:EUW Context Menu Object Reference=Get EUW_ContextMenu() -> (EUW_ContextMenu:EUW Context Menu Object Reference)
    Show:bool=true
    Mouse Position:Vector 2D Structure
    Name:String
   ) -> (then)
   n14 = "On Check State Changed (AddandConstraintCkBx)"!() -> (OutputDelegate, then, bIsChecked:bool) {is=title K2Node_ComponentBoundEvent}
   > n15 = Set Attach CheckBox(Attach CheckBox:bool=@n14.bIsChecked) -> (Output_Get:bool)
   > n16 = Branch(Condition=@n15)
   > Set CheckedState!(
    execute
    CheckedState:ECheckBoxState Enum=Unchecked
    self:Check Box Object Reference=Get JustImportCkBx() -> (JustImportCkBx:Editor Utility Check Box Object Reference)
   ) -> (then, Output_Get:ECheckBoxState Enum=Unchecked)
   > Set bIsEnabled(bIsEnabled=true, self=Get AttachmentPointSearchableComboBox())
   > Set bIsEnabled(bIsEnabled=true, self=Get ParentControlRigComboBox())
   n17 = Set CheckedState(execute=@n16.else, CheckedState=Checked, self=Get JustImportCkBx()) -> (Output_Get=Unchecked)
   Set bIsEnabled(execute=[@n17.then, @n20.then], bIsEnabled=false, self=Get AttachmentPointSearchableComboBox())
   > Set bIsEnabled(bIsEnabled=false, self=Get ParentControlRigComboBox())
   n18 = "On Check State Changed (JustImportCkBx)"!() -> (OutputDelegate, then, bIsChecked:bool) {is=title K2Node_ComponentBoundEvent}
   > n19 = Branch(Condition=@n18.bIsChecked)
   > Set CheckedState(CheckedState=Unchecked, self=Get AddandConstraintCkBx() -> (AddandConstraintCkBx:Editor Utility Check Box Object Reference)) -> (Output_Get=Unchecked)
   > n20 = Set Attach CheckBox(Attach CheckBox=false)
   Set CheckedState(execute=@n19.else, CheckedState=Checked, self=Get AddandConstraintCkBx()) -> (Output_Get=Unchecked)
   > Set Attach CheckBox(Attach CheckBox=true)
   > Set bIsEnabled(bIsEnabled=true, self=Get AttachmentPointSearchableComboBox())
   > Set bIsEnabled(bIsEnabled=true, self=Get ParentControlRigComboBox())
   "On Clicked (ClearSearchBox)"() {is=title K2Node_ComponentBoundEvent}
   > Set Text(self:Text Box Object Reference=Get SearchBar() -> (SearchBar:Editor Utility Editable Text Box Object Reference))
   > n21 = Set SearchBarFilter(SearchBarFilter:String) -> (Output_Get:String)
   RetriggerableDelay!(execute=[@n40.then, @n63.then, @n21.then], WorldContextObject:Object Reference, Duration=0.2) -> (then)
   > Update List View!(
    execute
    self
    Selected Plugin:String=GetSelectedOption!(self:"Searchable Combo Box (String) Object Reference"=Get SearchablePluginComboBox() -> (SearchablePluginComboBox:"Searchable Combo Box (String) Object Reference")) -> (ReturnValue:String)
    Filter:String=Get SearchBarFilter() -> (SearchBarFilter:String)
    Subfolder:String=GetSelectedOption(self=Get SearchableSubfolderComboBox() -> (SearchableSubfolderComboBox:"Searchable Combo Box (String) Object Reference"))
    Selected Asset:String=GetSelectedOption(self=Get SearchableAssetComboBox())
   ) -> (then)
   "On Clicked (SettingsButton)"() {is=title K2Node_ComponentBoundEvent}
   n22 = Get EUW_ContextMenu()
   n23 = Get Name!(self:EUW Context Menu Object Reference=@n22) -> (Name:String)
   > n24 = Branch(Condition=AND Boolean(A=IsVisible(self=@n22), B="Equal Exactly (String)"(A=@n23, B=Settings)))
   > Show(self=Get EUW_ContextMenu(), Show=false)
   Clear Context Menu(execute=@n24.else, self:EUW Context Menu Object Reference=Get EUW_ContextMenu())
   > n25 = Sequence!(execute) -> (then_0, then_1, then_2:Exec)
   > Add Icon Item!(
    execute
    self:EUW Context Menu Object Reference=Get EUW_ContextMenu()
    Event:"Delegate (by ref)"=Create Event(self=Self-Reference())
    Category:String=Developer
    Name:String=Browse to Project
    Font Size:double=10
    Icon:Texture 2D Object Reference
   ) -> (then, Context Menu Icon:EUW Context Menu Icon Object Reference)
   n26 = For Each Loop!(
    Exec=@n25.then_1
    Array:Array of BP Animation Browser Project Object References=Get Projects!() -> (Projects:Array of BP Animation Browser Project Object References) {is=title K2Node_CallFunction}
   ) -> (LoopBody, Array Element:BP Animation Browser Project Object Reference, Array Index, Completed)
   n27 = Get Animatable Props(Project=@n26)
   > Branch(Condition="integer > integer"(A=Length(TargetArray:Array of Animatable Creature Object References=@n27)))
   n28 = Get Name(self:BP Animation Browser Project Object Reference=@n26)
   > n29 = Add Icon Item(
    self=Get EUW_ContextMenu()
    Event=Create Event(self=Self-Reference())
    Category=Projects
    Name=@n28
    Font Size=10
    Icon=/AnimInEdTools/ContextMenu/Icon/Icon_Edit
   )
   > Set String Data!(execute, String Data:String=@n28, self:EUW Context Menu Icon Object Reference=@n29) -> (then, Output_Get:String)
   n30 = GetCachedGeometry!(self:Widget Object Reference=Get SettingsButton() -> (SettingsButton:Button Object Reference)) -> (ReturnValue:"Geometry Structure (by ref)")
   n31 = GetLocalSize!(Geometry:"Geometry Structure (by ref)"=@n30) -> (ReturnValue_X:double, ReturnValue_Y:double)
   Show(
    execute=@n25.then_2
    self=Get EUW_ContextMenu()
    Show=true
    Mouse Position="Local to Absolute (Coordinates)"!(Geometry:"Geometry Structure (by ref)"=@n30, LocalCoordinate_X:double=@n31, LocalCoordinate_Y:double=@n31.ReturnValue_Y) -> (ReturnValue:Vector 2D Structure)
    Name=Settings
   )
   n32 = "On Item Clicked (ListView)"!() -> (OutputDelegate, then, Item:Object Reference) {is=title K2Node_ComponentBoundEvent}
   > n33 = Sequence!(execute) -> (then_0, then_1)
   > Show(self=Get EUW_ContextMenu(), Show=false)
   n34 = For Each Loop(Exec=@n33.then_1, Array:Array of Animatable Creature Object References=Get Animatable Props(Project=Get Current Project())) -> (Array Element:Animatable Creature Object Reference)
   Is Valid(exec=@n34.Completed, InputObject=Get AnimatableProp())
   > Set Attach Bone selection(Default Attach Bone=Get Default Attach Bone(self=Get AnimatableProp()))
   n35 = Cast To Object_PropBrowser!(execute=@n34.LoopBody, Object=@n32.Item) -> (then, CastFailed, AsObject Prop Browser:Object Prop Browser Object Reference) {is=title K2Node_DynamicCast}
   n36 = Get Data!(self:Object Prop Browser Object Reference=@n35) -> (Data:Struct Prop Brower Item Structure)
   > Branch(Condition="Equal (Object)"(
     A=Break Struct Prop Brower Item!(Struct_PropBrower_Item:"Struct Prop Brower Item Structure (by ref)"=@n36) -> (
      AP_53_#1:Animatable Prop Object Reference
      Season_AssetType_57_#2:String
      AP_Type_62_#3:String
     ) {is=title K2Node_BreakStruct}
     B=@n34
    ))
   > n37 = Cast To AnimatableProp!(execute, Object=@n34) -> (then, CastFailed, AsAnimatable Prop:Animatable Prop Object Reference) {is=title K2Node_DynamicCast}
   > n38 = Set AnimatableProp(AnimatableProp:Animatable Prop Object Reference=@n37) -> (Output_Get:Animatable Prop Object Reference)
   > Disable Import Linked Prop Button!(execute, self, Animatable Prop:Animatable Prop Object Reference=@n38) -> (then) {is=title K2Node_CallFunction}
   n39 = "On Text Changed (SearchBar)"!() -> (OutputDelegate, then, Text:"Text (by ref)") {is=title K2Node_ComponentBoundEvent}
   > n40 = Set SearchBarFilter(SearchBarFilter="To String (Text)"(InText=@n39.Text))
   n41 = "On Selection Changed (SearchablePluginComboBox)"!() -> (OutputDelegate, then, SelectedItem:String, SelectionType:ESelectInfo Enum) {is=title K2Node_ComponentBoundEvent}
   > Set Selected Plugin(Selected Plugin:String=@n41.SelectedItem) -> (Output_Get:String)
   > n42 = For Each Loop(Array=Get Animatable Props(Project=Get Current Project()))
   n43 = Get Subfolders!() -> (Subfolders:Array of Strings) {is=title K2Node_CallFunction}
   n44 = Branch(execute=@n42.Completed, Condition="integer <= integer"(A=Length(TargetArray:Array of Strings=@n43), B=2))
   > Set Visibility(Visibility=Collapsed, self=Get SearchableSubfolderComboBox())
   > n45 = Set Visibility(Visibility=Collapsed, self=Get Subfolder() -> (Subfolder:Image Object Reference))
   Update Combo Box(execute=[@n45.then, @n46.then], Combo Box=Get SearchableSubfolderComboBox(), Items=Get Subfolders(), Default Item=All)
   > Update List View(
    Selected Plugin=Get Selected Plugin() -> (Selected Plugin:String)
    Filter=Get SearchBarFilter()
    Subfolder=All
    Selected Asset=Get Selected Asset() -> (Selected Asset:String)
   )
   > Save UI State()
   Set Visibility(execute=@n44.else, Visibility=Visible, self=Get SearchableSubfolderComboBox())
   > n46 = Set Visibility(Visibility=Visible, self=Get Subfolder())
   Branch(execute=@n42.LoopBody, Condition="Equal Exactly (String)"(A=Get Variant(self=@n42), B=Get Selected Plugin()))
   > n47 = Cast To AnimatableProp(Object=@n42)
   > Set AnimatableProp(AnimatableProp=@n47)
   "On Clicked (refreshUIButton)"() {is=title K2Node_ComponentBoundEvent}
   > Refresh()
   > Telemetry(feature=ui, action=refresh)
   "On Opening (AttachmentPointSearchableComboBox)"() {is=title K2Node_ComponentBoundEvent}
   > n48 = Set Control Rig Controls for Combo Box(
    Selected Control Rig=Get Selected Parent Rig()
    Combo Box=Get AttachmentPointSearchableComboBox()
    Socket attachment=Get CheckedState(self=Get attach_socket_chkbx()) -> (CheckedState=Unchecked)
   )
   > Set the Attach Point selection(Array=@n48)
   n49 = "On Selection Changed (SearchableAssetComboBox)"!() -> (OutputDelegate, then, SelectedItem:String, SelectionType:ESelectInfo Enum) {is=title K2Node_ComponentBoundEvent}
   > Set Selected Asset(Selected Asset:String=@n49.SelectedItem) -> (Output_Get:String)
   > Populate Asset ComboBox()
   n50 = "On Selection Changed (SearchableSubfolderComboBox)"!() -> (OutputDelegate, then, SelectedItem:String, SelectionType:ESelectInfo Enum) {is=title K2Node_ComponentBoundEvent}
   > Set Selected Subfolder(Selected Subfolder:String=@n50.SelectedItem) -> (Output_Get:String)
   > Update List View(
    Selected Plugin=Get Selected Plugin()
    Filter=Get SearchBarFilter()
    Subfolder=Get Selected Subfolder() -> (Selected Subfolder:String)
    Selected Asset=Get Selected Asset()
   )
   > Save UI State()
   n51 = "On Check State Changed (attach_socket_chkbx)"!() -> (OutputDelegate, then, bIsChecked:bool) {is=title K2Node_ComponentBoundEvent}
   > n52 = Branch(Condition=@n51.bIsChecked)
   n53 = Get attach_bone_chkbx() -> (attach_bone_chkbx:Editor Utility Check Box Object Reference)
   > n54 = Set CheckedState(CheckedState=Unchecked, self=@n53) -> (Output_Get=Unchecked)
   n55 = Set Control Rig Controls for Combo Box(
    execute=[@n69.then, @n56.then, @n54.then, @n68.then]
    Selected Control Rig=Get Selected Parent Rig()
    Combo Box=Get AttachmentPointSearchableComboBox()
    Socket attachment=Get CheckedState(self=Get attach_socket_chkbx()) -> (CheckedState=Unchecked)
   )
   > Set the Attach Point selection(Array=@n55)
   n56 = Set CheckedState(execute=@n52.else, CheckedState=Checked, self=@n53) -> (Output_Get=Unchecked)
   "On Clicked (SwitchDriveCtrl_Btn)"() {is=title K2Node_ComponentBoundEvent}
   > Telemetry(feature=prop_control, action=updated_attachment_point)
   n57 = Get Spawned Binding for Asset()
   > n58 = GetTracks!(execute, InBinding:"Movie Scene Binding Proxy Structure (by ref)"=@n57) -> (then, ReturnValue:Array of Movie Scene Track Object References)
   > n59 = For Each Loop(Array:Array of Movie Scene Track Object References=@n58) -> (Array Element:Movie Scene Track Object Reference)
   > Branch(Condition=Contains(SearchIn=GetDisplayName(Object=@n59), Substring=AttachTrack))
   > RemoveTrack!(execute, InBinding:"Movie Scene Binding Proxy Structure (by ref)"=@n57, TrackToRemove:Movie Scene Track Object Reference=@n59) -> (then)
   > Attach Track Function!(
    execute
    self
    Control Rig:Movie Scene Track Object Reference=Get Control Rig Track() -> (Control Rig Track:Movie Scene Track Object Reference)
   ) -> (then) {is=title K2Node_CallFunction}
   "On Clicked (OnlyImportPropBtn)"() {is=title K2Node_ComponentBoundEvent}
   > Telemetry(feature=import, action=prop_only, metadata=prop, metadata_function=GetDisplayName(Object=Get AnimatableProp()))
   > n60 = Import and Create Anim Sequence!(
    execute
    self
    AnimSequence Path:String
    LevelSequence Path:String
    Animatable Creature:Animatable Creature Object Reference=Get AnimatableProp()
   ) -> (then, Control Rig:Movie Scene Track Object Reference)
   > n61 = Set Control Rig Track(Control Rig Track:Movie Scene Track Object Reference=@n60) -> (Output_Get:Movie Scene Track Object Reference)
   > Attach Track Function(Control Rig=@n61)
   n62 = "On Text Committed (SearchBar)"!() -> (OutputDelegate, then, Text:"Text (by ref)", CommitMethod:ETextCommit Enum=Default) {is=title K2Node_ComponentBoundEvent}
   > n63 = Set SearchBarFilter(SearchBarFilter="To String (Text)"(InText=@n62.Text))
   n64 = "On Selection Changed (AttachmentPointSearchableComboBox)"!() -> (OutputDelegate, then, SelectedItem:String, SelectionType:ESelectInfo Enum) {is=title K2Node_ComponentBoundEvent}
   > Is Valid(InputObject=Get Selected Parent Rig())
   > Set Attach Bone Name(Attach Bone Name:Name=String To Name(InString=@n64.SelectedItem)) -> (Output_Get:Name=None)
   "On Clicked (Button_?)"() {is=title K2Node_ComponentBoundEvent}
   > Telemetry(feature=ui, action=help_button_opened)
   > Execute Python Script!(
    execute
    PythonScript="import webbrowser\r\n\r\nwebbrowser.open(\"https://confluence-epicgames.atlassian.net/wiki/spaces/FORT/pages/75894154/What+is+the+Prop+Browser+tool+in+UE\")"
   ) -> (then, ReturnValue)
   n65 = "On Check State Changed (attach_bone_chkbx)"!() -> (OutputDelegate, then, bIsChecked:bool) {is=title K2Node_ComponentBoundEvent}
   > n66 = Branch(Condition=@n65.bIsChecked)
   n67 = Get attach_socket_chkbx()
   > n68 = Set CheckedState(CheckedState=Unchecked, self=@n67) -> (Output_Get=Unchecked)
   n69 = Set CheckedState(execute=@n66.else, CheckedState=Checked, self=@n67) -> (Output_Get=Unchecked)
   n70 = "On Item Right Clicked (List View)"!() -> (OutputDelegate, then, Item:Object Prop Browser Object Reference, Mouse Event:Pointer Event Structure) {is=title K2Node_CustomEvent}
   > Set Selected Item!(
    execute
    self:List View Object Reference=Get ListView() -> (ListView:Editor Utility List View Object Reference)
    Item:Object Reference=@n70.Item
   ) -> (then) {is=title K2Node_CallFunction}
   > Clear Context Menu(self=Get EUW_ContextMenu())
   > n71 = Sequence()
   > Add Icon Item(
    self=Get EUW_ContextMenu()
    Event=Create Event()
    Category=Developer
    Name=Browse to AP file
    Font Size=10
    Icon=/AnimInEdTools/AnimationBrowser/Icons/Icon_Edit
   )
   Show(
    execute=@n71.then_1
    self=Get EUW_ContextMenu()
    Show=true
    Mouse Position=Get Screen Space Position!(Input:"Pointer Event Structure (by ref)"=@n70.Mouse Event) -> (ReturnValue:Vector 2D Structure) {is=title K2Node_CallFunction}
   )
   n72 = Import Prop With Linked AnimSequence!() -> (OutputDelegate, then, AnimSequence Path:String, LevelSequence Path:String, Animatable Creature:Animatable Creature Object Reference) {is=title K2Node_CustomEvent}
   n73 = GetDisplayName(Object=Get Level Sequence())
   > n74 = Import and Create Anim Sequence(
    AnimSequence Path=Select(
     Option 0:String=@n72.AnimSequence Path
     Option 1:String=Append(A=@n72.AnimSequence Path, B=_FP)
     Index=EndsWith(SourceString=@n73, InSuffix=_FP, SearchCase=CaseSensitive)
    ) -> (ReturnValue:String)
    LevelSequence Path=@n72.LevelSequence Path
    Animatable Creature=@n72.Animatable Creature
   )
   > n75 = Set Control Rig Track(Control Rig Track=@n74)
   > Attach Track Function(Control Rig=@n75)
   Refresh() {is=title K2Node_CustomEvent}
   > n76 = Branch(Condition=IsLoadingAssets!(self:Asset Registry Interface=GetAssetRegistry()) -> (ReturnValue:bool))
   Set Animatable Creature Array(
    execute=@n76.else
    Animatable Creature Array:Array of Animatable Creature Object References=Get Animatable Props(Project=Get Current Project())
   ) -> (Output_Get:Array of Animatable Creature Object References)
   > RefreshTypeCache()
   > RefreshMaster()
   > Populate Control Rig Combo Box!(
    execute
    self
    Level Sequence:Level Sequence Object Reference=Get Level Sequence()
    Combo Box:Editor Utility Combo Box String Object Reference=Get ParentControlRigComboBox()
   ) -> (then)
   n77 = SetProject!() -> (OutputDelegate, then, Project:BP Animation Browser Project Object Reference) {is=title K2Node_CustomEvent}
   > Set Project!(execute, self, Project:BP Animation Browser Project Object Reference=@n77.Project) -> (then) {is=title K2Node_CallFunction}
   > ClearChildren(self=Get ModalLayer())
   Prop Control Setup() {is=title K2Node_CustomEvent}
   > Set Context!(
    execute
    self:EUW Prop Control Setup Object Reference=Get PropControlSetup_Widget() -> (PropControlSetup_Widget:EUW Prop Control Setup Object Reference)
    In Level Sequence:Level Sequence Object Reference=Get Level Sequence()
    In Sequencer Proxy:Control Rig Sequencer Binding Proxy Structure=Get control rig sequencer proxy() -> (control rig sequencer proxy:Control Rig Sequencer Binding Proxy Structure)
    In Parent Rig:Control Rig Object Reference=Get Selected Parent Rig()
   ) -> (then) {is=title K2Node_CallFunction}
   > Get Prop Control Info(self:EUW Prop Control Setup Object Reference=Get PropControlSetup_Widget()) {is=title K2Node_CallFunction}
   Event Construct()
   > Execute Python Script(PythonScript="from utilities.error_reporter import ErrorReporter\r\nErrorReporter.store_construct_snapshot()")
   > n78 = GetCurrentLevelSequence!(execute) -> (then, ReturnValue:Level Sequence Object Reference)
   > Set Level Sequence(Level Sequence:Level Sequence Object Reference=@n78) -> (Output_Get:Level Sequence Object Reference)
   > n79 = Branch(Condition=IsLoadingAssets(self=GetAssetRegistry()))
   > n80 = Set Timer by Function Name!(
    execute
    Object:Object Reference=Self-Reference()
    FunctionName:String=Try Setup
    Time:float=0.2
    bLooping:bool=true
    bMaxOncePerFrame:bool=false
    InitialStartDelay:float=0
    InitialStartDelayVariance:float=0
   ) -> (then, ReturnValue:Timer Handle Structure) {is=title K2Node_CallFunction}
   > n81 = Set Setup Timer Handle(Setup Timer Handle:Timer Handle Structure=@n80) -> (Output_Get:Timer Handle Structure)
   Post SetupComplete(execute=[@n82.then, @n81.then])
   > RefreshTypeCache()
   n82 = Load From Save(execute=@n79.else)
   Event OnMouseEnter()
   > RefreshMaster()
   Event OnMouseLeave!() -> (OutputDelegate, then, MouseEvent:"Pointer Event Structure (by ref)") {is=title K2Node_Event}
   > n83 = GetFocusedLevelSequence!(execute) -> (then, ReturnValue:Level Sequence Object Reference)
   > Set Level Sequence(Level Sequence=@n83)
   Telemetry(feature=ikfk, action=v2_match)
   Telemetry(feature=prop_control, action=apply_space_switches)
  ]}
 Find Project {nodes [
   Find Project!() -> (then, Project Name:String) {is=title K2Node_FunctionEntry}
   > n1 = Get Projects!(execute:Exec, self:Self Object Reference) -> (then:Exec, Projects)
   > n2 = For Each Loop(Array:Array of BP Animation Browser Project Object References=@n1) -> (Array Element:BP Animation Browser Project Object Reference)
   Return Node!(execute=@n2.Completed, AsBP Animation Browser Project:BP Animation Browser Project Object Reference)
   > Branch(Condition="Equal Exactly (String)"(A=Get Name(self=@n2), B=Get Project Name() -> (Project Name:String)))
   > Return Node(AsBP Animation Browser Project=@n2)
  ]}
 Get Projects {nodes [
   Get Projects() {is=title K2Node_FunctionEntry}
   > n1 = Branch(Condition=Get bProjectCached() -> (bProjectCached:bool))
   > Return Node!(
    execute
    Projects:Array of BP Animation Browser Project Object References=Get CachedProjects() -> (CachedProjects:Array of BP Animation Browser Project Object References)
   )
   n2 = GetAssetsByClass!(
    execute=@n1.else
    self:Asset Registry Interface=GetAssetRegistry()
    ClassPathName:Top Level Asset Path Structure=GetSoftClassTopLevelAssetPath!(SoftClassReference:Object Soft Class Reference=/AnimInEdTools/AnimationBrowser/ProjectSettings/BP_AnimationBrowser_Project.BP_AnimationBrowser_Project_C) -> (ReturnValue:Top Level Asset Path Structure)
    bSearchSubClasses:bool=false
   ) -> (then, OutAssetData:Array of Asset Data Structures, ReturnValue:bool)
   > n3 = For Each Loop(Array:Array of Asset Data Structures=@n2) -> (Array Element:Asset Data Structure)
   Set bProjectCached(execute=@n3.Completed, bProjectCached:bool=true) -> (Output_Get:bool)
   > Return Node(Projects=Get CachedProjects())
   n4 = Break Asset Data(AssetData=@n3)
   n5 = LoadAsset(execute=@n3.LoopBody, AssetPath=GetBaseFilename(InPath="To String (Name)"(InName=@n4), bRemovePath=false))
   > n6 = Cast To BP_AnimationBrowser_Project!(execute, Object=@n5) -> (then, CastFailed, AsBP Animation Browser Project:BP Animation Browser Project Object Reference) {is=title K2Node_DynamicCast}
   > Add Unique(
    TargetArray:Array of BP Animation Browser Project Object References=Get CachedProjects()
    NewItem:"BP Animation Browser Project Object Reference (by ref)"=@n6
   )
  ]}
 Get Subfolders {nodes [
   Get Subfolders() {is=title K2Node_FunctionEntry}
   > n1 = Sequence()
   n2 = Get Variant(self=Get AnimatableProp())
   n3 = Get Selected Plugin()
   n4 = Get Selected Plugin()
   > Branch(Condition=OR Boolean(A="Equal Exactly (String)"(A=@n2, B=@n3), B="Equal Exactly (String)"(A=All, B=@n4)))
   > Clear(TargetArray=Get NewSubFolder() -> (NewSubFolder:Array of Strings))
   n5 = Get AnimatableProp()
   > Is Valid(InputObject=@n5)
   > Add Unique(TargetArray:Array of Strings=Get NewSubFolder(), NewItem:"String (by ref)"=All)
   > n6 = Branch(Condition=Get Populate Subfolders From Disk!(self:Animatable Creature Object Reference=@n5) -> (Populate Subfolders From Disk:bool))
   n7 = Get AnimatableProp()
   > n8 = Get Subfolder Root!(
    execute
    self:EU Animation Naming Convention Object Reference=Get Naming Convention!(self:Animatable Creature Object Reference=@n7) -> (ReturnValue:EU Animation Naming Convention Object Reference) {is=title K2Node_CallFunction}
    Path:String=Get Animation Path!(self:Animatable Creature Object Reference=@n7) -> (Animation Path:String)
   ) -> (then, Root Path:String) {is=title K2Node_CallFunction}
   > n9 = ListAssets!(execute, DirectoryPath:String=@n8, bRecursive:bool=false, bIncludeFolder:bool=true) -> (then, ReturnValue:Array of Strings)
   > n10 = For Each Loop(Array:Array of Strings=@n9) -> (Array Element:String)
   > n11 = DoesDirectoryExist!(execute, DirectoryPath:String=@n10) -> (then, ReturnValue:bool)
   > Branch(Condition=@n11)
   > Add Unique(
    TargetArray=Get NewSubFolder()
    NewItem=GetCleanFilename!(InPath:String=Split(SourceString=@n10, InStr=/, SearchDir=FromEnd)) -> (ReturnValue:String)
   )
   n12 = For Each Loop(
    Exec=@n6.else
    Array=Get Subfolders!(self:Animatable Creature Object Reference=Get AnimatableProp()) -> (Subfolders) {is=title K2Node_CallFunction}
   )
   > Add Unique(TargetArray=Get NewSubFolder(), NewItem=@n12)
   Return Node!(execute=@n1.then_1, Subfolders:Array of Strings=Get NewSubFolder())
  ]}
 Import and Create Anim Sequence {nodes [
   Import and Create Anim Sequence!() -> (then, AnimSequence Path:String, LevelSequence Path:String, Animatable Creature:Animatable Creature Object Reference) {is=title K2Node_FunctionEntry}
   > n1 = Sequence()
   n2 = Get Sequencer Actor(self=Get AnimatableProp())
   > n3 = Branch(Condition=IsValidSoftClassReference(SoftClassReference=@n2))
   > n4 = LoadClassAsset_Blocking(AssetClass=@n2)
   > n5 = AddSpawnableFromClass!(
    execute
    self:Level Sequence Editor Subsystem Object Reference=Get LevelSequenceEditorSubsystem() -> (ReturnValue:Level Sequence Editor Subsystem Object Reference) {is=title K2Node_GetEditorSubsystem}
    Sequence:Movie Scene Sequence Object Reference=Get Level Sequence()
    ClassToSpawn:Object Class Reference=@n4
   ) -> (then, ReturnValue:Movie Scene Binding Proxy Structure)
   > n6 = Is Valid(InputObject=Break Movie Scene Binding Proxy!(MovieSceneBindingProxy:"Movie Scene Binding Proxy Structure (by ref)"=@n5) -> (BindingID:Guid Structure, Sequence:Movie Scene Sequence Object Reference) {is=title K2Node_BreakStruct}.Sequence)
   Execute Python Script(
    execute=@n6.Is Not Valid
    PythonScript="from utilities.error_reporter import ErrorReporter\r\n\r\nErrorReporter.send(\r\n    tool_name=\"PropBrowser\",\r\n    feature=\"import_and_create/spawnable_from_class_failed\",\r\n    error=Exception(\"AddSpawnableFromClass returned an invalid binding — BP class may have failed to load or spawn\"),\r\n    traceback_str=\"No Python traceback — failure detected in Blueprint via MovieSceneBindingProxy.Sequence IsValid check\"\r\n)"
   )
   > PrintString(InString=Invalid Skeletal Mesh or Static Mesh set on AP data asset.)
   > Return Node!(execute, Control Rig:Movie Scene Track Object Reference)
   n7 = Set Spawned Binding for Asset(execute=@n6.Is Valid, Spawned Binding for Asset:Movie Scene Binding Proxy Structure=@n5) -> (Output_Get:Movie Scene Binding Proxy Structure)
   n8 = Load Asset Blocking!(
    execute=[@n7.then, @n19.then]
    InRef:"Control Rig Asset Soft Reference Structure (by ref)"=Get Control Rig!(self:Animatable Creature Object Reference=Get AnimatableProp()) -> (Control Rig:Control Rig Asset Soft Reference Structure)
    bForceLoad:bool=true
   ) -> (then, ReturnValue:Control Rig Asset Strong Reference Structure)
   > n9 = FindOrCreateControlRigTrackFromReference!(
    execute
    World:World Object Reference
    LevelSequence:Level Sequence Object Reference=Get Level Sequence()
    InRef:"Control Rig Asset Strong Reference Structure (by ref)"=@n8
    InBinding:"Movie Scene Binding Proxy Structure (by ref)"=Get Spawned Binding for Asset()
    bIsLayeredControlRig:bool=false
   ) -> (then, ReturnValue:Movie Scene Track Object Reference)
   > ForceUpdate()
   > Set Control Rig Created(Control Rig Created:Movie Scene Track Object Reference=@n9) -> (Output_Get:Movie Scene Track Object Reference)
   n10 = Get Skeletal Mesh Asset(self=Get AnimatableProp())
   n11 = Branch(execute=@n3.else, Condition=IsValidSoftObjectReference(SoftObjectReference=@n10))
   > LoadAsset_Blocking(Asset=@n10)
   > n12 = GetEditorWorld(self=Get UnrealEditorSubsystem())
   > n13 = SpawnActor Skeletal Mesh Actor!(
    execute:Exec
    WorldContextObject:Object Reference=@n12
    Class:Actor Class Reference=/Script/Engine.SkeletalMeshActor
    SpawnTransform:Transform=MakeTransform(Rotation="0, 0, 0")
    CollisionHandlingOverride:ESpawnActorCollisionHandlingMethod Enum=Undefined
    TransformScaleMethod:ESpawnActorScaleMethod Enum=MultiplyWithRoot
    Owner:Actor Object Reference
    Instigator:Pawn Object Reference
   ) -> (then:Exec, ReturnValue:Skeletal Mesh Actor Object Reference) {is=title K2Node_SpawnActorFromClass}
   n14 = Resolve Soft Reference!(Input:Skeletal Mesh Soft Object Reference=@n10) -> (Output:Skeletal Mesh Object Reference)
   > SetActorLabel(self=@n13, NewActorLabel=GetDisplayName(Object=@n14), bMarkDirty=true)
   > SetSkeletalMeshAsset!(
    execute
    self:Skeletal Mesh Component Object Reference=Get SkeletalMeshComponent!(self:Skeletal Mesh Actor Object Reference=@n13) -> (SkeletalMeshComponent:Skeletal Mesh Component Object Reference)
    NewMesh:Skeletal Mesh Object Reference=@n14
   ) -> (then)
   > n15 = AddPossessable!(execute, Sequence:Movie Scene Sequence Object Reference=Get Level Sequence(), ObjectToPossess:Object Reference=@n13) -> (then, ReturnValue:Movie Scene Binding Proxy Structure)
   > n16 = ConvertToSpawnable!(
    execute
    self:Level Sequence Editor Subsystem Object Reference=Get LevelSequenceEditorSubsystem()
    ObjectBinding:"Movie Scene Binding Proxy Structure (by ref)"=@n15
   ) -> (then, ReturnValue:Array of Movie Scene Binding Proxy Structures)
   n17 = "Get (a copy)"(Array:Array of Movie Scene Binding Proxy Structures=@n16) -> (Output:Movie Scene Binding Proxy Structure)
   > n18 = Is Valid(InputObject=Break Movie Scene Binding Proxy(MovieSceneBindingProxy=@n17).Sequence)
   Execute Python Script(
    execute=@n18.Is Not Valid
    PythonScript="from utilities.error_reporter import ErrorReporter\r\n\r\nErrorReporter.send(\r\n    tool_name=\"PropBrowser\",\r\n    feature=\"import_and_create/spawnable_from_class_failed\",\r\n    error=Exception(\"AddSpawnableFromClass returned an invalid binding — BP class may have failed to load or spawn\"),\r\n    traceback_str=\"No Python traceback — failure detected in Blueprint via MovieSceneBindingProxy.Sequence IsValid check\"\r\n)"
   )
   > PrintString(InString=Invalid Skeletal Mesh or Static Mesh set on AP data asset.)
   > Return Node()
   n19 = Set Spawned Binding for Asset(execute=@n18.Is Valid, Spawned Binding for Asset=@n17)
   n20 = Get Static Mesh Asset(self=Get AnimatableProp())
   n21 = Branch(execute=@n11.else, Condition=IsValidSoftObjectReference(SoftObjectReference=@n20))
   > n22 = LoadAsset_Blocking(Asset=@n20)
   > n23 = Cast To StaticMesh(Object=@n22)
   > n24 = AddSpawnableFromInstance!(
    execute
    self:Level Sequence Editor Subsystem Object Reference=Get LevelSequenceEditorSubsystem()
    Sequence:Movie Scene Sequence Object Reference=Get Level Sequence()
    ObjectToSpawn:Object Reference=@n23
   ) -> (then, ReturnValue:Movie Scene Binding Proxy Structure)
   > SetDisplayName!(
    execute
    InBinding:"Movie Scene Binding Proxy Structure (by ref)"=@n24
    InDisplayName:"Text (by ref)"="To Text (String)"(InString=GetDisplayName(Object=@n23))
   ) -> (then)
   > n25 = Is Valid(InputObject=Break Movie Scene Binding Proxy(MovieSceneBindingProxy=@n24).Sequence)
   Execute Python Script(
    execute=@n25.Is Not Valid
    PythonScript="from utilities.error_reporter import ErrorReporter\r\n\r\nErrorReporter.send(\r\n    tool_name=\"PropBrowser\",\r\n    feature=\"import_and_create/spawnable_from_class_failed\",\r\n    error=Exception(\"AddSpawnableFromClass returned an invalid binding — BP class may have failed to load or spawn\"),\r\n    traceback_str=\"No Python traceback — failure detected in Blueprint via MovieSceneBindingProxy.Sequence IsValid check\"\r\n)"
   )
   > PrintString(InString=Invalid Skeletal Mesh or Static Mesh set on AP data asset.)
   > Return Node()
   Set Spawned Binding for Asset(execute=@n25.Is Valid, Spawned Binding for Asset=@n24)
   PrintString(execute=[@n23.CastFailed, @n21.else], InString=Invalid Skeletal Mesh or Static Mesh set on AP data asset.)
   > Return Node()
   n26 = Branch(execute=@n1.then_1, Condition=IsEmpty(InString=Get AnimSequence Path() -> (AnimSequence Path:String)))
   Return Node(execute=[@n28.then, @n26.then], Control Rig=Get Control Rig Created() -> (Control Rig Created:Movie Scene Track Object Reference))
   n27 = Create Linked Anim Sequence!(
    execute=@n26.else
    self
    Level Sequence:Level Sequence Object Reference=Get Level Sequence()
    New AnimSequence Path:String=Get AnimSequence Path()
    Binding:Movie Scene Binding Proxy Structure=Get Spawned Binding for Asset()
   ) -> (then, Anim Sequence:Anim Sequence Object Reference) {is=title K2Node_CallFunction}
   > SaveLoadedAsset(AssetToSave=@n27, bOnlyIfIsDirty=false)
   > n28 = ForceUpdate()
  ]}
 Load From Save {nodes [
   Load From Save() {is=title K2Node_FunctionEntry}
   > n1 = Sequence!(execute) -> (then_0, then_1, then_2:Exec)
   > ClearChildren(self=Get ModalLayer())
   > n2 = LoadGameFromSlot!(execute, SlotName:String=Get Slot Name() -> (Slot Name:String), UserIndex:int=0) -> (then, ReturnValue:Save Game Object Reference)
   > n3 = Cast To SG_PropBrowser!(execute, Object=@n2) -> (then, CastFailed, AsSG Prop Browser:SG Prop Browser Object Reference) {is=title K2Node_DynamicCast}
   > n4 = Set Save Game(Save Game:SG Prop Browser Object Reference=@n3) -> (Output_Get:SG Prop Browser Object Reference)
   n5 = Get Project!(self:SG Prop Browser Object Reference=@n4) -> (Project:BP Animation Browser Project Object Reference)
   > n6 = Is Valid(InputObject=@n5)
   n7 = Display Modal Dialog!(
    execute=[@n3.CastFailed, @n6.Is Not Valid]
    self
    Class:User Widget Class Reference=&/EUW_SelectPropProjectDialog.EUW_SelectPropProjectDialog_C
   ) -> (then, Modal Widget:EUW Modal Dialog Object Reference, New Widget:User Widget Object Reference) {is=title K2Node_CallFunction}
   > n8 = Cast To EUW_SelectPropProjectDialog!(execute, Object=@n7.New Widget) -> (then, CastFailed, AsEUW Select Prop Project Dialog:EUW Select Prop Project Dialog Object Reference) {is=title K2Node_DynamicCast}
   > Set Projects!(
    execute
    self:EUW Select Prop Project Dialog Object Reference=@n8
    Projects:Array of BP Animation Browser Project Object References=Get Projects!() -> (Projects) {is=title K2Node_CallFunction}
   ) -> (then) {is=title K2Node_CallFunction}
   > Bind Event to Project Selected(self:EUW Select Prop Project Dialog Object Reference=@n8, Delegate=Create Event()) {is=title K2Node_AddDelegate}
   > n9 = CreateSaveGameObject!(execute, SaveGameClass:Save Game Class Reference=&/SG_PropBrowser.SG_PropBrowser_C) -> (then, ReturnValue:SG Prop Browser Object Reference)
   > n10 = Set Save Game(Save Game=@n9)
   > SaveGameToSlot!(execute, SaveGameObject:Save Game Object Reference=@n10, SlotName:String=Get Slot Name(), UserIndex:int=0) -> (then, ReturnValue:bool)
   Set Current Project(execute=@n6.Is Valid, Current Project=@n5)
   > RefreshMaster()
   > n11 = AnimatableProps Get Assets(AnimatableProps=Get Animatable Props(Project=Get Current Project()))
   > Update Asset ComboBox!(execute, self, Categories:Array of Strings=@n11, Default Item:String=All) -> (then)
   n12 = Get Asset(self:SG Prop Browser Object Reference=Get Save Game() -> (Save Game:SG Prop Browser Object Reference))
   > SetSelectedOption!(execute, self:"Searchable Combo Box (String) Object Reference"=Get SearchableAssetComboBox(), Option:String=@n12) -> (then)
   > Set Selected Asset(Selected Asset=@n12)
   > SetSelectedOption(self=Get SearchablePluginComboBox(), Option=Get Variant(self:SG Prop Browser Object Reference=Get Save Game()))
   > SetSelectedOption(self=Get SearchableSubfolderComboBox(), Option=Get subfolder!(self:SG Prop Browser Object Reference=Get Save Game()) -> (subfolder:String))
   Populating Extra Slot(execute=@n1.then_1)
   > n13 = Is Valid(InputObject=Get Level Sequence())
   Set bIsEnabled(
    execute=@n13.Is Not Valid
    bIsEnabled=false
    self=Get PropAddVerticalBoxOptions() -> (PropAddVerticalBoxOptions:Vertical Box Object Reference)
   )
   Set bIsEnabled(execute=@n13.Is Valid, bIsEnabled=true, self=Get PropAddVerticalBoxOptions())
   Disable Import Linked Prop Button(execute=@n1.then_2, Animatable Prop=Get AnimatableProp())
  ]}
 OnMouseButtonDown {nodes [
   OnMouseButtonDown!() -> (then, MyGeometry:Geometry Structure, MouseEvent:"Pointer Event Structure (by ref)") {is=title K2Node_FunctionEntry}
   > Show(self=Get EUW_ContextMenu(), Show=false)
   > Return Node!(execute, ReturnValue:Event Reply Structure=Unhandled())
  ]}
 On_ParentControlRigComboBox_GenerateWidget {nodes [
   On_ParentControlRigComboBox_GenerateWidget!() -> (then, Item:String) {is=title K2Node_FunctionEntry}
   > Return Node(ReturnValue:Widget Object Reference)
  ]}
 Populate Asset ComboBox {nodes [
   Populate Asset ComboBox() {is=title K2Node_FunctionEntry}
   > n1 = Sequence()
   > n2 = Animatable Prop Get Variants!(execute, self, Asset:String=Get Selected Asset(), Add All option:bool=true) -> (then, CollectedVariants:Array of Strings) {is=title K2Node_CallFunction}
   > Update Combo Box(Combo Box=Get SearchablePluginComboBox(), Items=@n2)
   > Branch(Condition="integer > integer"(A=Length(TargetArray=@n2)))
   > SetSelectedOption(self=Get SearchablePluginComboBox(), Option="Get (a copy)"(Array:Array of Strings=@n2) -> (Output:String))
   n3 = For Each Loop(Exec=@n1.then_1, Array:Array of Animatable Creature Object References=Get Animatable Props(Project=Get Current Project())) -> (Array Element:Animatable Creature Object Reference)
   > Branch(Condition="Equal Exactly (String)"(A=Get Asset(self:Animatable Creature Object Reference=@n3), B=Get Selected Asset()))
   > n4 = Cast To AnimatableProp(Object=@n3)
   > Set AnimatableProp(AnimatableProp=@n4)
   n5 = Get Subfolders!() -> (Subfolders)
   n6 = Branch(execute=@n1.then_2, Condition="integer <= integer"(A=Length(TargetArray=@n5), B=2))
   n7 = Get SearchableSubfolderComboBox()
   > Set Visibility(Visibility=Collapsed, self=@n7)
   > n8 = Set Visibility(Visibility=Collapsed, self=Get Subfolder())
   Update Combo Box(execute=[@n8.then, @n10.then], Combo Box=Get SearchableSubfolderComboBox(), Items=Get Subfolders(), Default Item=All)
   > Save UI State()
   > Populate Control Rig Combo Box(Level Sequence=Get Level Sequence(), Combo Box=Get ParentControlRigComboBox())
   n9 = Get SearchBar()
   > Update List View(
    Selected Plugin=GetSelectedOption(self=Get SearchablePluginComboBox())
    Filter="To String (Text)"(InText=Get Text(self:Text Box Object Reference=@n9))
    Subfolder=GetSelectedOption(self=Get SearchableSubfolderComboBox())
    Selected Asset=GetSelectedOption(self=Get SearchableAssetComboBox())
   )
   Set Visibility(execute=@n6.else, Visibility=Visible, self=@n7)
   > n10 = Set Visibility(Visibility=Visible, self=Get Subfolder())
  ]}
 Populate Control Rig Combo Box {nodes [
   n1 = Populate Control Rig Combo Box!() -> (then, Level Sequence:Level Sequence Object Reference, Combo Box:Editor Utility Combo Box String Object Reference) {is=title K2Node_FunctionEntry}
   > n2 = Sequence()
   > ClearOptions(self:"Combo Box (String) Object Reference"=Get Combo Box() -> (Combo Box:Editor Utility Combo Box String Object Reference))
   > Clear(TargetArray:Array of Control Rig Asset Soft Reference Structures=Get AP CRs() -> (AP CRs:Array of Control Rig Asset Soft Reference Structures))
   > Clear(TargetArray:Array of Control Rig Object References=Get ParentRigs())
   > n3 = For Each Loop(Array=Get Animatable Props(Project=Get Current Project()))
   > Add Unique(
    TargetArray:Array of Control Rig Asset Soft Reference Structures=Get AP CRs()
    NewItem:"Control Rig Asset Soft Reference Structure (by ref)"=Get Control Rig(self=@n3)
   )
   n4 = GetControlRigs(execute=@n2.then_1, LevelSequence=@n1)
   > n5 = For Each Loop!(Exec, Array:Array of Control Rig Sequencer Binding Proxy Structures=@n4) -> (
    LoopBody
    Array Element_Proxy:Movie Scene Binding Proxy Structure
    Array Element_ControlRig:Control Rig Object Reference
    Array Element_Track:Movie Scene Control Rig Parameter Track Object Reference
    Array Index
    Completed
   )
   n6 = GetAssetReference!(self:Control Rig Object Reference=@n5.Array Element_ControlRig) -> (ReturnValue:Control Rig Asset Strong Reference Structure)
   > n7 = Branch(Condition=Contains Item(
     TargetArray:Array of Control Rig Asset Soft Reference Structures=Get AP CRs()
     ItemToFind:"Control Rig Asset Soft Reference Structure (by ref)"="To Soft Reference (Control Rig Strong Reference)"!(InRef:"Control Rig Asset Strong Reference Structure (by ref)"=@n6) -> (ReturnValue:Control Rig Asset Soft Reference Structure)
    ))
   n8 = GetDisplayName!(execute:Exec=@n7.else, InBinding:"Movie Scene Binding Proxy Structure (by ref)"=@n5) -> (then:Exec, ReturnValue:Text)
   n9 = Append!(
    A="To String (Text)"(InText=@n8)
    B="   ("
    C:String=GetDisplayName!(Object:Object Reference=@n5.Array Element_ControlRig) -> (ReturnValue:String)
    D:String=")"
   ) -> (ReturnValue)
   > AddOption!(execute, self:"Combo Box (String) Object Reference"=Get Combo Box(), Option:String=@n9) -> (then)
   > Add(
    TargetArray:Array of Control Rig Object References=Get ParentRigs()
    NewItem:"Control Rig Object Reference (by ref)"=@n5.Array Element_ControlRig
   ) {is=title K2Node_CallArrayFunction}
   > Branch(Condition="Equal (Integer)"(A=@n5.Array Index))
   > Set Parent Rig Binding(Parent Rig Binding=@n5)
   > Set Selected Parent Asset(Selected Parent Asset=@n9)
   > Set Selected Parent Rig(Selected Parent Rig=@n5.Array Element_ControlRig)
   SetSelectedIndex(execute=@n2.then_2, self=Get Combo Box(), Index=0)
   > n10 = Set Control Rig Controls for Combo Box(
    Selected Control Rig=Get Selected Parent Rig()
    Combo Box=Get AttachmentPointSearchableComboBox()
    Socket attachment=Get CheckedState(self=Get attach_socket_chkbx()) -> (CheckedState=Unchecked)
   )
   > Set the Attach Point selection(Array=@n10)
  ]}
 Populate Season Box From Project Selection {nodes [
   Populate Season Box From Project Selection() {is=title K2Node_FunctionEntry}
   > n1 = AnimatableProps Get Assets(AnimatableProps=Get Animatable Props(Project=Get Current Project()))
   > Update Asset ComboBox(Categories=@n1, Default Item="Get (a ref)"(Array=@n1, Dimension 1=0))
  ]}
 Populating Extra Slot {nodes [
   Populating Extra Slot() {is=title K2Node_FunctionEntry}
   > Sequence!(execute) -> (then_0, then_1)
   > ClearChildren(self=Get ExtraSlot() -> (ExtraSlot:Named Slot Object Reference))
   n1 = Get Current Project()
   > Branch(Condition=Contains(SearchIn=GetDisplayName(Object=@n1), Substring=BRGameplay))
   > n2 = Create Widget(Class=&/EUW_Slot_Filler_Feral.EUW_Slot_Filler_Feral_C) -> (ReturnValue:EUW Slot Filler Feral Object Reference)
   > AddChild(self=Get ExtraSlot(), Content=@n2)
  ]}
 Post SetupComplete {nodes [
   Post SetupComplete() {is=title K2Node_FunctionEntry}
   > Execute Python Script(PythonScript="import PropBrowser\r\n\r\nfrom importlib import reload\r\nreload(PropBrowser)")
   > Execute Python Script!(
    execute
    PythonScript="try:\r\n  import unreal\r\n  from unpipe import import_shared_plugin\r\n  import_shared_plugin('epic-shrd-techart')\r\n  from epic_shrd_techart.easy_meter import EasyMeter\r\n  EasyMeter.configure(\r\n      app_name=app_name,\r\n      app_version=app_version,\r\n      environment=environment,\r\n      tags=[\"team:anim-in-ed\", \"team:techanim-gameplay\", \"type:euw\", \"dcc:unreal\"]\r\n  )\r\nexcept Exception as e:\r\n  print(f\"[Telemetry] configure skipped (non-fatal): {e}\")"
    appVersion:String=Get AppVersion() -> (AppVersion:String)
    app_name:String=Get tool_name() -> (tool_name:String)
    environment:String=Get Environment name() -> (Environment name:String)
   ) -> (then, ReturnValue) {inputs [appVersion,app_name,environment]}
   > Execute Python Script!(
    execute
    PythonScript="from utilities.error_reporter import ErrorReporter\r\nErrorReporter.check_construct_errors(tool_name=tool_name, feature=feature_name)"
    featureName:String=Get feature_name() -> (feature_name:String)
    toolName:String=Get tool_name()
   ) -> (then, ReturnValue) {inputs [featureName,toolName]}
  ]}
 RefreshMaster {nodes [
   RefreshMaster() {is=title K2Node_FunctionEntry}
   > n1 = GetFocusedLevelSequence()
   > n2 = Is Valid(InputObject=@n1)
   Set bIsEnabled(execute=@n2.Is Not Valid, bIsEnabled=false, self=Get PropAddVerticalBoxOptions())
   Set bIsEnabled(execute=@n2.Is Valid, bIsEnabled=true, self=Get PropAddVerticalBoxOptions())
  ]}
 RefreshTypeCache {nodes [
   RefreshTypeCache() {is=title K2Node_FunctionEntry}
   > n1 = For Each Loop!(Exec, Array:Array of Animatable Creature Object References=Get Animatable Props(Project=Get Current Project())) -> (LoopBody, Array Element:Animatable Creature Object Reference, Array Index, Completed)
   WriteStringIntoFile!(
    execute=@n1.Completed
    WriteString:String=Get WriteString() -> (WriteString:String)
    FilePath:String=Append!(A=GetProjectSavedDirectory!() -> (ReturnValue:String), B=sg_ap_input.txt) -> (ReturnValue)
    Unique:bool=true
    Clear:bool=true
    __WorldContext:Object Reference
   ) -> (then)
   > Execute Python Script!(
    execute
    PythonScript="import sys, os, unreal\r\n_sg_dir = os.path.join(unreal.Paths.project_content_dir(), \"Python\", \"TechAnim\", \"ShotGridReleaseVersion\")\r\nif _sg_dir not in sys.path:\r\n  sys.path.append(_sg_dir)\r\nimport sg_version_lookup\r\nsg_version_lookup.run_from_blueprint()"
   ) -> (then, ReturnValue)
   > Update List View(Selected Plugin=Get Selected Plugin(), Subfolder=All, Selected Asset=Get Selected Asset())
   n2 = Cast To AnimatableProp(execute=@n1.LoopBody, Object=@n1)
   > Set WriteString(WriteString:String=Append!(A=Get WriteString(), B=GetSoftObjectPath(Object=@n2), C:String="\\n") -> (ReturnValue)) -> (Output_Get:String)
  ]}
 Save UI State {nodes [
   Save UI State() {is=title K2Node_FunctionEntry}
   > n1 = CreateSaveGameObject(SaveGameClass=&SG_PropBrowser_C)
   > n2 = Set NewLocalVar(NewLocalVar:SG Prop Browser Object Reference=@n1) -> (Output_Get:SG Prop Browser Object Reference)
   > Set Asset!(execute, Asset:String=GetSelectedOption(self=Get SearchableAssetComboBox()), self:SG Prop Browser Object Reference=@n2) -> (then, Output_Get:String)
   > Set Variant!(execute, Variant:String=GetSelectedOption(self=Get SearchablePluginComboBox()), self:SG Prop Browser Object Reference=@n2) -> (then, Output_Get:String)
   > Set subfolder!(execute, subfolder:String=GetSelectedOption(self=Get SearchableSubfolderComboBox()), self:SG Prop Browser Object Reference=@n2) -> (then, Output_Get:String)
   > Set Project(Project=Get Current Project(), self:SG Prop Browser Object Reference=@n2) {is=title K2Node_VariableSet}
   > SaveGameToSlot(SaveGameObject=@n2, SlotName=Get Slot Name(), UserIndex=0)
  ]}
 Set Attach Bone selection {nodes [
   Set Attach Bone selection!() -> (then, Default Attach Bone:String) {is=title K2Node_FunctionEntry}
   > Set Default Attach Bone Exists(Default Attach Bone Exists:bool=false) -> (Output_Get:bool)
   > n1 = For Each Loop(Array:Array of Strings=Get Attach Point Array() -> (Attach Point Array:Array of Strings)) -> (Array Element:String)
   n2 = Branch(execute=@n1.Completed, Condition=Get Default Attach Bone Exists() -> (Default Attach Bone Exists:bool))
   > SetSelectedOption(self=Get AttachmentPointSearchableComboBox(), Option=Get Default Attach Bone!() -> (Default Attach Bone))
   SetSelectedIndex(execute=@n2.else, self:"Searchable Combo Box (String) Object Reference"=Get AttachmentPointSearchableComboBox(), Index=0)
   Branch(execute=@n1.LoopBody, Condition="Equal Exactly (String)"(A=@n1, B=Get Default Attach Bone()))
   > Set Default Attach Bone Exists(Default Attach Bone Exists=true)
  ]}
 Set Control Rig Controls for Combo Box {nodes [
   Set Control Rig Controls for Combo Box!() -> (
    then
    Selected Control Rig:Control Rig Object Reference
    Combo Box:"Searchable Combo Box (String) Object Reference"
    Socket attachment:ECheckBoxState Enum=Unchecked
   ) {is=title K2Node_FunctionEntry}
   > n1 = Sequence!(execute) -> (then_0, then_1, then_2:Exec, then_3:Exec, then_4:Exec)
   > ClearOptions(self:"Searchable Combo Box (String) Object Reference"=Get Combo Box() -> (Combo Box:"Searchable Combo Box (String) Object Reference"))
   > Clear(TargetArray:Array of Strings=Get Attach Point Array())
   n2 = Branch(
    execute=@n1.then_1
    Condition="Equal (Enum)"(A:ECheckBoxState Enum=Get Socket attachment() -> (Socket attachment:ECheckBoxState Enum=Unchecked), B:ECheckBoxState Enum=Checked)
   )
   > n3 = GetFocusedLevelSequence()
   > n4 = GetBindings!(execute, Sequence:Movie Scene Sequence Object Reference=@n3) -> (then, ReturnValue:Array of Movie Scene Binding Proxy Structures)
   > n5 = For Each Loop with Break(Array:Array of Movie Scene Binding Proxy Structures=@n4, Break=@n17.then) -> (Array Element:Movie Scene Binding Proxy Structure)
   > n6 = Set Binding(Binding:Movie Scene Binding Proxy Structure=@n5) -> (Output_Get:Movie Scene Binding Proxy Structure)
   > n7 = Branch(Condition="Equal (MovieSceneBindingProxy)"(A:Movie Scene Binding Proxy Structure=@n6, B:Movie Scene Binding Proxy Structure=Get Parent Rig Binding()) -> (ReturnValue:bool=~) {is=title K2Node_PromotableOperator})
   n8 = GetBindingID(execute=[@n7.then, @n19.then], Sequence=@n3, InBinding=Get Binding())
   > n9 = For Each Loop(Array:Array of Object References=GetBoundObjects(ObjectBinding=@n8)) -> (Array Element:Object Reference)
   > n10 = Cast To Actor(Object=@n9)
   n11 = Get SkeletalMesh(self=GetComponentByClass(self=@n10, ComponentClass=&SkeletalMeshComponent))
   > n12 = For Loop(LastIndex=NumSockets!(self:Skeletal Mesh Object Reference=@n11) -> (ReturnValue:int=0))
   n13 = GetSocketByIndex!(self:Skeletal Mesh Object Reference=@n11, Index:int=@n12) -> (ReturnValue:Skeletal Mesh Socket Object Reference)
   > Is Valid(InputObject=@n13)
   n14 = Get SocketName!(self:Skeletal Mesh Socket Object Reference=@n13) -> (SocketName:Name=None)
   > Set socket_name(socket_name:String="To String (Name)"(InName=@n14)) -> (Output_Get:String)
   n15 = "To String (Name)"(InName=@n14)
   > n16 = Branch(Condition=OR Boolean!(
     A=Contains(SearchIn=@n15, Substring="VB ", bUseCase=true)
     B=Contains(SearchIn=@n15, Substring=FX, bUseCase=true)
     C:bool=Contains(SearchIn=@n15, Substring=Melee, bUseCase=true)
     D:bool=Contains(SearchIn=@n15, Substring=SM_, bUseCase=true)
     E:bool=Contains(SearchIn=@n15, Substring=Water_Cap, bUseCase=true)
    ) -> (ReturnValue))
   n17 = Add Unique(
    execute=@n16.else
    TargetArray:Array of Strings=Get Attach Point Array()
    NewItem:"String (by ref)"=Get socket_name() -> (socket_name:String)
   )
   n18 = GetName!(execute=@n7.else, InBinding:"Movie Scene Binding Proxy Structure (by ref)"=Get Binding()) -> (then, ReturnValue:String)
   > n19 = Branch(Condition="Equal Exactly (String)"(A=@n18, B=Get Selected Parent Asset() -> (Selected Parent Asset:String)))
   n20 = Get Selected Control Rig() -> (Selected Control Rig:Control Rig Object Reference)
   n21 = For Each Loop!(
    Exec=@n2.else
    Array:Array of Rig Element Key Structures=Get Bones!(
     self:Rig Hierarchy Object Reference=GetHierarchy!(self:Control Rig Object Reference=@n20) -> (ReturnValue:Rig Hierarchy Object Reference)
     bTraverse:bool=true
    ) -> (ReturnValue:Array of Rig Element Key Structures) {is=title K2Node_CallFunction}
   ) -> (LoopBody, Array Element_Type:ERigElementType Enum, Array Element_Name:Name, Array Index, Completed)
   n22 = "To String (Name)"(InName=@n21.Array Element_Name)
   > n23 = Branch(Condition=StartsWith(SourceString=@n22, InPrefix=VB, SearchCase=CaseSensitive))
   Add Unique(execute=@n23.else, TargetArray=Get Attach Point Array(), NewItem=@n22)
   n24 = Execute Python Script!(execute=@n1.then_2, PythonScript="sorted_controls=sorted(controls)", controls:Array of Strings=Get Attach Point Array()) -> (then, ReturnValue, sorted_controls:Array of Strings) {inputs [controls], outputs [sorted_controls]}
   > n25 = Set Attach Point Array(Attach Point Array:Array of Strings=@n24.sorted_controls) -> (Output_Get:Array of Strings)
   > n26 = For Each Loop!(Exec, Array:Array of Strings=@n25) -> (LoopBody, Array Element:String, Array Index, Completed)
   > AddOption(self:"Searchable Combo Box (String) Object Reference"=Get Combo Box(), Option=@n26)
   n27 = Get Combo Box()
   Set Attach Bone Name(execute=@n1.then_3, Attach Bone Name=String To Name(InString=GetSelectedOption(self=@n27))) -> (Output_Get=None)
   Return Node!(execute=@n1.then_4, Attach Point Array:Array of Strings=Get Attach Point Array())
  ]}
 Set Project {nodes [
   n1 = Set Project!() -> (then, Project:BP Animation Browser Project Object Reference) {is=title K2Node_FunctionEntry}
   > Set Current Project(Current Project=@n1)
   > RefreshMaster() {is=title K2Node_CallFunction}
   > n2 = Sequence!(execute) -> (then_0, then_1)
   > Populate Season Box From Project Selection() {is=title K2Node_CallFunction}
   > SetSelectedOption(self=Get SearchableAssetComboBox(), Option=Get Asset(self:SG Prop Browser Object Reference=Get Save Game()))
   n3 = Get Save Game()
   Set Asset(execute=@n2.then_1, Asset=GetSelectedOption(self=Get SearchableAssetComboBox()), self=@n3)
   > Set Project(Project=Get Current Project(), self=@n3) {is=title K2Node_VariableSet}
   > Save UI State() {is=title K2Node_CallFunction}
  ]}
 Set the Attach Point selection {nodes [
   n1 = Set the Attach Point selection!() -> (then, Array:Array of Strings) {is=title K2Node_FunctionEntry}
   > SetSelectedIndex(self=Get AttachmentPointSearchableComboBox(), Index=0)
   > n2 = For Each Loop with Break(Array:Array of Strings=@n1, Break=@n4.then) -> (Array Element:String)
   n3 = Get AnimatableProp()
   > Branch(Condition="Equal Exactly (String)"(A=@n2, B=Get Default Attach Bone!(self:Animatable Prop Object Reference=@n3) -> (Default Attach Bone)))
   > n4 = SetSelectedIndex(self=Get AttachmentPointSearchableComboBox(), Index=@n2.Array Index)
  ]}
 SettingPropName in UpdateList {nodes [
   n1 = SettingPropName in UpdateList!() -> (then, Target:Animatable Prop Object Reference, Subfolder:String) {is=title K2Node_FunctionEntry}
   n2 = Get Skeletal Mesh Asset(self=@n1)
   > n3 = Branch(Condition=IsValidSoftObjectReference(SoftObjectReference=@n2))
   > n4 = Branch(Condition="Equal, Case Insensitive (String)"(A=Get Subfolder() -> (Subfolder:String), B=All))
   Return Node!(
    execute=[@n5.then, @n4.then]
    Prop Asset1:String=GetSoftPtrAssetName!(Asset:Soft Object Reference=@n2) -> (ReturnValue:String)
    ReturnValue:bool=true
   )
   n5 = Branch(
    execute=@n4.else
    Condition=Contains(SearchIn=To Soft Object Path!(SoftObjectReference:Soft Object Reference=@n2) -> (ReturnValue_PathString:String), Substring=Get Subfolder())
   )
   n6 = Get Sequencer Actor(self=@n1)
   n7 = Branch(execute=[@n3.else, @n5.else], Condition=IsValidSoftClassReference(SoftClassReference=@n6))
   > n8 = Branch(Condition="Equal, Case Insensitive (String)"(A=Get Subfolder(), B=All))
   n9 = "To String (SoftClassReference)"!(SoftClassReference:"Object Soft Class Reference (by ref)"=@n6) -> (ReturnValue:String)
   Return Node(
    execute=[@n10.then, @n8.then]
    Prop Asset1=Split(SourceString=Split(SourceString=@n9, InStr=., SearchDir=FromStart).RightS, InStr=_C, SearchDir=FromEnd)
    ReturnValue=true
   )
   n10 = Branch(execute=@n8.else, Condition=Contains(SearchIn="To String (SoftClassReference)"(SoftClassReference=@n6), Substring=Get Subfolder()))
   n11 = Get Static Mesh Asset(self=@n1)
   n12 = Branch(execute=[@n10.else, @n7.else], Condition=IsValidSoftObjectReference(SoftObjectReference=@n11))
   > n13 = Branch(Condition="Equal, Case Insensitive (String)"(A=Get Subfolder(), B=All))
   Return Node(execute=[@n14.then, @n13.then], Prop Asset1=GetSoftPtrAssetName(Asset=@n11), ReturnValue=true)
   n14 = Branch(
    execute=@n13.else
    Condition=Contains(
     SearchIn="To String (SoftObjectReference)"!(SoftObjectReference:"Soft Object Reference (by ref)"=@n11) -> (ReturnValue:String)
     Substring=Get Subfolder()
    )
   )
   Return Node(execute=[@n14.else, @n12.else], ReturnValue=false)
  ]}
 Settings Menu Browse to AP File {nodes [
   Settings Menu Browse to AP File!() -> (then, Context Item:Editor Utility Widget Object Reference, Slot Widget:Editor Utility Widget Object Reference) {is=title K2Node_FunctionEntry}
   n1 = Get AnimatableProp()
   > n2 = Is Valid(InputObject=@n1)
   n3 = Create Widget(execute=@n2.Is Not Valid, Class=&/EUW_PropBrowser_Error_Dialog.EUW_PropBrowser_Error_Dialog_C) -> (ReturnValue:EUW Prop Browser Error Dialog Object Reference)
   > AddChild(self=Get ModalLayer_ErrorDialog() -> (ModalLayer_ErrorDialog:Named Slot Object Reference), Content=@n3)
   > Show(self=Get EUW_ContextMenu(), Show=true)
   SyncBrowserToObjects(execute=@n2.Is Valid, AssetPaths=Make Array("[0]"=GetSoftObjectPath(Object=@n1)))
  ]}
 Telemetry {nodes [
   n1 = Telemetry!() -> (then, feature:String, action:String, metadata:String, metadata_function:String) {is=title K2Node_FunctionEntry}
   > Execute Python Script!(
    execute
    PythonScript="try:\r\n  from unpipe import import_shared_plugin\r\n  import_shared_plugin('epic-shrd-techart')\r\n  from epic_shrd_techart.easy_meter import EasyMeter\r\n  if metadata:\r\n      EasyMeter.get().usage(feature=feature_name, action=action_name, metadata=metadata_function)\r\n  else:\r\n      EasyMeter.get().usage(feature=feature_name, action=action_name)\r\nexcept Exception as e:\r\n  print(f\"[Telemetry] usage skipped (non-fatal): {e}\")"
    feature_name:String=@n1
    action_name:String=@n1.action
    metadata:String=@n1.metadata
    metadata_function:String=@n1.metadata_function
   ) -> (then, ReturnValue) {inputs [feature_name, action_name, metadata, metadata_function]}
  ]}
 Try Setup {nodes [
   Try Setup() {is=title K2Node_FunctionEntry}
   > n1 = Branch(Condition=IsLoadingAssets(self=GetAssetRegistry()))
   > Return Node!(execute)
   Clear and Invalidate Timer by Handle!(execute=@n1.else, WorldContextObject:Object Reference, Handle=Get Setup Timer Handle() -> (Setup Timer Handle:Timer Handle Structure)) -> (then)
   > Load From Save() {is=title K2Node_CallFunction}
  ]}
 Update Asset ComboBox {nodes [
   n1 = Update Asset ComboBox!() -> (then, Categories:Array of Strings, Default Item:String) {is=title K2Node_FunctionEntry}
   > Update Combo Box(Combo Box=Get SearchableAssetComboBox(), Items=@n1, Default Item=Get Default Item() -> (Default Item:String))
  ]}
 Update Combo Box {nodes [
   n1 = Update Combo Box!() -> (then, Combo Box:"Searchable Combo Box (String) Object Reference", Items:Array of Strings, Default Item:String) {is=title K2Node_FunctionEntry}
   > n2 = Set Current Index(Current Index:int=GetSelectedIndex(self:"Searchable Combo Box (String) Object Reference"=@n1) -> (ReturnValue=0)) -> (Output_Get:int=0)
   > Set Selected Item(Selected Item:String=GetOptionAtIndex!(self:"Searchable Combo Box (String) Object Reference"=Get Combo Box(), Index:int=@n2) -> (ReturnValue:String)) -> (Output_Get:String) {is=title K2Node_VariableSet}
   n3 = Get Combo Box()
   > n4 = For Loop(LastIndex=int - int(A=GetOptionCount!(self:"Searchable Combo Box (String) Object Reference"=@n3) -> (ReturnValue:int=0)))
   n5 = Branch(
    execute=@n4.Completed
    Condition=Identical!(ArrayA:Array of Strings=Get Items() -> (Items:Array of Strings), ArrayB:Array of Strings=Get Old Items() -> (Old Items:Array of Strings)) -> (ReturnValue:bool) {is=title K2Node_CallArrayFunction}
   )
   ClearOptions(execute=@n5.else, self=Get Combo Box())
   > n6 = Sequence!(execute) -> (then_0, then_1, then_2:Exec)
   n7 = Get Default Item()
   > n8 = Branch(Condition=IsEmpty(InString=@n7))
   AddOption(execute=@n8.else, self=Get Combo Box(), Option=@n7)
   n9 = For Each Loop(Exec=@n6.then_1, Array=Get Items())
   > AddOption(self=Get Combo Box(), Option=@n9)
   n10 = Get Combo Box()
   n11 = Get Current Index() -> (Current Index:int=0)
   n12 = Branch(
    execute=@n6.then_2
    Condition="Equal Exactly (String)"(A=GetOptionAtIndex(self=@n10, Index=@n11), B=Get Selected Item() -> (Selected Item:String))
   )
   n13 = Get Combo Box()
   > AddOption(self=@n13, Option=Temp)
   > SetSelectedIndex(self=@n13, Index=int + int(A=Get Current Index() -> (Current Index=0)))
   > SetSelectedIndex(self=@n13, Index=0)
   > RemoveOption!(execute, self:"Searchable Combo Box (String) Object Reference"=@n13, Option:String=Temp) -> (then, ReturnValue:bool)
   ClearSelection(execute=@n12.else, self:"Searchable Combo Box (String) Object Reference"=Get Combo Box())
   n14 = GetOptionAtIndex(self=Get Combo Box(), Index=@n4)
   n15 = Branch(execute=@n4.LoopBody, Condition="Equal Exactly (String)"(A=@n14, B=Get Default Item()))
   Add(execute=@n15.else, TargetArray:Array of Strings=Get Old Items(), NewItem:"String (by ref)"=@n14)
  ]}
 Update List View {nodes [
   Update List View!() -> (then, Selected Plugin:String, Filter:String, Subfolder:String, Selected Asset:String) {is=title K2Node_FunctionEntry}
   > n1 = Sequence!(execute) -> (then_0, then_1)
   > ClearListItems(self=Get ListView())
   > Clear(TargetArray=Get Search Split String() -> (Search Split String:Array of Strings))
   > n2 = For Each Loop(Array=ParseIntoArray(SourceString=Get Filter() -> (Filter:String), Delimiter=" "))
   > Append Array!(execute, TargetArray=Get Search Split String(), SourceArray:Array of Strings=ParseIntoArray(SourceString=@n2, Delimiter=_)) -> (then)
   n3 = For Each Loop(Exec=@n1.then_1, Array:Array of Animatable Creature Object References=Get Animatable Props(Project=Get Current Project())) -> (Array Element:Animatable Creature Object Reference)
   SetSelectedIndex(execute=@n3.Completed, self:List View Object Reference=Get ListView(), Index=0)
   > n4 = Sequence()
   > Show(self=Get EUW_ContextMenu(), Show=false)
   n5 = For Each Loop(Exec=@n4.then_1, Array=Get Animatable Props(Project=Get Current Project()))
   Is Valid(exec=@n5.Completed, InputObject=Get AnimatableProp())
   > Set Attach Bone selection(Default Attach Bone=Get Default Attach Bone(self=Get AnimatableProp())) {is=title K2Node_CallFunction}
   n6 = Cast To Object_PropBrowser(
    execute=@n5.LoopBody
    Object=Get Selected Item!(self:List View Object Reference=Get ListView()) -> (ReturnValue:Object Reference) {is=title K2Node_CallFunction}
   )
   n7 = Get Data(self=@n6)
   > Branch(Condition="Equal (Object)"(A=Break Struct Prop Brower Item(Struct_PropBrower_Item=@n7), B=@n5))
   > n8 = Cast To AnimatableProp(Object=@n5)
   > Set AnimatableProp(AnimatableProp=@n8)
   n9 = Get Selected Asset()
   n10 = Branch(execute=@n3.LoopBody, Condition="Equal, Case Insensitive (String)"(A=@n9, B=All))
   n11 = Cast To AnimatableProp(execute=[@n10.then, @n25.then], Object=@n3)
   > n12 = Set Temp AP(Temp AP:Animatable Prop Object Reference=@n11) -> (Output_Get:Animatable Prop Object Reference)
   n13 = Get Subfolder()
   > n14 = SettingPropName in UpdateList!(execute, self, Target:Animatable Prop Object Reference=@n12, Subfolder:String=@n13) -> (then, Prop Asset1:String, ReturnValue:bool) {is=title K2Node_CallFunction}
   > Set Prop Asset(Prop Asset:String=@n14) -> (Output_Get:String)
   > Branch(Condition=OR Boolean!(A="Equal, Case Insensitive (String)"(A=@n13, B=All), B=@n14.ReturnValue) -> (ReturnValue))
   > n15 = Branch(Condition=IsEmpty(InString=Get Filter()))
   n16 = Construct Object from Class(execute=[@n15.then, @n20.then], Class=&/Object_PropBrowser.Object_PropBrowser_C, self=Self-Reference()) -> (ReturnValue:Object Prop Browser Object Reference)
   n17 = GetSoftObjectPath(Object=Get Temp AP() -> (Temp AP:Animatable Prop Object Reference))
   n18 = Get TypeCache() -> (TypeCache:Map of Strings to Strings)
   > Set Data!(
    execute
    Data:Struct Prop Brower Item Structure=Make Struct Prop Brower Item!(
     AP_53_#1:Animatable Prop Object Reference=Get Temp AP()
     Season_AssetType_57_#2:String=Get Selected Plugin()
     AP_Type_62_#3:String=Find(TargetMap=@n18, Key=@n17)
    ) -> (Struct_PropBrower_Item:Struct Prop Brower Item Structure) {is=title K2Node_MakeStruct}
    self:Object Prop Browser Object Reference=@n16
   ) -> (then, Output_Get:Struct Prop Brower Item Structure)
   > AddItem!(execute, self:List View Object Reference=Get ListView(), Item:Object Reference=@n16) -> (then)
   Set Search String Contains(execute=@n15.else, Search String Contains:bool=true) -> (Output_Get:bool)
   > n19 = For Each Loop(Array:Array of Strings=Get Search Split String()) -> (Array Element:String)
   n20 = Branch(execute=@n19.Completed, Condition=Get Search String Contains() -> (Search String Contains:bool))
   n21 = Branch(execute=@n19.LoopBody, Condition=Contains(SearchIn=Get Prop Asset() -> (Prop Asset:String), Substring=@n19))
   Set Search String Contains(execute=@n21.else, Search String Contains=false)
   Branch(execute=@n10.else, Condition="Equal, Case Insensitive (String)"(A=Get Asset(self:Animatable Creature Object Reference=@n3), B=@n9))
   n22 = Get Variant(self:Animatable Creature Object Reference=@n3)
   n23 = Get Selected Plugin()
   n24 = Get Selected Plugin()
   > n25 = Branch(Condition=OR Boolean(A="Equal Exactly (String)"(A=@n22, B=@n23), B="Equal Exactly (String)"(A=@n24, B=All)))
  ]}
 Validate Parent and AttachBone {nodes [
   Validate Parent and AttachBone() {is=title K2Node_FunctionEntry}
   > n1 = Branch(Condition=IsEmpty(InString=Get Selected Parent Asset()))
   > Show Message Dialog(
    Title=Get Title() -> (Title:Text)
    Message="To Text (String)"(InString="Cannot import prop — Parent Asset is not set. Please select a parent asset before continuing.")
    MessageType=Ok
    DefaultValue=Ok
    MessageCategory=Error
   )
   > Return Node!(execute, "Validation Succeeded ":bool=false)
   n2 = Branch(execute=@n1.else, Condition=IsValid(InBinding=Get Parent Rig Binding()))
   > n3 = Branch(Condition="Equal (Name)"(A=Get Attach Bone Name() -> (Attach Bone Name=None)))
   > Show Message Dialog(
    Title=Get Title()
    Message="To Text (String)"(InString=Attachment Point is not set. Please select one.)
    MessageType=Ok
    DefaultValue=Ok
    MessageCategory=Error
   )
   > Return Node("Validation Succeeded "=false)
   Return Node(execute=@n3.else, "Validation Succeeded "=true)
   Show Message Dialog(
    execute=@n2.else
    Title=Get Title()
    Message="To Text (String)"(InString=Parent Asset is no longer valid and has been refreshed. Please reselect.)
    MessageType=Ok
    DefaultValue=Ok
    MessageCategory=Error
   )
   > Populate Control Rig Combo Box(Level Sequence=Get Level Sequence(), Combo Box=Get ParentControlRigComboBox()) {is=title K2Node_CallFunction}
   > Return Node("Validation Succeeded "=false)
  ]}
]
variables [
 Animatable Creature Array {type="array<AnimatableCreature_C>", default []}
 AnimatableProp {type=AnimatableProp_C, default=null}
 Animation Dialog Widget {type=EUW_Animation_Dialog_C, default=null}
 Attach Bone Name {type=string, default=None}
 Attach CheckBox {type=boolean, default=true}
 Attach Point Array {type="array<string>", default []}
 Attach Target Name {type=string, default=""}
 bProjectCached {type=boolean}
 CachedProjects {type="array<BP_AnimationBrowser_Project_C>", default []}
 Control Rig {type=ControlRigAssetSoftReference, default {blueprintRigClass=null, controlRigAsset=null}}
 Control Rig Class {type=Class, default=null}
 control rig sequencer proxy {
  type=ControlRigSequencerBindingProxy
  default {controlRig=null, track=null, proxy {bindingId=#4, sequence=null}}
 }
 Control Rig Track {type=MovieSceneTrack, default=null}
 Current Project {type=BP_AnimationBrowser_Project_C, default=null}
 CurrentIndex {type=integer}
 Level Sequence {type=LevelSequence, default=null}
 Parent Attach Class {type=Class, default=null}
 Parent Rig Binding {type=MovieSceneBindingProxy, default {bindingId=#4, sequence=null}}
 Parent Skeletal Component {type=string, default=""}
 ParentCR_ComboBoxSelection {type=string, default=""}
 ParentRigs {type="array<ControlRig>", default []}
 Prop Vis Parent Channel {type=MovieSceneScriptingChannel, default=null}
 Save Game {type=SG_PropBrowser_C, default=null}
 SearchBarFilter {type=string, default=""}
 Selected Asset {type=string, default=""}
 Selected Parent Asset {type=string, default=""}
 Selected Parent Rig {type=ControlRig, default=null}
 Selected Plugin {type=string, default=""}
 Selected Project Name {type=string, default=""}
 Selected Subfolder {type=string, default=""}
 Setup Timer Handle {type=TimerHandle}
 Slot Name {type=string, default=PropBrowser}
 Spawned Binding for Asset {type=MovieSceneBindingProxy, default {bindingId=#4, sequence=null}}
 Target {type=SG_PropBrowser_C, default=null}
 TypeCache {type="map<string>", default {}}
]
widgets {
 info {rootWidgetClass=&CanvasPanel, widgetCount=84}
 widgets [
  CanvasPanel_23 {widgetClassPath=&CanvasPanel, widget=CanvasPanel_23}
  VerticalBox_27 {
   widgetClassPath=&VerticalBox
   parent=CanvasPanel_23
   widget=VerticalBox_27
   slot {
    class=CanvasPanelSlot
    properties {bAutoSize=true, layoutData {alignment=<0 0>, anchors {maximum=<1 1>, minimum=<0 0>}, offsets=<ltrb 0 2.3269 -6.4064 6.7267>}}
   }
  }
  SizeBox_72 {
   widgetClassPath=&SizeBox
   parent=VerticalBox_27
   widget=SizeBox_72
   properties {
    bOverride_MaxDesiredHeight=true
    bOverride_MaxDesiredWidth=true
    heightOverride=650
    maxAspectRatio=16
    maxDesiredHeight=1000
    maxDesiredWidth=700
    minDesiredHeight=500
    minDesiredWidth=300
    widthOverride=600
   }
   slot {class=VerticalBoxSlot, properties {verticalAlignment=VAlign_Top}}
  }
  Border_60 {
   widgetClassPath=&Border
   parent=SizeBox_72
   widget=Border_60
   properties {brushColor=<rgba 0.013 0.013 0.013 1>, padding=<ltrb 2 2 2 2>}
   slot {class=SizeBoxSlot}
  }
  EditorUtilityScrollBox_0 {
   widgetClassPath=&EditorUtilityScrollBox
   parent=Border_60
   widget=EditorUtilityScrollBox_0
   slot {class=BorderSlot, properties {padding=<ltrb 2 2 2 2>}}
  }
  VerticalBox_75 {widgetClassPath=&VerticalBox, parent=EditorUtilityScrollBox_0, widget=VerticalBox_75, slot {class=ScrollBoxSlot}}
  EditorUtilityScrollBox_535 {
   widgetClassPath=&EditorUtilityScrollBox
   parent=VerticalBox_75
   widget=EditorUtilityScrollBox_535
   slot {class=VerticalBoxSlot}
  }
  AP_Filter_HB {widgetClassPath=&HorizontalBox, parent=EditorUtilityScrollBox_535, widget=AP_Filter_HB, slot {class=ScrollBoxSlot}}
  AssetImage {
   widgetClassPath=&Image
   bIsVariable=true
   parent=AP_Filter_HB
   widget=AssetImage
   properties {brush {
     bIsDynamicallyLoaded=false
     drawAs=Image
     imageType=FullColor
     mirroring=NoMirror
     resourceName=None
     resourceObject=null
     tiling=NoTile
     imageSize=<25 25>
     margin=<ltrb 0 0 0 0>
     outlineSettings {
      bUseBrushTransparency=false
      roundingType=HalfHeightRadius
      width=0
      color {colorUseRule=UseColor_Specified, specifiedColor=<rgba 0 0 0 0>}
      cornerRadii=<0 0 0 0>
     }
     tintColor {colorUseRule=UseColor_Specified, specifiedColor=<rgba 1 1 1 1>}
     uVRegion {bIsValid=false, max=<0 0>, min=<0 0>}
    }}
   slot {class=HorizontalBoxSlot, properties {horizontalAlignment=HAlign_Center, verticalAlignment=VAlign_Center}}
  }
  SearchableAssetComboBox {
   widgetClassPath=/Script/FortniteEditor.SearchableStringComboBox
   bIsVariable=true
   parent=AP_Filter_HB
   widget=SearchableAssetComboBox
   properties {
    bOverride_Cursor=true
    selectedOption=All
    toolTipText=Category Dropdown: Corresponds to 'Asset' field in AP Files
    defaultOptions [All]
    font {
     bForceMonospaced=false
     bMaterialIsStencil=false
     fontMaterial=null
     fontObject=&Roboto
     letterSpacing=0
     monospacedWidth=1
     size=12
     skewAmount=0
     typefaceFontName=Bold
     outlineSettings {
      bApplyOutlineToDropShadows=false
      bMiteredCorners=false
      bSeparateFillAlpha=false
      outlineMaterial=null
      outlineSize=0
      outlineColor=<rgba 0 0 0 1>
     }
    }
    foregroundColor {colorUseRule=UseColor_Specified, specifiedColor=<rgba 0.5271 0.5271 0.5271 1>}
    itemStyle {
     bUseParentRowBrush=false
     activeBrush {
      bIsDynamicallyLoaded=false
      drawAs=Image
      imageType=NoImage
      mirroring=NoMirror
      resourceName=None
      resourceObject=null
      tiling=NoTile
      imageSize=<0 0>
      margin=<ltrb 0 0 0 0>
      outlineSettings {
       bUseBrushTransparency=false
       roundingType=HalfHeightRadius
       width=0
       color {colorUseRule=UseColor_Specified, specifiedColor=<rgba 0 0 0 0>}
       cornerRadii=<0 0 0 0>
      }
      tintColor {colorUseRule=UseColor_Specified, specifiedColor=<rgba 0 0.162 0.7454 1>}
      uVRegion {bIsValid=false, max=<0 0>, min=<0 0>}
     }
     activeHighlightedBrush {
      bIsDynamicallyLoaded=false
      drawAs=Image
      imageType=NoImage
      mirroring=NoMirror
      resourceName=None
      resourceObject=null
      tiling=NoTile
      imageSize=<0 0>
      margin=<ltrb 0 0 0 0>
      outlineSettings {
       bUseBrushTransparency=false
       roundingType=HalfHeightRadius
       width=0
       color {colorUseRule=UseColor_Specified, specifiedColor=<rgba 0 0 0 0>}
       cornerRadii=<0 0 0 0>
      }
      tintColor {colorUseRule=UseColor_Specified, specifiedColor=<rgba 0.0252 0.0319 0.0423 1>}
      uVRegion {bIsValid=false, max=<0 0>, min=<0 0>}
     }
     activeHoveredBrush {
      bIsDynamicallyLoaded=false
      drawAs=Image
      imageType=NoImage
      mirroring=NoMirror
      resourceName=None
      resourceObject=null
      tiling=NoTile
      imageSize=<0 0>
      margin=<ltrb 0 0 0 0>
      outlineSettings {
       bUseBrushTransparency=false
       roundingType=HalfHeightRadius
       width=0
       color {colorUseRule=UseColor_Specified, specifiedColor=<rgba 0 0 0 0>}
       cornerRadii=<0 0 0 0>
      }
      tintColor {colorUseRule=UseColor_Specified, specifiedColor=<rgba 0 0.162 0.7454 1>}
      uVRegion {bIsValid=false, max=<0 0>, min=<0 0>}
     }
     dropIndicator_Above {
      bIsDynamicallyLoaded=false
      drawAs=Box
      imageType=FullColor
      mirroring=NoMirror
      resourceName=../../../Engine/Content/Slate/Common/DropZoneIndicator_Above.png
      resourceObject=null
      tiling=NoTile
      imageSize=<0 0>
      margin=<ltrb 0.625 0.625 0 0>
      outlineSettings {
       bUseBrushTransparency=false
       roundingType=HalfHeightRadius
       width=0
       color {colorUseRule=UseColor_Specified, specifiedColor=<rgba 0 0 0 0>}
       cornerRadii=<0 0 0 0>
      }
      tintColor {colorUseRule=UseColor_Specified, specifiedColor=<rgba 0.0091 0.3515 0.9647 1>}
      uVRegion {bIsValid=false, max=<0 0>, min=<0 0>}
     }
     dropIndicator_Below {
      bIsDynamicallyLoaded=false
      drawAs=Box
      imageType=FullColor
      mirroring=NoMirror
      resourceName=../../../Engine/Content/Slate/Common/DropZoneIndicator_Below.png
      resourceObject=null
      tiling=NoTile
      imageSize=<0 0>
      margin=<ltrb 0.625 0 0 0.625>
      outlineSettings {
       bUseBrushTransparency=false
       roundingType=HalfHeightRadius
       width=0
       color {colorUseRule=UseColor_Specified, specifiedColor=<rgba 0 0 0 0>}
       cornerRadii=<0 0 0 0>
      }
      tintColor {colorUseRule=UseColor_Specified, specifiedColor=<rgba 0.0091 0.3515 0.9647 1>}
      uVRegion {bIsValid=false, max=<0 0>, min=<0 0>}
     }
     dropIndicator_Onto {
      bIsDynamicallyLoaded=false
      drawAs=Box
      imageType=FullColor
      mirroring=NoMirror
      resourceName=../../../Engine/Content/Slate/Common/DropZoneIndicator_Onto.png
      resourceObject=null
      tiling=NoTile
      imageSize=<0 0>
      margin=<ltrb 0.25 0.25 0.25 0.25>
      outlineSettings {
       bUseBrushTransparency=false
       roundingType=HalfHeightRadius
       width=0
       color {colorUseRule=UseColor_Specified, specifiedColor=<rgba 0 0 0 0>}
       cornerRadii=<0 0 0 0>
      }
      tintColor {colorUseRule=UseColor_Specified, specifiedColor=<rgba 0.0091 0.3515 0.9647 1>}
      uVRegion {bIsValid=false, max=<0 0>, min=<0 0>}
     }
     evenRowBackgroundBrush {
      bIsDynamicallyLoaded=false
      drawAs=Image
      imageType=NoImage
      mirroring=NoMirror
      resourceName=None
      resourceObject=null
      tiling=NoTile
      imageSize=<0 0>
      margin=<ltrb 0 0 0 0>
      outlineSettings {
       bUseBrushTransparency=false
       roundingType=HalfHeightRadius
       width=0
       color {colorUseRule=UseColor_Specified, specifiedColor=<rgba 0 0 0 0>}
       cornerRadii=<0 0 0 0>
      }
      tintColor {colorUseRule=UseColor_Specified, specifiedColor=<rgba 0.0103 0.0103 0.0103 1>}
      uVRegion {bIsValid=false, max=<0 0>, min=<0 0>}
     }
     evenRowBackgroundHoveredBrush {
      bIsDynamicallyLoaded=false
      drawAs=Image
      imageType=NoImage
      mirroring=NoMirror
      resourceName=None
      resourceObject=null
      tiling=NoTile
      imageSize=<0 0>
      margin=<ltrb 0 0 0 0>
      outlineSettings {
       bUseBrushTransparency=false
       roundingType=HalfHeightRadius
       width=0
       color {colorUseRule=UseColor_Specified, specifiedColor=<rgba 0 0 0 0>}
       cornerRadii=<0 0 0 0>
      }
      tintColor {colorUseRule=UseColor_Specified, specifiedColor=<rgba 0.0176 0.0176 0.0176 1>}
      uVRegion {bIsValid=false, max=<0 0>, min=<0 0>}
     }
     inactiveBrush {
      bIsDynamicallyLoaded=false
      drawAs=Image
      imageType=NoImage
      mirroring=NoMirror
      resourceName=None
      resourceObject=null
      tiling=NoTile
      imageSize=<0 0>
      margin=<ltrb 0 0 0 0>
      outlineSettings {
       bUseBrushTransparency=false
       roundingType=HalfHeightRadius
       width=0
       color {colorUseRule=UseColor_Specified, specifiedColor=<rgba 0 0 0 0>}
       cornerRadii=<0 0 0 0>
      }
      tintColor {colorUseRule=UseColor_Specified, specifiedColor=<rgba 0.0513 0.0953 0.159 1>}
      uVRegion {bIsValid=false, max=<0 0>, min=<0 0>}
     }
     inactiveHighlightedBrush {
      bIsDynamicallyLoaded=false
      drawAs=Image
      imageType=NoImage
      mirroring=NoMirror
      resourceName=None
      resourceObject=null
      tiling=NoTile
      imageSize=<0 0>
      margin=<ltrb 0 0 0 0>
      outlineSettings {
       bUseBrushTransparency=false
       roundingType=HalfHeightRadius
       width=0
       color {colorUseRule=UseColor_Specified, specifiedColor=<rgba 0 0 0 0>}
       cornerRadii=<0 0 0 0>
      }
      tintColor {colorUseRule=UseColor_Specified, specifiedColor=<rgba 0.0252 0.0319 0.0423 1>}
      uVRegion {bIsValid=false, max=<0 0>, min=<0 0>}
     }
     inactiveHoveredBrush {
      bIsDynamicallyLoaded=false
      drawAs=Image
      imageType=NoImage
      mirroring=NoMirror
      resourceName=None
      resourceObject=null
      tiling=NoTile
      imageSize=<0 0>
      margin=<ltrb 0 0 0 0>
      outlineSettings {
       bUseBrushTransparency=false
       roundingType=HalfHeightRadius
       width=0
       color {colorUseRule=UseColor_Specified, specifiedColor=<rgba 0 0 0 0>}
       cornerRadii=<0 0 0 0>
      }
      tintColor {colorUseRule=UseColor_Specified, specifiedColor=<rgba 0.0513 0.0953 0.159 1>}
      uVRegion {bIsValid=false, max=<0 0>, min=<0 0>}
     }
     oddRowBackgroundBrush {
      bIsDynamicallyLoaded=false
      drawAs=Image
      imageType=NoImage
      mirroring=NoMirror
      resourceName=None
      resourceObject=null
      tiling=NoTile
      imageSize=<0 0>
      margin=<ltrb 0 0 0 0>
      outlineSettings {
       bUseBrushTransparency=false
       roundingType=HalfHeightRadius
       width=0
       color {colorUseRule=UseColor_Specified, specifiedColor=<rgba 0 0 0 0>}
       cornerRadii=<0 0 0 0>
      }
      tintColor {colorUseRule=UseColor_Specified, specifiedColor=<rgba 0.0103 0.0103 0.0103 1>}
      uVRegion {bIsValid=false, max=<0 0>, min=<0 0>}
     }
     oddRowBackgroundHoveredBrush {
      bIsDynamicallyLoaded=false
      drawAs=Image
      imageType=NoImage
      mirroring=NoMirror
      resourceName=None
      resourceObject=null
      tiling=NoTile
      imageSize=<0 0>
      margin=<ltrb 0 0 0 0>
      outlineSettings {
       bUseBrushTransparency=false
       roundingType=HalfHeightRadius
       width=0
       color {colorUseRule=UseColor_Specified, specifiedColor=<rgba 0 0 0 0>}
       cornerRadii=<0 0 0 0>
      }
      tintColor {colorUseRule=UseColor_Specified, specifiedColor=<rgba 0.0176 0.0176 0.0176 1>}
      uVRegion {bIsValid=false, max=<0 0>, min=<0 0>}
     }
     parentRowBackgroundBrush {
      bIsDynamicallyLoaded=false
      drawAs=Image
      imageType=NoImage
      mirroring=NoMirror
      resourceName=None
      resourceObject=null
      tiling=NoTile
      imageSize=<32 32>
      margin=<ltrb 0 0 0 0>
      outlineSettings {
       bUseBrushTransparency=false
       roundingType=HalfHeightRadius
       width=0
       color {colorUseRule=UseColor_Specified, specifiedColor=<rgba 0 0 0 0>}
       cornerRadii=<0 0 0 0>
      }
      tintColor {colorUseRule=UseColor_Specified, specifiedColor=<rgba 1 1 1 1>}
      uVRegion {bIsValid=false, max=<0 0>, min=<0 0>}
     }
     parentRowBackgroundHoveredBrush {
      bIsDynamicallyLoaded=false
      drawAs=Image
      imageType=NoImage
      mirroring=NoMirror
      resourceName=None
      resourceObject=null
      tiling=NoTile
      imageSize=<32 32>
      margin=<ltrb 0 0 0 0>
      outlineSettings {
       bUseBrushTransparency=false
       roundingType=HalfHeightRadius
       width=0
       color {colorUseRule=UseColor_Specified, specifiedColor=<rgba 0 0 0 0>}
       cornerRadii=<0 0 0 0>
      }
      tintColor {colorUseRule=UseColor_Specified, specifiedColor=<rgba 1 1 1 1>}
      uVRegion {bIsValid=false, max=<0 0>, min=<0 0>}
     }
     selectedTextColor {colorUseRule=UseColor_Specified, specifiedColor=<rgba 1 1 1 1>}
     selectorFocusedBrush {
      bIsDynamicallyLoaded=false
      drawAs=Border
      imageType=FullColor
      mirroring=NoMirror
      resourceName=../../../Engine/Content/Slate/Common/Selector.png
      resourceObject=null
      tiling=Both
      imageSize=<0 0>
      margin=<ltrb 0.25 0.25 0.25 0.25>
      outlineSettings {
       bUseBrushTransparency=false
       roundingType=HalfHeightRadius
       width=0
       color {colorUseRule=UseColor_Specified, specifiedColor=<rgba 0 0 0 0>}
       cornerRadii=<0 0 0 0>
      }
      tintColor {colorUseRule=UseColor_Specified, specifiedColor=<rgba 0.701 0.225 0.003 1>}
      uVRegion {bIsValid=false, max=<0 0>, min=<0 0>}
     }
     textColor {colorUseRule=UseColor_Specified, specifiedColor=<rgba 0.5271 0.5271 0.5271 1>}
    }
    widgetStyle {
     comboButtonStyle {
      downArrowAlign=VAlign_Center
      buttonStyle {
       clickedSlateSound {resourceObject=null}
       disabled {
        bIsDynamicallyLoaded=false
        drawAs=RoundedBox
        imageType=NoImage
        mirroring=NoMirror
        resourceName=None
        resourceObject=null
        tiling=NoTile
        imageSize=<0 0>
        margin=<ltrb 0 0 0 0>
        outlineSettings {
         bUseBrushTransparency=false
         roundingType=FixedRadius
         width=1
         color {colorUseRule=UseColor_ColorTable, specifiedColor=<rgba 0 0 0 0>}
         cornerRadii=<4 4 4 4>
        }
        tintColor {colorUseRule=UseColor_Specified, specifiedColor=<rgba 0.0395 0.0395 0.0395 1>}
        uVRegion {bIsValid=false, max=<0 0>, min=<0 0>}
       }
       disabledForeground {colorUseRule=UseColor_ColorTable, specifiedColor=<rgba 0 0 0 0>}
       hovered {
        bIsDynamicallyLoaded=false
        drawAs=RoundedBox
        imageType=NoImage
        mirroring=NoMirror
        resourceName=None
        resourceObject=null
        tiling=NoTile
        imageSize=<0 0>
        margin=<ltrb 0 0 0 0>
        outlineSettings {
         bUseBrushTransparency=false
         roundingType=FixedRadius
         width=1
         color {colorUseRule=UseColor_ColorTable, specifiedColor=<rgba 0 0 0 0>}
         cornerRadii=<4 4 4 4>
        }
        tintColor {colorUseRule=UseColor_Specified, specifiedColor=<rgba 0.0048 0.0048 0.0048 1>}
        uVRegion {bIsValid=false, max=<0 0>, min=<0 0>}
       }
       hoveredForeground {colorUseRule=UseColor_ColorTable, specifiedColor=<rgba 0 0 0 0>}
       hoveredSlateSound {resourceObject=null}
       normal {
        bIsDynamicallyLoaded=false
        drawAs=RoundedBox
        imageType=NoImage
        mirroring=NoMirror
        resourceName=None
        resourceObject=null
        tiling=NoTile
        imageSize=<0 0>
        margin=<ltrb 0 0 0 0>
        outlineSettings {
         bUseBrushTransparency=false
         roundingType=FixedRadius
         width=1
         color {colorUseRule=UseColor_ColorTable, specifiedColor=<rgba 0 0 0 0>}
         cornerRadii=<4 4 4 4>
        }
        tintColor {colorUseRule=UseColor_Specified, specifiedColor=<rgba 0.0048 0.0048 0.0048 1>}
        uVRegion {bIsValid=false, max=<0 0>, min=<0 0>}
       }
       normalForeground {colorUseRule=UseColor_ColorTable, specifiedColor=<rgba 0 0 0 0>}
       normalPadding=<ltrb 8 1 8 1>
       pressed {
        bIsDynamicallyLoaded=false
        drawAs=RoundedBox
        imageType=NoImage
        mirroring=NoMirror
        resourceName=None
        resourceObject=null
        tiling=NoTile
        imageSize=<0 0>
        margin=<ltrb 0 0 0 0>
        outlineSettings {
         bUseBrushTransparency=false
         roundingType=FixedRadius
         width=1
         color {colorUseRule=UseColor_ColorTable, specifiedColor=<rgba 0 0 0 0>}
         cornerRadii=<4 4 4 4>
        }
        tintColor {colorUseRule=UseColor_Specified, specifiedColor=<rgba 0.0048 0.0048 0.0048 1>}
        uVRegion {bIsValid=false, max=<0 0>, min=<0 0>}
       }
       pressedForeground {colorUseRule=UseColor_ColorTable, specifiedColor=<rgba 0 0 0 0>}
       pressedPadding=<ltrb 8 2 8 0>
       pressedSlateSound {resourceObject=null}
      }
      contentPadding=<ltrb 0 0 0 0>
      downArrowImage {
       bIsDynamicallyLoaded=false
       drawAs=Image
       imageType=Vector
       mirroring=NoMirror
       resourceName=../../../Engine/Content/Slate/Starship/CoreWidgets/ComboBox/wide-chevron-down.svg
       resourceObject=null
       tiling=NoTile
       imageSize=<16 16>
       margin=<ltrb 0 0 0 0>
       outlineSettings {
        bUseBrushTransparency=false
        roundingType=HalfHeightRadius
        width=0
        color {colorUseRule=UseColor_Specified, specifiedColor=<rgba 0 0 0 0>}
        cornerRadii=<0 0 0 0>
       }
       tintColor {colorUseRule=UseColor_Specified, specifiedColor=<rgba 1 1 1 1>}
       uVRegion {bIsValid=false, max=<0 0>, min=<0 0>}
      }
      downArrowPadding=<ltrb 2 2 2 2>
      menuBorderBrush {
       bIsDynamicallyLoaded=false
       drawAs=RoundedBox
       imageType=NoImage
       mirroring=NoMirror
       resourceName=None
       resourceObject=null
       tiling=NoTile
       imageSize=<0 0>
       margin=<ltrb 0 0 0 0>
       outlineSettings {
        bUseBrushTransparency=false
        roundingType=FixedRadius
        width=1
        color {colorUseRule=UseColor_ColorTable, specifiedColor=<rgba 0 0 0 0>}
        cornerRadii=<0 0 0 0>
       }
       tintColor {colorUseRule=UseColor_Specified, specifiedColor=<rgba 0.0395 0.0395 0.0395 1>}
       uVRegion {bIsValid=false, max=<0 0>, min=<0 0>}
      }
      menuBorderPadding=<ltrb 0 0 0 0>
      shadowColorAndOpacity=<rgba 0 0 0 1>
      shadowOffset=<0 0>
     }
     contentPadding=<ltrb 0 0 0 0>
     menuRowPadding=<ltrb 8 3 8 3>
     pressedSlateSound {resourceObject=null}
     selectionChangeSlateSound {resourceObject=null}
    }
   }
   slot {class=HorizontalBoxSlot, properties {size {sizeRule=Fill, value=1}}}
  }
  PluginImage {
   widgetClassPath=&Image
   bIsVariable=true
   parent=AP_Filter_HB
   widget=PluginImage
   properties {brush {
     bIsDynamicallyLoaded=false
     drawAs=Image
     imageType=FullColor
     mirroring=NoMirror
     resourceName=None
     resourceObject=/AnimInEdTools/AnimationBrowser/Icons/Icon_Plugin
     tiling=NoTile
     imageSize=<18 28>
     margin=<ltrb 0 0 0 0>
     outlineSettings {
      bUseBrushTransparency=false
      roundingType=HalfHeightRadius
      width=0
      color {colorUseRule=UseColor_Specified, specifiedColor=<rgba 0 0 0 0>}
      cornerRadii=<0 0 0 0>
     }
     tintColor {colorUseRule=UseColor_Specified, specifiedColor=<rgba 1 1 1 1>}
     uVRegion {bIsValid=false, max=<0 0>, min=<0 0>}
    }}
   slot {class=HorizontalBoxSlot}
  }
  SearchablePluginComboBox {
   widgetClassPath=&SearchableStringComboBox
   bIsVariable=true
   parent=AP_Filter_HB
   widget=SearchablePluginComboBox
   properties {
    bOverride_Cursor=true
    selectedOption=All
    toolTipText=Category Dropdown: Corresponds to 'Variant' field in AP Files
    defaultOptions [All]
    font {
     bForceMonospaced=false
     bMaterialIsStencil=false
     fontMaterial=null
     fontObject=&Roboto
     letterSpacing=0
     monospacedWidth=1
     size=12
     skewAmount=0
     typefaceFontName=Bold
     outlineSettings {
      bApplyOutlineToDropShadows=false
      bMiteredCorners=false
      bSeparateFillAlpha=false
      outlineMaterial=null
      outlineSize=0
      outlineColor=<rgba 0 0 0 1>
     }
    }
    foregroundColor {colorUseRule=UseColor_Specified, specifiedColor=<rgba 0.5271 0.5271 0.5271 1>}
    itemStyle {
     bUseParentRowBrush=false
     activeBrush {
      bIsDynamicallyLoaded=false
      drawAs=Image
      imageType=NoImage
      mirroring=NoMirror
      resourceName=None
      resourceObject=null
      tiling=NoTile
      imageSize=<0 0>
      margin=<ltrb 0 0 0 0>
      outlineSettings {
       bUseBrushTransparency=false
       roundingType=HalfHeightRadius
       width=0
       color {colorUseRule=UseColor_Specified, specifiedColor=<rgba 0 0 0 0>}
       cornerRadii=<0 0 0 0>
      }
      tintColor {colorUseRule=UseColor_Specified, specifiedColor=<rgba 0 0.162 0.7454 1>}
      uVRegion {bIsValid=false, max=<0 0>, min=<0 0>}
     }
     activeHighlightedBrush {
      bIsDynamicallyLoaded=false
      drawAs=Image
      imageType=NoImage
      mirroring=NoMirror
      resourceName=None
      resourceObject=null
      tiling=NoTile
      imageSize=<0 0>
      margin=<ltrb 0 0 0 0>
      outlineSettings {
       bUseBrushTransparency=false
       roundingType=HalfHeightRadius
       width=0
       color {colorUseRule=UseColor_Specified, specifiedColor=<rgba 0 0 0 0>}
       cornerRadii=<0 0 0 0>
      }
      tintColor {colorUseRule=UseColor_Specified, specifiedColor=<rgba 0.0252 0.0319 0.0423 1>}
      uVRegion {bIsValid=false, max=<0 0>, min=<0 0>}
     }
     activeHoveredBrush {
      bIsDynamicallyLoaded=false
      drawAs=Image
      imageType=NoImage
      mirroring=NoMirror
      resourceName=None
      resourceObject=null
      tiling=NoTile
      imageSize=<0 0>
      margin=<ltrb 0 0 0 0>
      outlineSettings {
       bUseBrushTransparency=false
       roundingType=HalfHeightRadius
       width=0
       color {colorUseRule=UseColor_Specified, specifiedColor=<rgba 0 0 0 0>}
       cornerRadii=<0 0 0 0>
      }
      tintColor {colorUseRule=UseColor_Specified, specifiedColor=<rgba 0 0.162 0.7454 1>}
      uVRegion {bIsValid=false, max=<0 0>, min=<0 0>}
     }
     dropIndicator_Above {
      bIsDynamicallyLoaded=false
      drawAs=Box
      imageType=FullColor
      mirroring=NoMirror
      resourceName=../../../Engine/Content/Slate/Common/DropZoneIndicator_Above.png
      resourceObject=null
      tiling=NoTile
      imageSize=<0 0>
      margin=<ltrb 0.625 0.625 0 0>
      outlineSettings {
       bUseBrushTransparency=false
       roundingType=HalfHeightRadius
       width=0
       color {colorUseRule=UseColor_Specified, specifiedColor=<rgba 0 0 0 0>}
       cornerRadii=<0 0 0 0>
      }
      tintColor {colorUseRule=UseColor_Specified, specifiedColor=<rgba 0.0091 0.3515 0.9647 1>}
      uVRegion {bIsValid=false, max=<0 0>, min=<0 0>}
     }
     dropIndicator_Below {
      bIsDynamicallyLoaded=false
      drawAs=Box
      imageType=FullColor
      mirroring=NoMirror
      resourceName=../../../Engine/Content/Slate/Common/DropZoneIndicator_Below.png
      resourceObject=null
      tiling=NoTile
      imageSize=<0 0>
      margin=<ltrb 0.625 0 0 0.625>
      outlineSettings {
       bUseBrushTransparency=false
       roundingType=HalfHeightRadius
       width=0
       color {colorUseRule=UseColor_Specified, specifiedColor=<rgba 0 0 0 0>}
       cornerRadii=<0 0 0 0>
      }
      tintColor {colorUseRule=UseColor_Specified, specifiedColor=<rgba 0.0091 0.3515 0.9647 1>}
      uVRegion {bIsValid=false, max=<0 0>, min=<0 0>}
     }
     dropIndicator_Onto {
      bIsDynamicallyLoaded=false
      drawAs=Box
      imageType=FullColor
      mirroring=NoMirror
      resourceName=../../../Engine/Content/Slate/Common/DropZoneIndicator_Onto.png
      resourceObject=null
      tiling=NoTile
      imageSize=<0 0>
      margin=<ltrb 0.25 0.25 0.25 0.25>
      outlineSettings {
       bUseBrushTransparency=false
       roundingType=HalfHeightRadius
       width=0
       color {colorUseRule=UseColor_Specified, specifiedColor=<rgba 0 0 0 0>}
       cornerRadii=<0 0 0 0>
      }
      tintColor {colorUseRule=UseColor_Specified, specifiedColor=<rgba 0.0091 0.3515 0.9647 1>}
      uVRegion {bIsValid=false, max=<0 0>, min=<0 0>}
     }
     evenRowBackgroundBrush {
      bIsDynamicallyLoaded=false
      drawAs=Image
      imageType=NoImage
      mirroring=NoMirror
      resourceName=None
      resourceObject=null
      tiling=NoTile
      imageSize=<0 0>
      margin=<ltrb 0 0 0 0>
      outlineSettings {
       bUseBrushTransparency=false
       roundingType=HalfHeightRadius
       width=0
       color {colorUseRule=UseColor_Specified, specifiedColor=<rgba 0 0 0 0>}
       cornerRadii=<0 0 0 0>
      }
      tintColor {colorUseRule=UseColor_Specified, specifiedColor=<rgba 0.0103 0.0103 0.0103 1>}
      uVRegion {bIsValid=false, max=<0 0>, min=<0 0>}
     }
     evenRowBackgroundHoveredBrush {
      bIsDynamicallyLoaded=false
      drawAs=Image
      imageType=NoImage
      mirroring=NoMirror
      resourceName=None
      resourceObject=null
      tiling=NoTile
      imageSize=<0 0>
      margin=<ltrb 0 0 0 0>
      outlineSettings {
       bUseBrushTransparency=false
       roundingType=HalfHeightRadius
       width=0
       color {colorUseRule=UseColor_Specified, specifiedColor=<rgba 0 0 0 0>}
       cornerRadii=<0 0 0 0>
      }
      tintColor {colorUseRule=UseColor_Specified, specifiedColor=<rgba 0.0176 0.0176 0.0176 1>}
      uVRegion {bIsValid=false, max=<0 0>, min=<0 0>}
     }
     inactiveBrush {
      bIsDynamicallyLoaded=false
      drawAs=Image
      imageType=NoImage
      mirroring=NoMirror
      resourceName=None
      resourceObject=null
      tiling=NoTile
      imageSize=<0 0>
      margin=<ltrb 0 0 0 0>
      outlineSettings {
       bUseBrushTransparency=false
       roundingType=HalfHeightRadius
       width=0
       color {colorUseRule=UseColor_Specified, specifiedColor=<rgba 0 0 0 0>}
       cornerRadii=<0 0 0 0>
      }
      tintColor {colorUseRule=UseColor_Specified, specifiedColor=<rgba 0.0513 0.0953 0.159 1>}
      uVRegion {bIsValid=false, max=<0 0>, min=<0 0>}
     }
     inactiveHighlightedBrush {
      bIsDynamicallyLoaded=false
      drawAs=Image
      imageType=NoImage
      mirroring=NoMirror
      resourceName=None
      resourceObject=null
      tiling=NoTile
      imageSize=<0 0>
      margin=<ltrb 0 0 0 0>
      outlineSettings {
       bUseBrushTransparency=false
       roundingType=HalfHeightRadius
       width=0
       color {colorUseRule=UseColor_Specified, specifiedColor=<rgba 0 0 0 0>}
       cornerRadii=<0 0 0 0>
      }
      tintColor {colorUseRule=UseColor_Specified, specifiedColor=<rgba 0.0252 0.0319 0.0423 1>}
      uVRegion {bIsValid=false, max=<0 0>, min=<0 0>}
     }
     inactiveHoveredBrush {
      bIsDynamicallyLoaded=false
      drawAs=Image
      imageType=NoImage
      mirroring=NoMirror
      resourceName=None
      resourceObject=null
      tiling=NoTile
      imageSize=<0 0>
      margin=<ltrb 0 0 0 0>
      outlineSettings {
       bUseBrushTransparency=false
       roundingType=HalfHeightRadius
       width=0
       color {colorUseRule=UseColor_Specified, specifiedColor=<rgba 0 0 0 0>}
       cornerRadii=<0 0 0 0>
      }
      tintColor {colorUseRule=UseColor_Specified, specifiedColor=<rgba 0.0513 0.0953 0.159 1>}
      uVRegion {bIsValid=false, max=<0 0>, min=<0 0>}
     }
     oddRowBackgroundBrush {
      bIsDynamicallyLoaded=false
      drawAs=Image
      imageType=NoImage
      mirroring=NoMirror
      resourceName=None
      resourceObject=null
      tiling=NoTile
      imageSize=<0 0>
      margin=<ltrb 0 0 0 0>
      outlineSettings {
       bUseBrushTransparency=false
       roundingType=HalfHeightRadius
       width=0
       color {colorUseRule=UseColor_Specified, specifiedColor=<rgba 0 0 0 0>}
       cornerRadii=<0 0 0 0>
      }
      tintColor {colorUseRule=UseColor_Specified, specifiedColor=<rgba 0.0103 0.0103 0.0103 1>}
      uVRegion {bIsValid=false, max=<0 0>, min=<0 0>}
     }
     oddRowBackgroundHoveredBrush {
      bIsDynamicallyLoaded=false
      drawAs=Image
      imageType=NoImage
      mirroring=NoMirror
      resourceName=None
      resourceObject=null
      tiling=NoTile
      imageSize=<0 0>
      margin=<ltrb 0 0 0 0>
      outlineSettings {
       bUseBrushTransparency=false
       roundingType=HalfHeightRadius
       width=0
       color {colorUseRule=UseColor_Specified, specifiedColor=<rgba 0 0 0 0>}
       cornerRadii=<0 0 0 0>
      }
      tintColor {colorUseRule=UseColor_Specified, specifiedColor=<rgba 0.0176 0.0176 0.0176 1>}
      uVRegion {bIsValid=false, max=<0 0>, min=<0 0>}
     }
     parentRowBackgroundBrush {
      bIsDynamicallyLoaded=false
      drawAs=Image
      imageType=NoImage
      mirroring=NoMirror
      resourceName=None
      resourceObject=null
      tiling=NoTile
      imageSize=<32 32>
      margin=<ltrb 0 0 0 0>
      outlineSettings {
       bUseBrushTransparency=false
       roundingType=HalfHeightRadius
       width=0
       color {colorUseRule=UseColor_Specified, specifiedColor=<rgba 0 0 0 0>}
       cornerRadii=<0 0 0 0>
      }
      tintColor {colorUseRule=UseColor_Specified, specifiedColor=<rgba 1 1 1 1>}
      uVRegion {bIsValid=false, max=<0 0>, min=<0 0>}
     }
     parentRowBackgroundHoveredBrush {
      bIsDynamicallyLoaded=false
      drawAs=Image
      imageType=NoImage
      mirroring=NoMirror
      resourceName=None
      resourceObject=null
      tiling=NoTile
      imageSize=<32 32>
      margin=<ltrb 0 0 0 0>
      outlineSettings {
       bUseBrushTransparency=false
       roundingType=HalfHeightRadius
       width=0
       color {colorUseRule=UseColor_Specified, specifiedColor=<rgba 0 0 0 0>}
       cornerRadii=<0 0 0 0>
      }
      tintColor {colorUseRule=UseColor_Specified, specifiedColor=<rgba 1 1 1 1>}
      uVRegion {bIsValid=false, max=<0 0>, min=<0 0>}
     }
     selectedTextColor {colorUseRule=UseColor_Specified, specifiedColor=<rgba 1 1 1 1>}
     selectorFocusedBrush {
      bIsDynamicallyLoaded=false
      drawAs=Border
      imageType=FullColor
      mirroring=NoMirror
      resourceName=../../../Engine/Content/Slate/Common/Selector.png
      resourceObject=null
      tiling=Both
      imageSize=<0 0>
      margin=<ltrb 0.25 0.25 0.25 0.25>
      outlineSettings {
       bUseBrushTransparency=false
       roundingType=HalfHeightRadius
       width=0
       color {colorUseRule=UseColor_Specified, specifiedColor=<rgba 0 0 0 0>}
       cornerRadii=<0 0 0 0>
      }
      tintColor {colorUseRule=UseColor_Specified, specifiedColor=<rgba 0.701 0.225 0.003 1>}
      uVRegion {bIsValid=false, max=<0 0>, min=<0 0>}
     }
     textColor {colorUseRule=UseColor_Specified, specifiedColor=<rgba 0.5271 0.5271 0.5271 1>}
    }
    widgetStyle {
     comboButtonStyle {
      downArrowAlign=VAlign_Center
      buttonStyle {
       clickedSlateSound {resourceObject=null}
       disabled {
        bIsDynamicallyLoaded=false
        drawAs=RoundedBox
        imageType=NoImage
        mirroring=NoMirror
        resourceName=None
        resourceObject=null
        tiling=NoTile
        imageSize=<0 0>
        margin=<ltrb 0 0 0 0>
        outlineSettings {
         bUseBrushTransparency=false
         roundingType=FixedRadius
         width=1
         color {colorUseRule=UseColor_ColorTable, specifiedColor=<rgba 0 0 0 0>}
         cornerRadii=<4 4 4 4>
        }
        tintColor {colorUseRule=UseColor_Specified, specifiedColor=<rgba 0.0395 0.0395 0.0395 1>}
        uVRegion {bIsValid=false, max=<0 0>, min=<0 0>}
       }
       disabledForeground {colorUseRule=UseColor_ColorTable, specifiedColor=<rgba 0 0 0 0>}
       hovered {
        bIsDynamicallyLoaded=false
        drawAs=RoundedBox
        imageType=NoImage
        mirroring=NoMirror
        resourceName=None
        resourceObject=null
        tiling=NoTile
        imageSize=<0 0>
        margin=<ltrb 0 0 0 0>
        outlineSettings {
         bUseBrushTransparency=false
         roundingType=FixedRadius
         width=1
         color {colorUseRule=UseColor_ColorTable, specifiedColor=<rgba 0 0 0 0>}
         cornerRadii=<4 4 4 4>
        }
        tintColor {colorUseRule=UseColor_Specified, specifiedColor=<rgba 0.0048 0.0048 0.0048 1>}
        uVRegion {bIsValid=false, max=<0 0>, min=<0 0>}
       }
       hoveredForeground {colorUseRule=UseColor_ColorTable, specifiedColor=<rgba 0 0 0 0>}
       hoveredSlateSound {resourceObject=null}
       normal {
        bIsDynamicallyLoaded=false
        drawAs=RoundedBox
        imageType=NoImage
        mirroring=NoMirror
        resourceName=None
        resourceObject=null
        tiling=NoTile
        imageSize=<0 0>
        margin=<ltrb 0 0 0 0>
        outlineSettings {
         bUseBrushTransparency=false
         roundingType=FixedRadius
         width=1
         color {colorUseRule=UseColor_ColorTable, specifiedColor=<rgba 0 0 0 0>}
         cornerRadii=<4 4 4 4>
        }
        tintColor {colorUseRule=UseColor_Specified, specifiedColor=<rgba 0.0048 0.0048 0.0048 1>}
        uVRegion {bIsValid=false, max=<0 0>, min=<0 0>}
       }
       normalForeground {colorUseRule=UseColor_ColorTable, specifiedColor=<rgba 0 0 0 0>}
       normalPadding=<ltrb 8 1 8 1>
       pressed {
        bIsDynamicallyLoaded=false
        drawAs=RoundedBox
        imageType=NoImage
        mirroring=NoMirror
        resourceName=None
        resourceObject=null
        tiling=NoTile
        imageSize=<0 0>
        margin=<ltrb 0 0 0 0>
        outlineSettings {
         bUseBrushTransparency=false
         roundingType=FixedRadius
         width=1
         color {colorUseRule=UseColor_ColorTable, specifiedColor=<rgba 0 0 0 0>}
         cornerRadii=<4 4 4 4>
        }
        tintColor {colorUseRule=UseColor_Specified, specifiedColor=<rgba 0.0048 0.0048 0.0048 1>}
        uVRegion {bIsValid=false, max=<0 0>, min=<0 0>}
       }
       pressedForeground {colorUseRule=UseColor_ColorTable, specifiedColor=<rgba 0 0 0 0>}
       pressedPadding=<ltrb 8 2 8 0>
       pressedSlateSound {resourceObject=null}
      }
      contentPadding=<ltrb 0 0 0 0>
      downArrowImage {
       bIsDynamicallyLoaded=false
       drawAs=Image
       imageType=Vector
       mirroring=NoMirror
       resourceName=../../../Engine/Content/Slate/Starship/CoreWidgets/ComboBox/wide-chevron-down.svg
       resourceObject=null
       tiling=NoTile
       imageSize=<16 16>
       margin=<ltrb 0 0 0 0>
       outlineSettings {
        bUseBrushTransparency=false
        roundingType=HalfHeightRadius
        width=0
        color {colorUseRule=UseColor_Specified, specifiedColor=<rgba 0 0 0 0>}
        cornerRadii=<0 0 0 0>
       }
       tintColor {colorUseRule=UseColor_Specified, specifiedColor=<rgba 1 1 1 1>}
       uVRegion {bIsValid=false, max=<0 0>, min=<0 0>}
      }
      downArrowPadding=<ltrb 2 2 2 2>
      menuBorderBrush {
       bIsDynamicallyLoaded=false
       drawAs=RoundedBox
       imageType=NoImage
       mirroring=NoMirror
       resourceName=None
       resourceObject=null
       tiling=NoTile
       imageSize=<0 0>
       margin=<ltrb 0 0 0 0>
       outlineSettings {
        bUseBrushTransparency=false
        roundingType=FixedRadius
        width=1
        color {colorUseRule=UseColor_ColorTable, specifiedColor=<rgba 0 0 0 0>}
        cornerRadii=<0 0 0 0>
       }
       tintColor {colorUseRule=UseColor_Specified, specifiedColor=<rgba 0.0395 0.0395 0.0395 1>}
       uVRegion {bIsValid=false, max=<0 0>, min=<0 0>}
      }
      menuBorderPadding=<ltrb 0 0 0 0>
      shadowColorAndOpacity=<rgba 0 0 0 1>
      shadowOffset=<0 0>
     }
     contentPadding=<ltrb 0 0 0 0>
     menuRowPadding=<ltrb 8 3 8 3>
     pressedSlateSound {resourceObject=null}
     selectionChangeSlateSound {resourceObject=null}
    }
   }
   slot {class=HorizontalBoxSlot, properties {size {sizeRule=Fill, value=1}}}
  }
  Subfolder {
   widgetClassPath=&Image
   bIsVariable=true
   parent=AP_Filter_HB
   widget=Subfolder
   properties {brush {
     bIsDynamicallyLoaded=false
     drawAs=Image
     imageType=FullColor
     mirroring=NoMirror
     resourceName=None
     resourceObject=/AnimInEdTools/AnimationBrowser/Icons/Icon_Categories
     tiling=NoTile
     imageSize=<18 28>
     margin=<ltrb 0 0 0 0>
     outlineSettings {
      bUseBrushTransparency=false
      roundingType=HalfHeightRadius
      width=0
      color {colorUseRule=UseColor_Specified, specifiedColor=<rgba 0 0 0 0>}
      cornerRadii=<0 0 0 0>
     }
     tintColor {colorUseRule=UseColor_Specified, specifiedColor=<rgba 1 1 1 1>}
     uVRegion {bIsValid=false, max=<0 0>, min=<0 0>}
    }}
   slot {class=HorizontalBoxSlot}
  }
  SearchableSubfolderComboBox {
   widgetClassPath=&SearchableStringComboBox
   bIsVariable=true
   parent=AP_Filter_HB
   widget=SearchableSubfolderComboBox
   properties {
    bOverride_Cursor=true
    selectedOption=All
    toolTipText=Category Dropdown: Corresponds to 'Subfolders' field in AP Files
    defaultOptions [All]
    font {
     bForceMonospaced=false
     bMaterialIsStencil=false
     fontMaterial=null
     fontObject=&Roboto
     letterSpacing=0
     monospacedWidth=1
     size=12
     skewAmount=0
     typefaceFontName=Bold
     outlineSettings {
      bApplyOutlineToDropShadows=false
      bMiteredCorners=false
      bSeparateFillAlpha=false
      outlineMaterial=null
      outlineSize=0
      outlineColor=<rgba 0 0 0 1>
     }
    }
    foregroundColor {colorUseRule=UseColor_Specified, specifiedColor=<rgba 0.5271 0.5271 0.5271 1>}
    itemStyle {
     bUseParentRowBrush=false
     activeBrush {
      bIsDynamicallyLoaded=false
      drawAs=Image
      imageType=NoImage
      mirroring=NoMirror
      resourceName=None
      resourceObject=null
      tiling=NoTile
      imageSize=<0 0>
      margin=<ltrb 0 0 0 0>
      outlineSettings {
       bUseBrushTransparency=false
       roundingType=HalfHeightRadius
       width=0
       color {colorUseRule=UseColor_Specified, specifiedColor=<rgba 0 0 0 0>}
       cornerRadii=<0 0 0 0>
      }
      tintColor {colorUseRule=UseColor_Specified, specifiedColor=<rgba 0 0.162 0.7454 1>}
      uVRegion {bIsValid=false, max=<0 0>, min=<0 0>}
     }
     activeHighlightedBrush {
      bIsDynamicallyLoaded=false
      drawAs=Image
      imageType=NoImage
      mirroring=NoMirror
      resourceName=None
      resourceObject=null
      tiling=NoTile
      imageSize=<0 0>
      margin=<ltrb 0 0 0 0>
      outlineSettings {
       bUseBrushTransparency=false
       roundingType=HalfHeightRadius
       width=0
       color {colorUseRule=UseColor_Specified, specifiedColor=<rgba 0 0 0 0>}
       cornerRadii=<0 0 0 0>
      }
      tintColor {colorUseRule=UseColor_Specified, specifiedColor=<rgba 0.0252 0.0319 0.0423 1>}
      uVRegion {bIsValid=false, max=<0 0>, min=<0 0>}
     }
     activeHoveredBrush {
      bIsDynamicallyLoaded=false
      drawAs=Image
      imageType=NoImage
      mirroring=NoMirror
      resourceName=None
      resourceObject=null
      tiling=NoTile
      imageSize=<0 0>
      margin=<ltrb 0 0 0 0>
      outlineSettings {
       bUseBrushTransparency=false
       roundingType=HalfHeightRadius
       width=0
       color {colorUseRule=UseColor_Specified, specifiedColor=<rgba 0 0 0 0>}
       cornerRadii=<0 0 0 0>
      }
      tintColor {colorUseRule=UseColor_Specified, specifiedColor=<rgba 0 0.162 0.7454 1>}
      uVRegion {bIsValid=false, max=<0 0>, min=<0 0>}
     }
     dropIndicator_Above {
      bIsDynamicallyLoaded=false
      drawAs=Box
      imageType=FullColor
      mirroring=NoMirror
      resourceName=../../../Engine/Content/Slate/Common/DropZoneIndicator_Above.png
      resourceObject=null
      tiling=NoTile
      imageSize=<0 0>
      margin=<ltrb 0.625 0.625 0 0>
      outlineSettings {
       bUseBrushTransparency=false
       roundingType=HalfHeightRadius
       width=0
       color {colorUseRule=UseColor_Specified, specifiedColor=<rgba 0 0 0 0>}
       cornerRadii=<0 0 0 0>
      }
      tintColor {colorUseRule=UseColor_Specified, specifiedColor=<rgba 0.0091 0.3515 0.9647 1>}
      uVRegion {bIsValid=false, max=<0 0>, min=<0 0>}
     }
     dropIndicator_Below {
      bIsDynamicallyLoaded=false
      drawAs=Box
      imageType=FullColor
      mirroring=NoMirror
      resourceName=../../../Engine/Content/Slate/Common/DropZoneIndicator_Below.png
      resourceObject=null
      tiling=NoTile
      imageSize=<0 0>
      margin=<ltrb 0.625 0 0 0.625>
      outlineSettings {
       bUseBrushTransparency=false
       roundingType=HalfHeightRadius
       width=0
       color {colorUseRule=UseColor_Specified, specifiedColor=<rgba 0 0 0 0>}
       cornerRadii=<0 0 0 0>
      }
      tintColor {colorUseRule=UseColor_Specified, specifiedColor=<rgba 0.0091 0.3515 0.9647 1>}
      uVRegion {bIsValid=false, max=<0 0>, min=<0 0>}
     }
     dropIndicator_Onto {
      bIsDynamicallyLoaded=false
      drawAs=Box
      imageType=FullColor
      mirroring=NoMirror
      resourceName=../../../Engine/Content/Slate/Common/DropZoneIndicator_Onto.png
      resourceObject=null
      tiling=NoTile
      imageSize=<0 0>
      margin=<ltrb 0.25 0.25 0.25 0.25>
      outlineSettings {
       bUseBrushTransparency=false
       roundingType=HalfHeightRadius
       width=0
       color {colorUseRule=UseColor_Specified, specifiedColor=<rgba 0 0 0 0>}
       cornerRadii=<0 0 0 0>
      }
      tintColor {colorUseRule=UseColor_Specified, specifiedColor=<rgba 0.0091 0.3515 0.9647 1>}
      uVRegion {bIsValid=false, max=<0 0>, min=<0 0>}
     }
     evenRowBackgroundBrush {
      bIsDynamicallyLoaded=false
      drawAs=Image
      imageType=NoImage
      mirroring=NoMirror
      resourceName=None
      resourceObject=null
      tiling=NoTile
      imageSize=<0 0>
      margin=<ltrb 0 0 0 0>
      outlineSettings {
       bUseBrushTransparency=false
       roundingType=HalfHeightRadius
       width=0
       color {colorUseRule=UseColor_Specified, specifiedColor=<rgba 0 0 0 0>}
       cornerRadii=<0 0 0 0>
      }
      tintColor {colorUseRule=UseColor_Specified, specifiedColor=<rgba 0.0103 0.0103 0.0103 1>}
      uVRegion {bIsValid=false, max=<0 0>, min=<0 0>}
     }
     evenRowBackgroundHoveredBrush {
      bIsDynamicallyLoaded=false
      drawAs=Image
      imageType=NoImage
      mirroring=NoMirror
      resourceName=None
      resourceObject=null
      tiling=NoTile
      imageSize=<0 0>
      margin=<ltrb 0 0 0 0>
      outlineSettings {
       bUseBrushTransparency=false
       roundingType=HalfHeightRadius
       width=0
       color {colorUseRule=UseColor_Specified, specifiedColor=<rgba 0 0 0 0>}
       cornerRadii=<0 0 0 0>
      }
      tintColor {colorUseRule=UseColor_Specified, specifiedColor=<rgba 0.0176 0.0176 0.0176 1>}
      uVRegion {bIsValid=false, max=<0 0>, min=<0 0>}
     }
     inactiveBrush {
      bIsDynamicallyLoaded=false
      drawAs=Image
      imageType=NoImage
      mirroring=NoMirror
      resourceName=None
      resourceObject=null
      tiling=NoTile
      imageSize=<0 0>
      margin=<ltrb 0 0 0 0>
      outlineSettings {
       bUseBrushTransparency=false
       roundingType=HalfHeightRadius
       width=0
       color {colorUseRule=UseColor_Specified, specifiedColor=<rgba 0 0 0 0>}
       cornerRadii=<0 0 0 0>
      }
      tintColor {colorUseRule=UseColor_Specified, specifiedColor=<rgba 0.0513 0.0953 0.159 1>}
      uVRegion {bIsValid=false, max=<0 0>, min=<0 0>}
     }
     inactiveHighlightedBrush {
      bIsDynamicallyLoaded=false
      drawAs=Image
      imageType=NoImage
      mirroring=NoMirror
      resourceName=None
      resourceObject=null
      tiling=NoTile
      imageSize=<0 0>
      margin=<ltrb 0 0 0 0>
      outlineSettings {
       bUseBrushTransparency=false
       roundingType=HalfHeightRadius
       width=0
       color {colorUseRule=UseColor_Specified, specifiedColor=<rgba 0 0 0 0>}
       cornerRadii=<0 0 0 0>
      }
      tintColor {colorUseRule=UseColor_Specified, specifiedColor=<rgba 0.0252 0.0319 0.0423 1>}
      uVRegion {bIsValid=false, max=<0 0>, min=<0 0>}
     }
     inactiveHoveredBrush {
      bIsDynamicallyLoaded=false
      drawAs=Image
      imageType=NoImage
      mirroring=NoMirror
      resourceName=None
      resourceObject=null
      tiling=NoTile
      imageSize=<0 0>
      margin=<ltrb 0 0 0 0>
      outlineSettings {
       bUseBrushTransparency=false
       roundingType=HalfHeightRadius
       width=0
       color {colorUseRule=UseColor_Specified, specifiedColor=<rgba 0 0 0 0>}
       cornerRadii=<0 0 0 0>
      }
      tintColor {colorUseRule=UseColor_Specified, specifiedColor=<rgba 0.0513 0.0953 0.159 1>}
      uVRegion {bIsValid=false, max=<0 0>, min=<0 0>}
     }
     oddRowBackgroundBrush {
      bIsDynamicallyLoaded=false
      drawAs=Image
      imageType=NoImage
      mirroring=NoMirror
      resourceName=None
      resourceObject=null
      tiling=NoTile
      imageSize=<0 0>
      margin=<ltrb 0 0 0 0>
      outlineSettings {
       bUseBrushTransparency=false
       roundingType=HalfHeightRadius
       width=0
       color {colorUseRule=UseColor_Specified, specifiedColor=<rgba 0 0 0 0>}
       cornerRadii=<0 0 0 0>
      }
      tintColor {colorUseRule=UseColor_Specified, specifiedColor=<rgba 0.0103 0.0103 0.0103 1>}
      uVRegion {bIsValid=false, max=<0 0>, min=<0 0>}
     }
     oddRowBackgroundHoveredBrush {
      bIsDynamicallyLoaded=false
      drawAs=Image
      imageType=NoImage
      mirroring=NoMirror
      resourceName=None
      resourceObject=null
      tiling=NoTile
      imageSize=<0 0>
      margin=<ltrb 0 0 0 0>
      outlineSettings {
       bUseBrushTransparency=false
       roundingType=HalfHeightRadius
       width=0
       color {colorUseRule=UseColor_Specified, specifiedColor=<rgba 0 0 0 0>}
       cornerRadii=<0 0 0 0>
      }
      tintColor {colorUseRule=UseColor_Specified, specifiedColor=<rgba 0.0176 0.0176 0.0176 1>}
      uVRegion {bIsValid=false, max=<0 0>, min=<0 0>}
     }
     parentRowBackgroundBrush {
      bIsDynamicallyLoaded=false
      drawAs=Image
      imageType=NoImage
      mirroring=NoMirror
      resourceName=None
      resourceObject=null
      tiling=NoTile
      imageSize=<32 32>
      margin=<ltrb 0 0 0 0>
      outlineSettings {
       bUseBrushTransparency=false
       roundingType=HalfHeightRadius
       width=0
       color {colorUseRule=UseColor_Specified, specifiedColor=<rgba 0 0 0 0>}
       cornerRadii=<0 0 0 0>
      }
      tintColor {colorUseRule=UseColor_Specified, specifiedColor=<rgba 1 1 1 1>}
      uVRegion {bIsValid=false, max=<0 0>, min=<0 0>}
     }
     parentRowBackgroundHoveredBrush {
      bIsDynamicallyLoaded=false
      drawAs=Image
      imageType=NoImage
      mirroring=NoMirror
      resourceName=None
      resourceObject=null
      tiling=NoTile
      imageSize=<32 32>
      margin=<ltrb 0 0 0 0>
      outlineSettings {
       bUseBrushTransparency=false
       roundingType=HalfHeightRadius
       width=0
       color {colorUseRule=UseColor_Specified, specifiedColor=<rgba 0 0 0 0>}
       cornerRadii=<0 0 0 0>
      }
      tintColor {colorUseRule=UseColor_Specified, specifiedColor=<rgba 1 1 1 1>}
      uVRegion {bIsValid=false, max=<0 0>, min=<0 0>}
     }
     selectedTextColor {colorUseRule=UseColor_Specified, specifiedColor=<rgba 1 1 1 1>}
     selectorFocusedBrush {
      bIsDynamicallyLoaded=false
      drawAs=Border
      imageType=FullColor
      mirroring=NoMirror
      resourceName=../../../Engine/Content/Slate/Common/Selector.png
      resourceObject=null
      tiling=Both
      imageSize=<0 0>
      margin=<ltrb 0.25 0.25 0.25 0.25>
      outlineSettings {
       bUseBrushTransparency=false
       roundingType=HalfHeightRadius
       width=0
       color {colorUseRule=UseColor_Specified, specifiedColor=<rgba 0 0 0 0>}
       cornerRadii=<0 0 0 0>
      }
      tintColor {colorUseRule=UseColor_Specified, specifiedColor=<rgba 0.701 0.225 0.003 1>}
      uVRegion {bIsValid=false, max=<0 0>, min=<0 0>}
     }
     textColor {colorUseRule=UseColor_Specified, specifiedColor=<rgba 0.5271 0.5271 0.5271 1>}
    }
    widgetStyle {
     comboButtonStyle {
      downArrowAlign=VAlign_Center
      buttonStyle {
       clickedSlateSound {resourceObject=null}
       disabled {
        bIsDynamicallyLoaded=false
        drawAs=RoundedBox
        imageType=NoImage
        mirroring=NoMirror
        resourceName=None
        resourceObject=null
        tiling=NoTile
        imageSize=<0 0>
        margin=<ltrb 0 0 0 0>
        outlineSettings {
         bUseBrushTransparency=false
         roundingType=FixedRadius
         width=1
         color {colorUseRule=UseColor_ColorTable, specifiedColor=<rgba 0 0 0 0>}
         cornerRadii=<4 4 4 4>
        }
        tintColor {colorUseRule=UseColor_Specified, specifiedColor=<rgba 0.0395 0.0395 0.0395 1>}
        uVRegion {bIsValid=false, max=<0 0>, min=<0 0>}
       }
       disabledForeground {colorUseRule=UseColor_ColorTable, specifiedColor=<rgba 0 0 0 0>}
       hovered {
        bIsDynamicallyLoaded=false
        drawAs=RoundedBox
        imageType=NoImage
        mirroring=NoMirror
        resourceName=None
        resourceObject=null
        tiling=NoTile
        imageSize=<0 0>
        margin=<ltrb 0 0 0 0>
        outlineSettings {
         bUseBrushTransparency=false
         roundingType=FixedRadius
         width=1
         color {colorUseRule=UseColor_ColorTable, specifiedColor=<rgba 0 0 0 0>}
         cornerRadii=<4 4 4 4>
        }
        tintColor {colorUseRule=UseColor_Specified, specifiedColor=<rgba 0.0048 0.0048 0.0048 1>}
        uVRegion {bIsValid=false, max=<0 0>, min=<0 0>}
       }
       hoveredForeground {colorUseRule=UseColor_ColorTable, specifiedColor=<rgba 0 0 0 0>}
       hoveredSlateSound {resourceObject=null}
       normal {
        bIsDynamicallyLoaded=false
        drawAs=RoundedBox
        imageType=NoImage
        mirroring=NoMirror
        resourceName=None
        resourceObject=null
        tiling=NoTile
        imageSize=<0 0>
        margin=<ltrb 0 0 0 0>
        outlineSettings {
         bUseBrushTransparency=false
         roundingType=FixedRadius
         width=1
         color {colorUseRule=UseColor_ColorTable, specifiedColor=<rgba 0 0 0 0>}
         cornerRadii=<4 4 4 4>
        }
        tintColor {colorUseRule=UseColor_Specified, specifiedColor=<rgba 0.0048 0.0048 0.0048 1>}
        uVRegion {bIsValid=false, max=<0 0>, min=<0 0>}
       }
       normalForeground {colorUseRule=UseColor_ColorTable, specifiedColor=<rgba 0 0 0 0>}
       normalPadding=<ltrb 8 1 8 1>
       pressed {
        bIsDynamicallyLoaded=false
        drawAs=RoundedBox
        imageType=NoImage
        mirroring=NoMirror
        resourceName=None
        resourceObject=null
        tiling=NoTile
        imageSize=<0 0>
        margin=<ltrb 0 0 0 0>
        outlineSettings {
         bUseBrushTransparency=false
         roundingType=FixedRadius
         width=1
         color {colorUseRule=UseColor_ColorTable, specifiedColor=<rgba 0 0 0 0>}
         cornerRadii=<4 4 4 4>
        }
        tintColor {colorUseRule=UseColor_Specified, specifiedColor=<rgba 0.0048 0.0048 0.0048 1>}
        uVRegion {bIsValid=false, max=<0 0>, min=<0 0>}
       }
       pressedForeground {colorUseRule=UseColor_ColorTable, specifiedColor=<rgba 0 0 0 0>}
       pressedPadding=<ltrb 8 2 8 0>
       pressedSlateSound {resourceObject=null}
      }
      contentPadding=<ltrb 0 0 0 0>
      downArrowImage {
       bIsDynamicallyLoaded=false
       drawAs=Image
       imageType=Vector
       mirroring=NoMirror
       resourceName=../../../Engine/Content/Slate/Starship/CoreWidgets/ComboBox/wide-chevron-down.svg
       resourceObject=null
       tiling=NoTile
       imageSize=<16 16>
       margin=<ltrb 0 0 0 0>
       outlineSettings {
        bUseBrushTransparency=false
        roundingType=HalfHeightRadius
        width=0
        color {colorUseRule=UseColor_Specified, specifiedColor=<rgba 0 0 0 0>}
        cornerRadii=<0 0 0 0>
       }
       tintColor {colorUseRule=UseColor_Specified, specifiedColor=<rgba 1 1 1 1>}
       uVRegion {bIsValid=false, max=<0 0>, min=<0 0>}
      }
      downArrowPadding=<ltrb 2 2 2 2>
      menuBorderBrush {
       bIsDynamicallyLoaded=false
       drawAs=RoundedBox
       imageType=NoImage
       mirroring=NoMirror
       resourceName=None
       resourceObject=null
       tiling=NoTile
       imageSize=<0 0>
       margin=<ltrb 0 0 0 0>
       outlineSettings {
        bUseBrushTransparency=false
        roundingType=FixedRadius
        width=1
        color {colorUseRule=UseColor_ColorTable, specifiedColor=<rgba 0 0 0 0>}
        cornerRadii=<0 0 0 0>
       }
       tintColor {colorUseRule=UseColor_Specified, specifiedColor=<rgba 0.0395 0.0395 0.0395 1>}
       uVRegion {bIsValid=false, max=<0 0>, min=<0 0>}
      }
      menuBorderPadding=<ltrb 0 0 0 0>
      shadowColorAndOpacity=<rgba 0 0 0 1>
      shadowOffset=<0 0>
     }
     contentPadding=<ltrb 0 0 0 0>
     menuRowPadding=<ltrb 8 3 8 3>
     pressedSlateSound {resourceObject=null}
     selectionChangeSlateSound {resourceObject=null}
    }
   }
   slot {class=HorizontalBoxSlot, properties {size {sizeRule=Fill, value=1}}}
  }
  EditorUtilityScrollBox_638 {
   widgetClassPath=&EditorUtilityScrollBox
   parent=VerticalBox_75
   widget=EditorUtilityScrollBox_638
   slot {class=VerticalBoxSlot}
  }
  SearchBar_HorizontalBox {
   widgetClassPath=&HorizontalBox
   parent=EditorUtilityScrollBox_638
   widget=SearchBar_HorizontalBox
   slot {class=ScrollBoxSlot, properties {padding=<ltrb 18 0 0 0>}}
  }
  SearchBar {
   widgetClassPath=&EditorUtilityEditableTextBox
   bIsVariable=true
   parent=SearchBar_HorizontalBox
   widget=SearchBar
   properties {hintText=Search}
   slot {class=HorizontalBoxSlot, properties {size {sizeRule=Fill, value=1}}}
  }
  ClearSearchBox {
   widgetClassPath=&Button
   parent=SearchBar_HorizontalBox
   widget=ClearSearchBox
   properties {widgetStyle {
     clickedSlateSound {resourceObject=null}
     disabled {
      bIsDynamicallyLoaded=false
      drawAs=RoundedBox
      imageType=NoImage
      mirroring=NoMirror
      resourceName=None
      resourceObject=null
      tiling=NoTile
      imageSize=<0 0>
      margin=<ltrb 0 0 0 0>
      outlineSettings {
       bUseBrushTransparency=false
       roundingType=FixedRadius
       width=1
       color {colorUseRule=UseColor_ColorTable, specifiedColor=<rgba 0 0 0 0>}
       cornerRadii=<4 4 4 4>
      }
      tintColor {colorUseRule=UseColor_Specified, specifiedColor=<rgba 0.0395 0.0395 0.0395 1>}
      uVRegion {bIsValid=false, max=<0 0>, min=<0 0>}
     }
     disabledForeground {colorUseRule=UseColor_ColorTable, specifiedColor=<rgba 0 0 0 0>}
     hovered {
      bIsDynamicallyLoaded=false
      drawAs=RoundedBox
      imageType=NoImage
      mirroring=NoMirror
      resourceName=None
      resourceObject=null
      tiling=NoTile
      imageSize=<32 32>
      margin=<ltrb 0 0 0 0>
      outlineSettings {
       bUseBrushTransparency=false
       roundingType=FixedRadius
       width=0
       color {colorUseRule=UseColor_ColorTable, specifiedColor=<rgba 0 0 0 0>}
       cornerRadii=<4 4 4 4>
      }
      tintColor {colorUseRule=UseColor_Specified, specifiedColor=<rgba 0.0953 0.0953 0.0953 1>}
      uVRegion {bIsValid=false, max=<0 0>, min=<0 0>}
     }
     hoveredForeground {colorUseRule=UseColor_ColorTable, specifiedColor=<rgba 0 0 0 0>}
     hoveredSlateSound {resourceObject=null}
     normal {
      bIsDynamicallyLoaded=false
      drawAs=NoDrawType
      imageType=FullColor
      mirroring=NoMirror
      resourceName=None
      resourceObject=null
      tiling=NoTile
      imageSize=<16 16>
      margin=<ltrb 0 0 0 0>
      outlineSettings {
       bUseBrushTransparency=false
       roundingType=FixedRadius
       width=0
       color {colorUseRule=UseColor_ColorTable, specifiedColor=<rgba 0 0 0 0>}
       cornerRadii=<4 4 4 4>
      }
      tintColor {colorUseRule=UseColor_Specified, specifiedColor=<rgba 0.0176 0.0176 0.0176 1>}
      uVRegion {bIsValid=false, max=<0 0>, min=<0 0>}
     }
     normalForeground {colorUseRule=UseColor_ColorTable, specifiedColor=<rgba 0 0 0 0>}
     normalPadding=<ltrb 6 1.5 6 1.5>
     pressed {
      bIsDynamicallyLoaded=false
      drawAs=RoundedBox
      imageType=NoImage
      mirroring=NoMirror
      resourceName=None
      resourceObject=null
      tiling=NoTile
      imageSize=<32 32>
      margin=<ltrb 0 0 0 0>
      outlineSettings {
       bUseBrushTransparency=false
       roundingType=FixedRadius
       width=0
       color {colorUseRule=UseColor_ColorTable, specifiedColor=<rgba 0 0 0 0>}
       cornerRadii=<4 4 4 4>
      }
      tintColor {colorUseRule=UseColor_Specified, specifiedColor=<rgba 0.0284 0.0284 0.0284 1>}
      uVRegion {bIsValid=false, max=<0 0>, min=<0 0>}
     }
     pressedForeground {colorUseRule=UseColor_ColorTable, specifiedColor=<rgba 0 0 0 0>}
     pressedPadding=<ltrb 6 2.5 6 0.5>
     pressedSlateSound {resourceObject=null}
    }}
   slot {class=HorizontalBoxSlot, properties {horizontalAlignment=HAlign_Center, padding=<ltrb 0 0 5 0>}}
  }
  X {
   widgetClassPath=&TextBlock
   parent=ClearSearchBox
   widget=X
   properties {
    autoWrapText=true
    justification=Center
    text=x
    colorAndOpacity {colorUseRule=UseColor_Specified, specifiedColor=<rgba 0.5255 0.5255 0.5255 1>}
    font {
     bForceMonospaced=false
     bMaterialIsStencil=false
     fontMaterial=null
     fontObject=&Roboto
     letterSpacing=0
     monospacedWidth=1
     size=10
     skewAmount=0
     typefaceFontName=Bold
     outlineSettings {
      bApplyOutlineToDropShadows=false
      bMiteredCorners=false
      bSeparateFillAlpha=false
      outlineMaterial=null
      outlineSize=0
      outlineColor=<rgba 0 0 0 1>
     }
    }
   }
   slot {class=ButtonSlot}
  }
  SettingsButton {
   widgetClassPath=&Button
   parent=SearchBar_HorizontalBox
   widget=SettingsButton
   properties {
    toolTipText=Change Project
    widgetStyle {
     clickedSlateSound {resourceObject=null}
     disabled {
      bIsDynamicallyLoaded=false
      drawAs=RoundedBox
      imageType=NoImage
      mirroring=NoMirror
      resourceName=None
      resourceObject=null
      tiling=NoTile
      imageSize=<0 0>
      margin=<ltrb 0 0 0 0>
      outlineSettings {
       bUseBrushTransparency=false
       roundingType=FixedRadius
       width=1
       color {colorUseRule=UseColor_ColorTable, specifiedColor=<rgba 0 0 0 0>}
       cornerRadii=<4 4 4 4>
      }
      tintColor {colorUseRule=UseColor_Specified, specifiedColor=<rgba 0.0395 0.0395 0.0395 1>}
      uVRegion {bIsValid=false, max=<0 0>, min=<0 0>}
     }
     disabledForeground {colorUseRule=UseColor_ColorTable, specifiedColor=<rgba 0 0 0 0>}
     hovered {
      bIsDynamicallyLoaded=false
      drawAs=RoundedBox
      imageType=FullColor
      mirroring=NoMirror
      resourceName=None
      resourceObject=null
      tiling=NoTile
      imageSize=<28 28>
      margin=<ltrb 0 0 0 0>
      outlineSettings {
       bUseBrushTransparency=false
       roundingType=FixedRadius
       width=0
       color {colorUseRule=UseColor_ColorTable, specifiedColor=<rgba 0 0 0 0>}
       cornerRadii=<4 4 4 4>
      }
      tintColor {colorUseRule=UseColor_Specified, specifiedColor=<rgba 0.0941 0.0941 0.0941 1>}
      uVRegion {bIsValid=false, max=<0 0>, min=<0 0>}
     }
     hoveredForeground {colorUseRule=UseColor_ColorTable, specifiedColor=<rgba 0 0 0 0>}
     hoveredSlateSound {resourceObject=null}
     normal {
      bIsDynamicallyLoaded=false
      drawAs=NoDrawType
      imageType=FullColor
      mirroring=NoMirror
      resourceName=None
      resourceObject=null
      tiling=NoTile
      imageSize=<28 28>
      margin=<ltrb 0 0 0 0>
      outlineSettings {
       bUseBrushTransparency=false
       roundingType=FixedRadius
       width=0
       color {colorUseRule=UseColor_ColorTable, specifiedColor=<rgba 0 0 0 0>}
       cornerRadii=<4 4 4 4>
      }
      tintColor {colorUseRule=UseColor_Specified, specifiedColor=<rgba 0.0176 0.0176 0.0176 1>}
      uVRegion {bIsValid=false, max=<0 0>, min=<0 0>}
     }
     normalForeground {colorUseRule=UseColor_ColorTable, specifiedColor=<rgba 0 0 0 0>}
     normalPadding=<ltrb 0 0 0 0>
     pressed {
      bIsDynamicallyLoaded=false
      drawAs=RoundedBox
      imageType=FullColor
      mirroring=NoMirror
      resourceName=None
      resourceObject=null
      tiling=NoTile
      imageSize=<28 28>
      margin=<ltrb 0 0 0 0>
      outlineSettings {
       bUseBrushTransparency=false
       roundingType=FixedRadius
       width=0
       color {colorUseRule=UseColor_ColorTable, specifiedColor=<rgba 0 0 0 0>}
       cornerRadii=<4 4 4 4>
      }
      tintColor {colorUseRule=UseColor_Specified, specifiedColor=<rgba 0.0392 0.0392 0.0392 1>}
      uVRegion {bIsValid=false, max=<0 0>, min=<0 0>}
     }
     pressedForeground {colorUseRule=UseColor_ColorTable, specifiedColor=<rgba 0 0 0 0>}
     pressedPadding=<ltrb 0 0 0 0>
     pressedSlateSound {resourceObject=null}
    }
   }
   slot {class=HorizontalBoxSlot, properties {horizontalAlignment=HAlign_Center, verticalAlignment=VAlign_Center, padding=<ltrb 0 0 5 0>}}
  }
  HorizontalBox_5 {
   widgetClassPath=&HorizontalBox
   parent=SettingsButton
   widget=HorizontalBox_5
   slot {class=ButtonSlot, properties {padding=<ltrb 0 0 0 0>}}
  }
  SettingsButtonImage {
   widgetClassPath=&Image
   bIsVariable=true
   parent=HorizontalBox_5
   widget=SettingsButtonImage
   properties {brush {
     bIsDynamicallyLoaded=false
     drawAs=Image
     imageType=FullColor
     mirroring=NoMirror
     resourceName=None
     resourceObject=/AnimInEdTools/AnimationBrowser/Icons/Icon_Settings
     tiling=NoTile
     imageSize=<28 28>
     margin=<ltrb 0 0 0 0>
     outlineSettings {
      bUseBrushTransparency=false
      roundingType=HalfHeightRadius
      width=0
      color {colorUseRule=UseColor_Specified, specifiedColor=<rgba 0 0 0 0>}
      cornerRadii=<0 0 0 0>
     }
     tintColor {colorUseRule=UseColor_Specified, specifiedColor=<rgba 1 1 1 1>}
     uVRegion {bIsValid=false, max=<0 0>, min=<0 0>}
    }}
   slot {class=HorizontalBoxSlot, properties {horizontalAlignment=HAlign_Center, verticalAlignment=VAlign_Center}}
  }
  Button_? {
   widgetClassPath=&EditorUtilityButton
   parent=SearchBar_HorizontalBox
   widget=Button_?
   properties {
    toolTipText=Open Confluence Page
    widgetStyle {
     clickedSlateSound {resourceObject=null}
     disabled {
      bIsDynamicallyLoaded=false
      drawAs=RoundedBox
      imageType=NoImage
      mirroring=NoMirror
      resourceName=None
      resourceObject=null
      tiling=NoTile
      imageSize=<0 0>
      margin=<ltrb 0 0 0 0>
      outlineSettings {
       bUseBrushTransparency=false
       roundingType=FixedRadius
       width=1
       color {colorUseRule=UseColor_ColorTable, specifiedColor=<rgba 0 0 0 0>}
       cornerRadii=<4 4 4 4>
      }
      tintColor {colorUseRule=UseColor_Specified, specifiedColor=<rgba 0.0395 0.0395 0.0395 1>}
      uVRegion {bIsValid=false, max=<0 0>, min=<0 0>}
     }
     disabledForeground {colorUseRule=UseColor_ColorTable, specifiedColor=<rgba 0 0 0 0>}
     hovered {
      bIsDynamicallyLoaded=false
      drawAs=RoundedBox
      imageType=NoImage
      mirroring=NoMirror
      resourceName=None
      resourceObject=null
      tiling=NoTile
      imageSize=<32 32>
      margin=<ltrb 0 0 0 0>
      outlineSettings {
       bUseBrushTransparency=false
       roundingType=FixedRadius
       width=1
       color {colorUseRule=UseColor_ColorTable, specifiedColor=<rgba 0 0 0 0>}
       cornerRadii=<4 4 4 4>
      }
      tintColor {colorUseRule=UseColor_Specified, specifiedColor=<rgba 0.0953 0.0953 0.0953 1>}
      uVRegion {bIsValid=false, max=<0 0>, min=<0 0>}
     }
     hoveredForeground {colorUseRule=UseColor_ColorTable, specifiedColor=<rgba 0 0 0 0>}
     hoveredSlateSound {resourceObject=null}
     normal {
      bIsDynamicallyLoaded=false
      drawAs=RoundedBox
      imageType=NoImage
      mirroring=NoMirror
      resourceName=None
      resourceObject=null
      tiling=NoTile
      imageSize=<32 32>
      margin=<ltrb 0 0 0 0>
      outlineSettings {
       bUseBrushTransparency=false
       roundingType=FixedRadius
       width=1
       color {colorUseRule=UseColor_ColorTable, specifiedColor=<rgba 0 0 0 0>}
       cornerRadii=<4 4 4 4>
      }
      tintColor {colorUseRule=UseColor_Specified, specifiedColor=<rgba 0.0395 0.0395 0.0395 1>}
      uVRegion {bIsValid=false, max=<0 0>, min=<0 0>}
     }
     normalForeground {colorUseRule=UseColor_ColorTable, specifiedColor=<rgba 0 0 0 0>}
     normalPadding=<ltrb 8 2 6 2>
     pressed {
      bIsDynamicallyLoaded=false
      drawAs=RoundedBox
      imageType=NoImage
      mirroring=NoMirror
      resourceName=None
      resourceObject=null
      tiling=NoTile
      imageSize=<32 32>
      margin=<ltrb 0 0 0 0>
      outlineSettings {
       bUseBrushTransparency=false
       roundingType=FixedRadius
       width=1
       color {colorUseRule=UseColor_ColorTable, specifiedColor=<rgba 0 0 0 0>}
       cornerRadii=<4 4 4 4>
      }
      tintColor {colorUseRule=UseColor_Specified, specifiedColor=<rgba 0.0284 0.0284 0.0284 1>}
      uVRegion {bIsValid=false, max=<0 0>, min=<0 0>}
     }
     pressedForeground {colorUseRule=UseColor_ColorTable, specifiedColor=<rgba 0 0 0 0>}
     pressedPadding=<ltrb 8 2.5 6 2.5>
     pressedSlateSound {resourceObject=null}
    }
   }
   slot {class=HorizontalBoxSlot, properties {horizontalAlignment=HAlign_Center, verticalAlignment=VAlign_Center, padding=<ltrb 0 0 5 0>}}
  }
  TextBlock_14 {
   widgetClassPath=&TextBlock
   parent=Button_?
   widget=TextBlock_14
   properties {
    text=?
    colorAndOpacity {colorUseRule=UseColor_Specified, specifiedColor=<rgba 0 1 0 1>}
    font {
     bForceMonospaced=false
     bMaterialIsStencil=false
     fontMaterial=null
     fontObject=&Roboto
     letterSpacing=0
     monospacedWidth=1
     size=12
     skewAmount=0
     typefaceFontName=Bold
     outlineSettings {
      bApplyOutlineToDropShadows=false
      bMiteredCorners=false
      bSeparateFillAlpha=false
      outlineMaterial=null
      outlineSize=0
      outlineColor=<rgba 0 0 0 1>
     }
    }
   }
   slot {class=ButtonSlot, properties {horizontalAlignment=HAlign_Left, padding=<ltrb 0 0 0 0>}}
  }
  refreshUIButton {
   widgetClassPath=&Button
   parent=SearchBar_HorizontalBox
   widget=refreshUIButton
   properties {
    toolTipText=Refresh Assets
    widgetStyle {
     clickedSlateSound {resourceObject=null}
     disabled {
      bIsDynamicallyLoaded=false
      drawAs=RoundedBox
      imageType=NoImage
      mirroring=NoMirror
      resourceName=None
      resourceObject=null
      tiling=NoTile
      imageSize=<0 0>
      margin=<ltrb 0 0 0 0>
      outlineSettings {
       bUseBrushTransparency=false
       roundingType=FixedRadius
       width=1
       color {colorUseRule=UseColor_ColorTable, specifiedColor=<rgba 0 0 0 0>}
       cornerRadii=<4 4 4 4>
      }
      tintColor {colorUseRule=UseColor_Specified, specifiedColor=<rgba 0.0395 0.0395 0.0395 1>}
      uVRegion {bIsValid=false, max=<0 0>, min=<0 0>}
     }
     disabledForeground {colorUseRule=UseColor_ColorTable, specifiedColor=<rgba 0 0 0 0>}
     hovered {
      bIsDynamicallyLoaded=false
      drawAs=RoundedBox
      imageType=FullColor
      mirroring=NoMirror
      resourceName=None
      resourceObject=null
      tiling=NoTile
      imageSize=<28 28>
      margin=<ltrb 0 0 0 0>
      outlineSettings {
       bUseBrushTransparency=false
       roundingType=FixedRadius
       width=0
       color {colorUseRule=UseColor_ColorTable, specifiedColor=<rgba 0 0 0 0>}
       cornerRadii=<4 4 4 4>
      }
      tintColor {colorUseRule=UseColor_Specified, specifiedColor=<rgba 0.0941 0.0941 0.0941 1>}
      uVRegion {bIsValid=false, max=<0 0>, min=<0 0>}
     }
     hoveredForeground {colorUseRule=UseColor_ColorTable, specifiedColor=<rgba 0 0 0 0>}
     hoveredSlateSound {resourceObject=null}
     normal {
      bIsDynamicallyLoaded=false
      drawAs=NoDrawType
      imageType=FullColor
      mirroring=NoMirror
      resourceName=None
      resourceObject=null
      tiling=NoTile
      imageSize=<28 28>
      margin=<ltrb 0 0 0 0>
      outlineSettings {
       bUseBrushTransparency=false
       roundingType=FixedRadius
       width=0
       color {colorUseRule=UseColor_ColorTable, specifiedColor=<rgba 0 0 0 0>}
       cornerRadii=<4 4 4 4>
      }
      tintColor {colorUseRule=UseColor_Specified, specifiedColor=<rgba 0.0176 0.0176 0.0176 1>}
      uVRegion {bIsValid=false, max=<0 0>, min=<0 0>}
     }
     normalForeground {colorUseRule=UseColor_ColorTable, specifiedColor=<rgba 0 0 0 0>}
     normalPadding=<ltrb 0 0 0 0>
     pressed {
      bIsDynamicallyLoaded=false
      drawAs=RoundedBox
      imageType=FullColor
      mirroring=NoMirror
      resourceName=None
      resourceObject=null
      tiling=NoTile
      imageSize=<28 28>
      margin=<ltrb 0 0 0 0>
      outlineSettings {
       bUseBrushTransparency=false
       roundingType=FixedRadius
       width=0
       color {colorUseRule=UseColor_ColorTable, specifiedColor=<rgba 0 0 0 0>}
       cornerRadii=<4 4 4 4>
      }
      tintColor {colorUseRule=UseColor_Specified, specifiedColor=<rgba 0.0392 0.0392 0.0392 1>}
      uVRegion {bIsValid=false, max=<0 0>, min=<0 0>}
     }
     pressedForeground {colorUseRule=UseColor_ColorTable, specifiedColor=<rgba 0 0 0 0>}
     pressedPadding=<ltrb 0 0 0 0>
     pressedSlateSound {resourceObject=null}
    }
   }
   slot {class=HorizontalBoxSlot, properties {horizontalAlignment=HAlign_Center, verticalAlignment=VAlign_Center, padding=<ltrb 0 0 5 0>}}
  }
  HorizontalBox_3 {
   widgetClassPath=&HorizontalBox
   parent=refreshUIButton
   widget=HorizontalBox_3
   properties {toolTipText=Reload Tool}
   slot {class=ButtonSlot, properties {padding=<ltrb 0 0 0 0>}}
  }
  refreshicon {
   widgetClassPath=&Image
   bIsVariable=true
   parent=HorizontalBox_3
   widget=refreshicon
   properties {
    visibility=HitTestInvisible
    brush {
     bIsDynamicallyLoaded=false
     drawAs=Image
     imageType=FullColor
     mirroring=NoMirror
     resourceName=None
     resourceObject=/AnimInEdTools/AnimationBrowser/Icons/RefreshSpinner
     tiling=NoTile
     imageSize=<16 16>
     margin=<ltrb 0 0 0 0>
     outlineSettings {
      bUseBrushTransparency=false
      roundingType=HalfHeightRadius
      width=0
      color {colorUseRule=UseColor_Specified, specifiedColor=<rgba 0 0 0 0>}
      cornerRadii=<0 0 0 0>
     }
     tintColor {colorUseRule=UseColor_Specified, specifiedColor=<rgba 1 1 1 1>}
     uVRegion {bIsValid=false, max=<0 0>, min=<0 0>}
    }
   }
   slot {class=HorizontalBoxSlot, properties {horizontalAlignment=HAlign_Center, verticalAlignment=VAlign_Center}}
  }
  EUW_SosButton {
   widgetClassPath=/AnimInEdTools/PropBrowser/SOSButton/EUW_SosButton
   bIsVariable=true
   parent=SearchBar_HorizontalBox
   widget=EUW_SosButton
   properties {incoming_toolname=PropBrowser, visibility=Visible}
   slot {class=HorizontalBoxSlot}
  }
  SizeBox_0 {
   widgetClassPath=&SizeBox
   parent=VerticalBox_75
   widget=SizeBox_0
   properties {bOverride_HeightOverride=true, heightOverride=165}
   slot {class=VerticalBoxSlot, properties {padding=<ltrb 5 2 5 0>}}
  }
  ListViewBorder {
   widgetClassPath=&Border
   parent=SizeBox_0
   widget=ListViewBorder
   properties {brushColor=<rgba 0.0103 0.0103 0.0103 1>, desiredSizeScale=<1.1 1.1>, padding=<ltrb 0 0 0 0>}
   slot {class=SizeBoxSlot}
  }
  ListView {
   widgetClassPath=/Script/Blutility.EditorUtilityListView
   bIsVariable=true
   parent=ListViewBorder
   widget=ListView
   properties {
    allowOverscroll=false
    bEnableRightClickScrolling=false
    bReturnFocusToSelection=true
    entryWidgetClass=&/EUW_PropBrowser_ListItem.EUW_PropBrowser_ListItem_C
    scrollBarStyle {
     thickness=8
     draggedThumbImage {
      bIsDynamicallyLoaded=false
      drawAs=RoundedBox
      imageType=NoImage
      mirroring=NoMirror
      resourceName=None
      resourceObject=null
      tiling=NoTile
      imageSize=<0 0>
      margin=<ltrb 0 0 0 0>
      outlineSettings {
       bUseBrushTransparency=false
       roundingType=FixedRadius
       width=0
       color {colorUseRule=UseColor_Specified, specifiedColor=<rgba 0 0 0 0>}
       cornerRadii=<4 4 4 4>
      }
      tintColor {colorUseRule=UseColor_Specified, specifiedColor=<rgba 0.2159 0.2159 0.2159 1>}
      uVRegion {bIsValid=false, max=<0 0>, min=<0 0>}
     }
     horizontalBackgroundImage {
      bIsDynamicallyLoaded=false
      drawAs=NoDrawType
      imageType=NoImage
      mirroring=NoMirror
      resourceName=None
      resourceObject=null
      tiling=NoTile
      imageSize=<0 0>
      margin=<ltrb 0 0 0 0>
      outlineSettings {
       bUseBrushTransparency=false
       roundingType=HalfHeightRadius
       width=0
       color {colorUseRule=UseColor_Specified, specifiedColor=<rgba 0 0 0 0>}
       cornerRadii=<0 0 0 0>
      }
      tintColor {colorUseRule=UseColor_Specified, specifiedColor=<rgba 1 1 1 1>}
      uVRegion {bIsValid=false, max=<0 0>, min=<0 0>}
     }
     horizontalBottomSlotImage {
      bIsDynamicallyLoaded=false
      drawAs=NoDrawType
      imageType=NoImage
      mirroring=NoMirror
      resourceName=None
      resourceObject=null
      tiling=NoTile
      imageSize=<0 0>
      margin=<ltrb 0 0 0 0>
      outlineSettings {
       bUseBrushTransparency=false
       roundingType=HalfHeightRadius
       width=0
       color {colorUseRule=UseColor_Specified, specifiedColor=<rgba 0 0 0 0>}
       cornerRadii=<0 0 0 0>
      }
      tintColor {colorUseRule=UseColor_Specified, specifiedColor=<rgba 1 1 1 1>}
      uVRegion {bIsValid=false, max=<0 0>, min=<0 0>}
     }
     horizontalTopSlotImage {
      bIsDynamicallyLoaded=false
      drawAs=NoDrawType
      imageType=NoImage
      mirroring=NoMirror
      resourceName=None
      resourceObject=null
      tiling=NoTile
      imageSize=<0 0>
      margin=<ltrb 0 0 0 0>
      outlineSettings {
       bUseBrushTransparency=false
       roundingType=HalfHeightRadius
       width=0
       color {colorUseRule=UseColor_Specified, specifiedColor=<rgba 0 0 0 0>}
       cornerRadii=<0 0 0 0>
      }
      tintColor {colorUseRule=UseColor_Specified, specifiedColor=<rgba 1 1 1 1>}
      uVRegion {bIsValid=false, max=<0 0>, min=<0 0>}
     }
     hoveredThumbImage {
      bIsDynamicallyLoaded=false
      drawAs=RoundedBox
      imageType=NoImage
      mirroring=NoMirror
      resourceName=None
      resourceObject=null
      tiling=NoTile
      imageSize=<0 0>
      margin=<ltrb 0 0 0 0>
      outlineSettings {
       bUseBrushTransparency=false
       roundingType=FixedRadius
       width=0
       color {colorUseRule=UseColor_Specified, specifiedColor=<rgba 0 0 0 0>}
       cornerRadii=<4 4 4 4>
      }
      tintColor {colorUseRule=UseColor_Specified, specifiedColor=<rgba 0.2159 0.2159 0.2159 1>}
      uVRegion {bIsValid=false, max=<0 0>, min=<0 0>}
     }
     normalThumbImage {
      bIsDynamicallyLoaded=false
      drawAs=RoundedBox
      imageType=NoImage
      mirroring=NoMirror
      resourceName=None
      resourceObject=null
      tiling=NoTile
      imageSize=<0 0>
      margin=<ltrb 0 0 0 0>
      outlineSettings {
       bUseBrushTransparency=false
       roundingType=FixedRadius
       width=0
       color {colorUseRule=UseColor_Specified, specifiedColor=<rgba 0 0 0 0>}
       cornerRadii=<4 4 4 4>
      }
      tintColor {colorUseRule=UseColor_Specified, specifiedColor=<rgba 0.0953 0.0953 0.0953 1>}
      uVRegion {bIsValid=false, max=<0 0>, min=<0 0>}
     }
     verticalBackgroundImage {
      bIsDynamicallyLoaded=false
      drawAs=NoDrawType
      imageType=NoImage
      mirroring=NoMirror
      resourceName=None
      resourceObject=null
      tiling=NoTile
      imageSize=<0 0>
      margin=<ltrb 0 0 0 0>
      outlineSettings {
       bUseBrushTransparency=false
       roundingType=HalfHeightRadius
       width=0
       color {colorUseRule=UseColor_Specified, specifiedColor=<rgba 0 0 0 0>}
       cornerRadii=<0 0 0 0>
      }
      tintColor {colorUseRule=UseColor_Specified, specifiedColor=<rgba 1 1 1 1>}
      uVRegion {bIsValid=false, max=<0 0>, min=<0 0>}
     }
     verticalBottomSlotImage {
      bIsDynamicallyLoaded=false
      drawAs=NoDrawType
      imageType=NoImage
      mirroring=NoMirror
      resourceName=None
      resourceObject=null
      tiling=NoTile
      imageSize=<0 0>
      margin=<ltrb 0 0 0 0>
      outlineSettings {
       bUseBrushTransparency=false
       roundingType=HalfHeightRadius
       width=0
       color {colorUseRule=UseColor_Specified, specifiedColor=<rgba 0 0 0 0>}
       cornerRadii=<0 0 0 0>
      }
      tintColor {colorUseRule=UseColor_Specified, specifiedColor=<rgba 1 1 1 1>}
      uVRegion {bIsValid=false, max=<0 0>, min=<0 0>}
     }
     verticalTopSlotImage {
      bIsDynamicallyLoaded=false
      drawAs=NoDrawType
      imageType=NoImage
      mirroring=NoMirror
      resourceName=None
      resourceObject=null
      tiling=NoTile
      imageSize=<0 0>
      margin=<ltrb 0 0 0 0>
      outlineSettings {
       bUseBrushTransparency=false
       roundingType=HalfHeightRadius
       width=0
       color {colorUseRule=UseColor_Specified, specifiedColor=<rgba 0 0 0 0>}
       cornerRadii=<0 0 0 0>
      }
      tintColor {colorUseRule=UseColor_Specified, specifiedColor=<rgba 1 1 1 1>}
      uVRegion {bIsValid=false, max=<0 0>, min=<0 0>}
     }
    }
   }
   slot {class=BorderSlot, properties {padding=<ltrb 0 0 0 0>}}
  }
  PropAddVerticalBoxOptions {
   widgetClassPath=&VerticalBox
   bIsVariable=true
   parent=VerticalBox_75
   widget=PropAddVerticalBoxOptions
   properties {bOverride_Cursor=true}
   slot {class=VerticalBoxSlot, properties {verticalAlignment=VAlign_Top, padding=<ltrb 0 8 0 8>}}
  }
  CheckBoxesContainer {
   widgetClassPath=&HorizontalBox
   parent=PropAddVerticalBoxOptions
   widget=CheckBoxesContainer
   properties {bOverride_Cursor=true}
   slot {class=VerticalBoxSlot, properties {size {sizeRule=Fill, value=1}}}
  }
  Border_98 {
   widgetClassPath=&Border
   parent=CheckBoxesContainer
   widget=Border_98
   properties {
    background {
     bIsDynamicallyLoaded=false
     drawAs=Image
     imageType=NoImage
     mirroring=NoMirror
     resourceName=None
     resourceObject=null
     tiling=NoTile
     imageSize=<32 32>
     margin=<ltrb 0 0 0 0>
     outlineSettings {
      bUseBrushTransparency=false
      roundingType=HalfHeightRadius
      width=0
      color {colorUseRule=UseColor_Specified, specifiedColor=<rgba 0 0 0 0>}
      cornerRadii=<0 0 0 0>
     }
     tintColor {colorUseRule=UseColor_Foreground, specifiedColor=<rgba 1 1 1 1>}
     uVRegion {bIsValid=false, max=<0 0>, min=<0 0>}
    }
    brushColor=<rgba 0 0 0 0.502>
   }
   slot {class=HorizontalBoxSlot, properties {verticalAlignment=VAlign_Center, padding=<ltrb -2 -2 -2 -2>, size {sizeRule=Fill, value=1}}}
  }
  Border_114 {
   widgetClassPath=&Border
   parent=Border_98
   widget=Border_114
   properties {
    background {
     bIsDynamicallyLoaded=false
     drawAs=Box
     imageType=NoImage
     mirroring=NoMirror
     resourceName=None
     resourceObject=null
     tiling=NoTile
     imageSize=<32 32>
     margin=<ltrb 0 0 0 0>
     outlineSettings {
      bUseBrushTransparency=false
      roundingType=HalfHeightRadius
      width=0
      color {colorUseRule=UseColor_Specified, specifiedColor=<rgba 0 0 0 0>}
      cornerRadii=<0 0 0 0>
     }
     tintColor {colorUseRule=UseColor_Specified, specifiedColor=<rgba 1 1 1 1>}
     uVRegion {bIsValid=false, max=<0 0>, min=<0 0>}
    }
    brushColor=<rgba 0.0467 0.0467 0.0497 1>
    padding=<ltrb 0 0 0 0>
   }
   slot {class=BorderSlot}
  }
  VerticalBox_0 {
   widgetClassPath=&VerticalBox
   parent=Border_114
   widget=VerticalBox_0
   properties {bOverride_Cursor=true}
   slot {class=BorderSlot, properties {padding=<ltrb 0 0 0 0>}}
  }
  TextBlock_105 {
   widgetClassPath=&TextBlock
   parent=VerticalBox_0
   widget=TextBlock_105
   properties {
    text=Attach Options
    colorAndOpacity {colorUseRule=UseColor_Foreground, specifiedColor=<rgba 1 1 1 1>}
    font {
     bForceMonospaced=false
     bMaterialIsStencil=false
     fontMaterial=null
     fontObject=&Roboto
     letterSpacing=0
     monospacedWidth=1
     size=13.5
     skewAmount=0
     typefaceFontName=Bold
     outlineSettings {
      bApplyOutlineToDropShadows=false
      bMiteredCorners=false
      bSeparateFillAlpha=false
      outlineMaterial=null
      outlineSize=0
      outlineColor=<rgba 0 0 0 1>
     }
    }
   }
   slot {class=VerticalBoxSlot, properties {padding=<ltrb 5 5 5 5>}}
  }
  Border_0 {
   widgetClassPath=&Border
   parent=VerticalBox_0
   widget=Border_0
   properties {brushColor=<rgba 0.0185 0.0185 0.0194 1>}
   slot {class=VerticalBoxSlot, properties {verticalAlignment=VAlign_Center, size {sizeRule=Fill, value=1}}}
  }
  VerticalBox_5 {widgetClassPath=&VerticalBox, parent=Border_0, widget=VerticalBox_5, slot {class=BorderSlot}}
  JustAddHorizontalBox {
   widgetClassPath=&HorizontalBox
   parent=VerticalBox_5
   widget=JustAddHorizontalBox
   properties {toolTipText="Prop will only be added to levelSequence "}
   slot {class=VerticalBoxSlot}
  }
  JustImportCkBx {
   widgetClassPath=&EditorUtilityCheckBox
   bIsVariable=true
   parent=JustAddHorizontalBox
   widget=JustImportCkBx
   slot {class=HorizontalBoxSlot, properties {horizontalAlignment=HAlign_Center, verticalAlignment=VAlign_Center}}
  }
  AddandConstrain {
   widgetClassPath=&TextBlock
   parent=JustImportCkBx
   widget=AddandConstrain
   properties {
    text=Just Add
    font {
     bForceMonospaced=false
     bMaterialIsStencil=false
     fontMaterial=null
     fontObject=&Roboto
     letterSpacing=0
     monospacedWidth=1
     size=12
     skewAmount=0
     typefaceFontName=Regular
     outlineSettings {
      bApplyOutlineToDropShadows=false
      bMiteredCorners=false
      bSeparateFillAlpha=false
      outlineMaterial=null
      outlineSize=0
      outlineColor=<rgba 0 0 0 1>
     }
    }
   }
   slot {class=PanelSlot}
  }
  AddandAttachTrackHorizontalBox {
   widgetClassPath=&HorizontalBox
   parent=VerticalBox_5
   widget=AddandAttachTrackHorizontalBox
   properties {toolTipText="If True, Prop will be add to level sequence and attachtrack added.. "}
   slot {class=VerticalBoxSlot}
  }
  AddandConstraintCkBx {
   widgetClassPath=&EditorUtilityCheckBox
   bIsVariable=true
   parent=AddandAttachTrackHorizontalBox
   widget=AddandConstraintCkBx
   properties {checkedState=Checked, horizontalAlignment=HAlign_Center}
   slot {class=HorizontalBoxSlot}
  }
  AddandConstrain_1 {
   widgetClassPath=&TextBlock
   parent=AddandAttachTrackHorizontalBox
   widget=AddandConstrain_1
   properties {
    text=Add and Attach
    font {
     bForceMonospaced=false
     bMaterialIsStencil=false
     fontMaterial=null
     fontObject=&Roboto
     letterSpacing=0
     monospacedWidth=1
     size=12
     skewAmount=0
     typefaceFontName=Regular
     outlineSettings {
      bApplyOutlineToDropShadows=false
      bMiteredCorners=false
      bSeparateFillAlpha=false
      outlineMaterial=null
      outlineSize=0
      outlineColor=<rgba 0 0 0 1>
     }
    }
   }
   slot {class=HorizontalBoxSlot}
  }
  Border_93 {
   widgetClassPath=&Border
   parent=VerticalBox_5
   widget=Border_93
   properties {
    background {
     bIsDynamicallyLoaded=false
     drawAs=Image
     imageType=NoImage
     mirroring=NoMirror
     resourceName=None
     resourceObject=null
     tiling=NoTile
     imageSize=<32 32>
     margin=<ltrb 0 0 0 0>
     outlineSettings {
      bUseBrushTransparency=false
      roundingType=HalfHeightRadius
      width=0
      color {colorUseRule=UseColor_Specified, specifiedColor=<rgba 0 0 0 0>}
      cornerRadii=<0 0 0 0>
     }
     tintColor {colorUseRule=UseColor_Specified, specifiedColor=<rgba 0.0103 0.0103 0.0103 1>}
     uVRegion {bIsValid=false, max=<0 0>, min=<0 0>}
    }
    desiredSizeScale=<0.5 0.5>
   }
   slot {class=VerticalBoxSlot, properties {padding=<ltrb 0 5 0 5>}}
  }
  TextCurrentRigsinSequencer {
   widgetClassPath=&TextBlock
   parent=VerticalBox_5
   widget=TextCurrentRigsinSequencer
   properties {
    text=Parent Asset
    font {
     bForceMonospaced=false
     bMaterialIsStencil=false
     fontMaterial=null
     fontObject=&Roboto
     letterSpacing=0
     monospacedWidth=1
     size=12
     skewAmount=0
     typefaceFontName=Regular
     outlineSettings {
      bApplyOutlineToDropShadows=false
      bMiteredCorners=false
      bSeparateFillAlpha=false
      outlineMaterial=null
      outlineSize=0
      outlineColor=<rgba 0 0 0 1>
     }
    }
   }
   slot {
    class=VerticalBoxSlot
    properties {horizontalAlignment=HAlign_Left, verticalAlignment=VAlign_Center, padding=<ltrb 4 2 4 2>, size {sizeRule=Fill, value=1}}
   }
  }
  ParentControlRigComboBox {
   widgetClassPath=/Script/Blutility.EditorUtilityComboBoxString
   bIsVariable=true
   parent=VerticalBox_5
   widget=ParentControlRigComboBox
   properties {
    toolTipText=Control rig you want to constrain prop to
    contentPadding=<ltrb 0 0 0 0>
    font {
     bForceMonospaced=false
     bMaterialIsStencil=false
     fontMaterial=null
     fontObject=&Roboto
     letterSpacing=0
     monospacedWidth=1
     size=12
     skewAmount=0
     typefaceFontName=Bold
     outlineSettings {
      bApplyOutlineToDropShadows=false
      bMiteredCorners=false
      bSeparateFillAlpha=false
      outlineMaterial=null
      outlineSize=0
      outlineColor=<rgba 0 0 0 1>
     }
    }
    foregroundColor {colorUseRule=UseColor_Specified, specifiedColor=<rgba 0.5271 0.5271 0.5271 1>}
   }
   slot {class=VerticalBoxSlot, properties {padding=<ltrb 0 0 0 5>}}
  }
  RichTextBlock_129 {
   widgetClassPath=/Script/UMG.RichTextBlock
   bIsVariable=true
   parent=VerticalBox_5
   widget=RichTextBlock_129
   slot {class=VerticalBoxSlot, properties {size {sizeRule=Fill, value=1}}}
  }
  Border_2 {
   widgetClassPath=&Border
   parent=VerticalBox_5
   widget=Border_2
   properties {
    background {
     bIsDynamicallyLoaded=false
     drawAs=Image
     imageType=NoImage
     mirroring=NoMirror
     resourceName=None
     resourceObject=null
     tiling=NoTile
     imageSize=<32 32>
     margin=<ltrb 0 0 0 0>
     outlineSettings {
      bUseBrushTransparency=false
      roundingType=HalfHeightRadius
      width=0
      color {colorUseRule=UseColor_Specified, specifiedColor=<rgba 0 0 0 0>}
      cornerRadii=<0 0 0 0>
     }
     tintColor {colorUseRule=UseColor_Specified, specifiedColor=<rgba 0.0103 0.0103 0.0103 1>}
     uVRegion {bIsValid=false, max=<0 0>, min=<0 0>}
    }
    desiredSizeScale=<0.5 0.5>
   }
   slot {class=VerticalBoxSlot, properties {padding=<ltrb 0 5 0 5>}}
  }
  VerticalBox_2 {
   widgetClassPath=&VerticalBox
   parent=VerticalBox_5
   widget=VerticalBox_2
   slot {class=VerticalBoxSlot, properties {verticalAlignment=VAlign_Bottom, padding=<ltrb 5 0 5 0>}}
  }
  TextBlock_274 {
   widgetClassPath=&TextBlock
   parent=VerticalBox_2
   widget=TextBlock_274
   properties {
    text=Type
    font {
     bForceMonospaced=false
     bMaterialIsStencil=false
     fontMaterial=null
     fontObject=&Roboto
     letterSpacing=0
     monospacedWidth=1
     size=12
     skewAmount=0
     typefaceFontName=Regular
     outlineSettings {
      bApplyOutlineToDropShadows=false
      bMiteredCorners=false
      bSeparateFillAlpha=false
      outlineMaterial=null
      outlineSize=0
      outlineColor=<rgba 0 0 0 1>
     }
    }
   }
   slot {class=VerticalBoxSlot, properties {horizontalAlignment=HAlign_Left, padding=<ltrb 0 0 0 5>}}
  }
  attach_bone_chkbx {
   widgetClassPath=&EditorUtilityCheckBox
   bIsVariable=true
   parent=VerticalBox_2
   widget=attach_bone_chkbx
   properties {checkedState=Checked}
   slot {class=VerticalBoxSlot, properties {verticalAlignment=VAlign_Bottom}}
  }
  Bones {
   widgetClassPath=&TextBlock
   parent=attach_bone_chkbx
   widget=Bones
   properties {
    text=Bone attachments
    font {
     bForceMonospaced=false
     bMaterialIsStencil=false
     fontMaterial=null
     fontObject=&Roboto
     letterSpacing=0
     monospacedWidth=1
     size=12
     skewAmount=0
     typefaceFontName=Regular
     outlineSettings {
      bApplyOutlineToDropShadows=false
      bMiteredCorners=false
      bSeparateFillAlpha=false
      outlineMaterial=null
      outlineSize=0
      outlineColor=<rgba 0 0 0 1>
     }
    }
   }
   slot {class=PanelSlot}
  }
  attach_socket_chkbx {
   widgetClassPath=&EditorUtilityCheckBox
   bIsVariable=true
   parent=VerticalBox_2
   widget=attach_socket_chkbx
   slot {class=VerticalBoxSlot, properties {verticalAlignment=VAlign_Bottom}}
  }
  Sockets {
   widgetClassPath=&TextBlock
   parent=attach_socket_chkbx
   widget=Sockets
   properties {
    text=Socket attachments
    font {
     bForceMonospaced=false
     bMaterialIsStencil=false
     fontMaterial=null
     fontObject=&Roboto
     letterSpacing=0
     monospacedWidth=1
     size=12
     skewAmount=0
     typefaceFontName=Regular
     outlineSettings {
      bApplyOutlineToDropShadows=false
      bMiteredCorners=false
      bSeparateFillAlpha=false
      outlineMaterial=null
      outlineSize=0
      outlineColor=<rgba 0 0 0 1>
     }
    }
   }
   slot {class=PanelSlot}
  }
  VerticalBox_4 {
   widgetClassPath=&VerticalBox
   parent=VerticalBox_5
   widget=VerticalBox_4
   slot {class=VerticalBoxSlot, properties {padding=<ltrb 5 5 5 5>}}
  }
  DriverControlsTextBlock {
   widgetClassPath=&TextBlock
   parent=VerticalBox_4
   widget=DriverControlsTextBlock
   properties {
    text=Attachment Point
    font {
     bForceMonospaced=false
     bMaterialIsStencil=false
     fontMaterial=null
     fontObject=&Roboto
     letterSpacing=0
     monospacedWidth=1
     size=12
     skewAmount=0
     typefaceFontName=Regular
     outlineSettings {
      bApplyOutlineToDropShadows=false
      bMiteredCorners=false
      bSeparateFillAlpha=false
      outlineMaterial=null
      outlineSize=0
      outlineColor=<rgba 0 0 0 1>
     }
    }
   }
   slot {class=VerticalBoxSlot}
  }
  HorizontalBox_1 {widgetClassPath=&HorizontalBox, parent=VerticalBox_4, widget=HorizontalBox_1, slot {class=VerticalBoxSlot}}
  ConstraintoBoneVerticcalBox {
   widgetClassPath=&VerticalBox
   parent=HorizontalBox_1
   widget=ConstraintoBoneVerticcalBox
   properties {toolTipText=What Bone the Prop will be attached to}
   slot {class=HorizontalBoxSlot, properties {size {sizeRule=Fill, value=1}}}
  }
  AttachmentPointSearchableComboBox {
   widgetClassPath=&SearchableStringComboBox
   bIsVariable=true
   parent=ConstraintoBoneVerticcalBox
   widget=AttachmentPointSearchableComboBox
   properties {
    contentPadding=<ltrb 0 0 0 0>
    font {
     bForceMonospaced=false
     bMaterialIsStencil=false
     fontMaterial=null
     fontObject=&Roboto
     letterSpacing=0
     monospacedWidth=1
     size=12
     skewAmount=0
     typefaceFontName=Bold
     outlineSettings {
      bApplyOutlineToDropShadows=false
      bMiteredCorners=false
      bSeparateFillAlpha=false
      outlineMaterial=null
      outlineSize=0
      outlineColor=<rgba 0 0 0 1>
     }
    }
    foregroundColor {colorUseRule=UseColor_Specified, specifiedColor=<rgba 0.5271 0.5271 0.5271 1>}
    itemStyle {
     bUseParentRowBrush=false
     activeBrush {
      bIsDynamicallyLoaded=false
      drawAs=Image
      imageType=NoImage
      mirroring=NoMirror
      resourceName=None
      resourceObject=null
      tiling=NoTile
      imageSize=<0 0>
      margin=<ltrb 0 0 0 0>
      outlineSettings {
       bUseBrushTransparency=false
       roundingType=HalfHeightRadius
       width=0
       color {colorUseRule=UseColor_Specified, specifiedColor=<rgba 0 0 0 0>}
       cornerRadii=<0 0 0 0>
      }
      tintColor {colorUseRule=UseColor_Specified, specifiedColor=<rgba 0 0.162 0.7454 1>}
      uVRegion {bIsValid=false, max=<0 0>, min=<0 0>}
     }
     activeHighlightedBrush {
      bIsDynamicallyLoaded=false
      drawAs=Image
      imageType=NoImage
      mirroring=NoMirror
      resourceName=None
      resourceObject=null
      tiling=NoTile
      imageSize=<0 0>
      margin=<ltrb 0 0 0 0>
      outlineSettings {
       bUseBrushTransparency=false
       roundingType=HalfHeightRadius
       width=0
       color {colorUseRule=UseColor_Specified, specifiedColor=<rgba 0 0 0 0>}
       cornerRadii=<0 0 0 0>
      }
      tintColor {colorUseRule=UseColor_Specified, specifiedColor=<rgba 0.0252 0.0319 0.0423 1>}
      uVRegion {bIsValid=false, max=<0 0>, min=<0 0>}
     }
     activeHoveredBrush {
      bIsDynamicallyLoaded=false
      drawAs=Image
      imageType=NoImage
      mirroring=NoMirror
      resourceName=None
      resourceObject=null
      tiling=NoTile
      imageSize=<0 0>
      margin=<ltrb 0 0 0 0>
      outlineSettings {
       bUseBrushTransparency=false
       roundingType=HalfHeightRadius
       width=0
       color {colorUseRule=UseColor_Specified, specifiedColor=<rgba 0 0 0 0>}
       cornerRadii=<0 0 0 0>
      }
      tintColor {colorUseRule=UseColor_Specified, specifiedColor=<rgba 0 0.162 0.7454 1>}
      uVRegion {bIsValid=false, max=<0 0>, min=<0 0>}
     }
     dropIndicator_Above {
      bIsDynamicallyLoaded=false
      drawAs=Box
      imageType=FullColor
      mirroring=NoMirror
      resourceName=../../../Engine/Content/Slate/Common/DropZoneIndicator_Above.png
      resourceObject=null
      tiling=NoTile
      imageSize=<0 0>
      margin=<ltrb 0.625 0.625 0 0>
      outlineSettings {
       bUseBrushTransparency=false
       roundingType=HalfHeightRadius
       width=0
       color {colorUseRule=UseColor_Specified, specifiedColor=<rgba 0 0 0 0>}
       cornerRadii=<0 0 0 0>
      }
      tintColor {colorUseRule=UseColor_Specified, specifiedColor=<rgba 0.0091 0.3515 0.9647 1>}
      uVRegion {bIsValid=false, max=<0 0>, min=<0 0>}
     }
     dropIndicator_Below {
      bIsDynamicallyLoaded=false
      drawAs=Box
      imageType=FullColor
      mirroring=NoMirror
      resourceName=../../../Engine/Content/Slate/Common/DropZoneIndicator_Below.png
      resourceObject=null
      tiling=NoTile
      imageSize=<0 0>
      margin=<ltrb 0.625 0 0 0.625>
      outlineSettings {
       bUseBrushTransparency=false
       roundingType=HalfHeightRadius
       width=0
       color {colorUseRule=UseColor_Specified, specifiedColor=<rgba 0 0 0 0>}
       cornerRadii=<0 0 0 0>
      }
      tintColor {colorUseRule=UseColor_Specified, specifiedColor=<rgba 0.0091 0.3515 0.9647 1>}
      uVRegion {bIsValid=false, max=<0 0>, min=<0 0>}
     }
     dropIndicator_Onto {
      bIsDynamicallyLoaded=false
      drawAs=Box
      imageType=FullColor
      mirroring=NoMirror
      resourceName=../../../Engine/Content/Slate/Common/DropZoneIndicator_Onto.png
      resourceObject=null
      tiling=NoTile
      imageSize=<0 0>
      margin=<ltrb 0.25 0.25 0.25 0.25>
      outlineSettings {
       bUseBrushTransparency=false
       roundingType=HalfHeightRadius
       width=0
       color {colorUseRule=UseColor_Specified, specifiedColor=<rgba 0 0 0 0>}
       cornerRadii=<0 0 0 0>
      }
      tintColor {colorUseRule=UseColor_Specified, specifiedColor=<rgba 0.0091 0.3515 0.9647 1>}
      uVRegion {bIsValid=false, max=<0 0>, min=<0 0>}
     }
     evenRowBackgroundBrush {
      bIsDynamicallyLoaded=false
      drawAs=Image
      imageType=NoImage
      mirroring=NoMirror
      resourceName=None
      resourceObject=null
      tiling=NoTile
      imageSize=<0 0>
      margin=<ltrb 0 0 0 0>
      outlineSettings {
       bUseBrushTransparency=false
       roundingType=HalfHeightRadius
       width=0
       color {colorUseRule=UseColor_Specified, specifiedColor=<rgba 0 0 0 0>}
       cornerRadii=<0 0 0 0>
      }
      tintColor {colorUseRule=UseColor_Specified, specifiedColor=<rgba 0.0103 0.0103 0.0103 1>}
      uVRegion {bIsValid=false, max=<0 0>, min=<0 0>}
     }
     evenRowBackgroundHoveredBrush {
      bIsDynamicallyLoaded=false
      drawAs=Image
      imageType=NoImage
      mirroring=NoMirror
      resourceName=None
      resourceObject=null
      tiling=NoTile
      imageSize=<0 0>
      margin=<ltrb 0 0 0 0>
      outlineSettings {
       bUseBrushTransparency=false
       roundingType=HalfHeightRadius
       width=0
       color {colorUseRule=UseColor_Specified, specifiedColor=<rgba 0 0 0 0>}
       cornerRadii=<0 0 0 0>
      }
      tintColor {colorUseRule=UseColor_Specified, specifiedColor=<rgba 0.0176 0.0176 0.0176 1>}
      uVRegion {bIsValid=false, max=<0 0>, min=<0 0>}
     }
     inactiveBrush {
      bIsDynamicallyLoaded=false
      drawAs=Image
      imageType=NoImage
      mirroring=NoMirror
      resourceName=None
      resourceObject=null
      tiling=NoTile
      imageSize=<0 0>
      margin=<ltrb 0 0 0 0>
      outlineSettings {
       bUseBrushTransparency=false
       roundingType=HalfHeightRadius
       width=0
       color {colorUseRule=UseColor_Specified, specifiedColor=<rgba 0 0 0 0>}
       cornerRadii=<0 0 0 0>
      }
      tintColor {colorUseRule=UseColor_Specified, specifiedColor=<rgba 0.0513 0.0953 0.159 1>}
      uVRegion {bIsValid=false, max=<0 0>, min=<0 0>}
     }
     inactiveHighlightedBrush {
      bIsDynamicallyLoaded=false
      drawAs=Image
      imageType=NoImage
      mirroring=NoMirror
      resourceName=None
      resourceObject=null
      tiling=NoTile
      imageSize=<0 0>
      margin=<ltrb 0 0 0 0>
      outlineSettings {
       bUseBrushTransparency=false
       roundingType=HalfHeightRadius
       width=0
       color {colorUseRule=UseColor_Specified, specifiedColor=<rgba 0 0 0 0>}
       cornerRadii=<0 0 0 0>
      }
      tintColor {colorUseRule=UseColor_Specified, specifiedColor=<rgba 0.0252 0.0319 0.0423 1>}
      uVRegion {bIsValid=false, max=<0 0>, min=<0 0>}
     }
     inactiveHoveredBrush {
      bIsDynamicallyLoaded=false
      drawAs=Image
      imageType=NoImage
      mirroring=NoMirror
      resourceName=None
      resourceObject=null
      tiling=NoTile
      imageSize=<0 0>
      margin=<ltrb 0 0 0 0>
      outlineSettings {
       bUseBrushTransparency=false
       roundingType=HalfHeightRadius
       width=0
       color {colorUseRule=UseColor_Specified, specifiedColor=<rgba 0 0 0 0>}
       cornerRadii=<0 0 0 0>
      }
      tintColor {colorUseRule=UseColor_Specified, specifiedColor=<rgba 0.0513 0.0953 0.159 1>}
      uVRegion {bIsValid=false, max=<0 0>, min=<0 0>}
     }
     oddRowBackgroundBrush {
      bIsDynamicallyLoaded=false
      drawAs=Image
      imageType=NoImage
      mirroring=NoMirror
      resourceName=None
      resourceObject=null
      tiling=NoTile
      imageSize=<0 0>
      margin=<ltrb 0 0 0 0>
      outlineSettings {
       bUseBrushTransparency=false
       roundingType=HalfHeightRadius
       width=0
       color {colorUseRule=UseColor_Specified, specifiedColor=<rgba 0 0 0 0>}
       cornerRadii=<0 0 0 0>
      }
      tintColor {colorUseRule=UseColor_Specified, specifiedColor=<rgba 0.0103 0.0103 0.0103 1>}
      uVRegion {bIsValid=false, max=<0 0>, min=<0 0>}
     }
     oddRowBackgroundHoveredBrush {
      bIsDynamicallyLoaded=false
      drawAs=Image
      imageType=NoImage
      mirroring=NoMirror
      resourceName=None
      resourceObject=null
      tiling=NoTile
      imageSize=<0 0>
      margin=<ltrb 0 0 0 0>
      outlineSettings {
       bUseBrushTransparency=false
       roundingType=HalfHeightRadius
       width=0
       color {colorUseRule=UseColor_Specified, specifiedColor=<rgba 0 0 0 0>}
       cornerRadii=<0 0 0 0>
      }
      tintColor {colorUseRule=UseColor_Specified, specifiedColor=<rgba 0.0176 0.0176 0.0176 1>}
      uVRegion {bIsValid=false, max=<0 0>, min=<0 0>}
     }
     parentRowBackgroundBrush {
      bIsDynamicallyLoaded=false
      drawAs=Image
      imageType=NoImage
      mirroring=NoMirror
      resourceName=None
      resourceObject=null
      tiling=NoTile
      imageSize=<32 32>
      margin=<ltrb 0 0 0 0>
      outlineSettings {
       bUseBrushTransparency=false
       roundingType=HalfHeightRadius
       width=0
       color {colorUseRule=UseColor_Specified, specifiedColor=<rgba 0 0 0 0>}
       cornerRadii=<0 0 0 0>
      }
      tintColor {colorUseRule=UseColor_Specified, specifiedColor=<rgba 1 1 1 1>}
      uVRegion {bIsValid=false, max=<0 0>, min=<0 0>}
     }
     parentRowBackgroundHoveredBrush {
      bIsDynamicallyLoaded=false
      drawAs=Image
      imageType=NoImage
      mirroring=NoMirror
      resourceName=None
      resourceObject=null
      tiling=NoTile
      imageSize=<32 32>
      margin=<ltrb 0 0 0 0>
      outlineSettings {
       bUseBrushTransparency=false
       roundingType=HalfHeightRadius
       width=0
       color {colorUseRule=UseColor_Specified, specifiedColor=<rgba 0 0 0 0>}
       cornerRadii=<0 0 0 0>
      }
      tintColor {colorUseRule=UseColor_Specified, specifiedColor=<rgba 1 1 1 1>}
      uVRegion {bIsValid=false, max=<0 0>, min=<0 0>}
     }
     selectedTextColor {colorUseRule=UseColor_Specified, specifiedColor=<rgba 1 1 1 1>}
     selectorFocusedBrush {
      bIsDynamicallyLoaded=false
      drawAs=Border
      imageType=FullColor
      mirroring=NoMirror
      resourceName=../../../Engine/Content/Slate/Common/Selector.png
      resourceObject=null
      tiling=Both
      imageSize=<0 0>
      margin=<ltrb 0.25 0.25 0.25 0.25>
      outlineSettings {
       bUseBrushTransparency=false
       roundingType=HalfHeightRadius
       width=0
       color {colorUseRule=UseColor_Specified, specifiedColor=<rgba 0 0 0 0>}
       cornerRadii=<0 0 0 0>
      }
      tintColor {colorUseRule=UseColor_Specified, specifiedColor=<rgba 0.701 0.225 0.003 1>}
      uVRegion {bIsValid=false, max=<0 0>, min=<0 0>}
     }
     textColor {colorUseRule=UseColor_Specified, specifiedColor=<rgba 0.5271 0.5271 0.5271 1>}
    }
    widgetStyle {
     comboButtonStyle {
      downArrowAlign=VAlign_Center
      buttonStyle {
       clickedSlateSound {resourceObject=null}
       disabled {
        bIsDynamicallyLoaded=false
        drawAs=RoundedBox
        imageType=NoImage
        mirroring=NoMirror
        resourceName=None
        resourceObject=null
        tiling=NoTile
        imageSize=<0 0>
        margin=<ltrb 0 0 0 0>
        outlineSettings {
         bUseBrushTransparency=false
         roundingType=FixedRadius
         width=1
         color {colorUseRule=UseColor_ColorTable, specifiedColor=<rgba 0 0 0 0>}
         cornerRadii=<4 4 4 4>
        }
        tintColor {colorUseRule=UseColor_Specified, specifiedColor=<rgba 0.0395 0.0395 0.0395 1>}
        uVRegion {bIsValid=false, max=<0 0>, min=<0 0>}
       }
       disabledForeground {colorUseRule=UseColor_ColorTable, specifiedColor=<rgba 0 0 0 0>}
       hovered {
        bIsDynamicallyLoaded=false
        drawAs=RoundedBox
        imageType=NoImage
        mirroring=NoMirror
        resourceName=None
        resourceObject=null
        tiling=NoTile
        imageSize=<0 0>
        margin=<ltrb 0 0 0 0>
        outlineSettings {
         bUseBrushTransparency=false
         roundingType=FixedRadius
         width=1
         color {colorUseRule=UseColor_ColorTable, specifiedColor=<rgba 0 0 0 0>}
         cornerRadii=<4 4 4 4>
        }
        tintColor {colorUseRule=UseColor_Specified, specifiedColor=<rgba 0.0048 0.0048 0.0048 1>}
        uVRegion {bIsValid=false, max=<0 0>, min=<0 0>}
       }
       hoveredForeground {colorUseRule=UseColor_ColorTable, specifiedColor=<rgba 0 0 0 0>}
       hoveredSlateSound {resourceObject=null}
       normal {
        bIsDynamicallyLoaded=false
        drawAs=RoundedBox
        imageType=NoImage
        mirroring=NoMirror
        resourceName=None
        resourceObject=null
        tiling=NoTile
        imageSize=<0 0>
        margin=<ltrb 0 0 0 0>
        outlineSettings {
         bUseBrushTransparency=false
         roundingType=FixedRadius
         width=1
         color {colorUseRule=UseColor_ColorTable, specifiedColor=<rgba 0 0 0 0>}
         cornerRadii=<4 4 4 4>
        }
        tintColor {colorUseRule=UseColor_Specified, specifiedColor=<rgba 0.0048 0.0048 0.0048 1>}
        uVRegion {bIsValid=false, max=<0 0>, min=<0 0>}
       }
       normalForeground {colorUseRule=UseColor_ColorTable, specifiedColor=<rgba 0 0 0 0>}
       normalPadding=<ltrb 8 1 8 1>
       pressed {
        bIsDynamicallyLoaded=false
        drawAs=RoundedBox
        imageType=NoImage
        mirroring=NoMirror
        resourceName=None
        resourceObject=null
        tiling=NoTile
        imageSize=<0 0>
        margin=<ltrb 0 0 0 0>
        outlineSettings {
         bUseBrushTransparency=false
         roundingType=FixedRadius
         width=1
         color {colorUseRule=UseColor_ColorTable, specifiedColor=<rgba 0 0 0 0>}
         cornerRadii=<4 4 4 4>
        }
        tintColor {colorUseRule=UseColor_Specified, specifiedColor=<rgba 0.0048 0.0048 0.0048 1>}
        uVRegion {bIsValid=false, max=<0 0>, min=<0 0>}
       }
       pressedForeground {colorUseRule=UseColor_ColorTable, specifiedColor=<rgba 0 0 0 0>}
       pressedPadding=<ltrb 8 2 8 0>
       pressedSlateSound {resourceObject=null}
      }
      contentPadding=<ltrb 0 0 0 0>
      downArrowImage {
       bIsDynamicallyLoaded=false
       drawAs=Image
       imageType=Vector
       mirroring=NoMirror
       resourceName=../../../Engine/Content/Slate/Starship/CoreWidgets/ComboBox/wide-chevron-down.svg
       resourceObject=null
       tiling=NoTile
       imageSize=<16 16>
       margin=<ltrb 0 0 0 0>
       outlineSettings {
        bUseBrushTransparency=false
        roundingType=HalfHeightRadius
        width=0
        color {colorUseRule=UseColor_Specified, specifiedColor=<rgba 0 0 0 0>}
        cornerRadii=<0 0 0 0>
       }
       tintColor {colorUseRule=UseColor_Specified, specifiedColor=<rgba 1 1 1 1>}
       uVRegion {bIsValid=false, max=<0 0>, min=<0 0>}
      }
      downArrowPadding=<ltrb 2 2 2 2>
      menuBorderBrush {
       bIsDynamicallyLoaded=false
       drawAs=RoundedBox
       imageType=NoImage
       mirroring=NoMirror
       resourceName=None
       resourceObject=null
       tiling=NoTile
       imageSize=<0 0>
       margin=<ltrb 0 0 0 0>
       outlineSettings {
        bUseBrushTransparency=false
        roundingType=FixedRadius
        width=1
        color {colorUseRule=UseColor_ColorTable, specifiedColor=<rgba 0 0 0 0>}
        cornerRadii=<0 0 0 0>
       }
       tintColor {colorUseRule=UseColor_Specified, specifiedColor=<rgba 0.0395 0.0395 0.0395 1>}
       uVRegion {bIsValid=false, max=<0 0>, min=<0 0>}
      }
      menuBorderPadding=<ltrb 0 0 0 0>
      shadowColorAndOpacity=<rgba 0 0 0 1>
      shadowOffset=<0 0>
     }
     contentPadding=<ltrb 0 0 0 0>
     menuRowPadding=<ltrb 8 3 8 3>
     pressedSlateSound {resourceObject=null}
     selectionChangeSlateSound {resourceObject=null}
    }
   }
   slot {class=VerticalBoxSlot, properties {verticalAlignment=VAlign_Bottom, size {sizeRule=Fill, value=1}}}
  }
  SwitchDriveCtrl_Btn {
   widgetClassPath=&EditorUtilityButton
   parent=HorizontalBox_1
   widget=SwitchDriveCtrl_Btn
   properties {bIsEnabled=false, toolTipText=Updates the attachment point for the asset you just imported via the dropdowns.}
   slot {class=HorizontalBoxSlot, properties {verticalAlignment=VAlign_Bottom, size {sizeRule=Fill, value=1}}}
  }
  SwitchDriverText {
   widgetClassPath=&TextBlock
   parent=SwitchDriveCtrl_Btn
   widget=SwitchDriverText
   properties {
    text=Update Attachment Point
    font {
     bForceMonospaced=false
     bMaterialIsStencil=false
     fontMaterial=null
     fontObject=&Roboto
     letterSpacing=0
     monospacedWidth=1
     size=12
     skewAmount=0
     typefaceFontName=Regular
     outlineSettings {
      bApplyOutlineToDropShadows=false
      bMiteredCorners=false
      bSeparateFillAlpha=false
      outlineMaterial=null
      outlineSize=0
      outlineColor=<rgba 0 0 0 1>
     }
    }
   }
   slot {class=ButtonSlot}
  }
  AddtoLevelSequenceBox {widgetClassPath=&VerticalBox, parent=VerticalBox_75, widget=AddtoLevelSequenceBox, slot {class=VerticalBoxSlot}}
  Border_99 {
   widgetClassPath=&Border
   parent=AddtoLevelSequenceBox
   widget=Border_99
   properties {brushColor=<rgba 0 0 0 0.502>}
   slot {class=VerticalBoxSlot, properties {padding=<ltrb -2 -2 -2 -2>, size {sizeRule=Fill, value=1}}}
  }
  Border_95 {
   widgetClassPath=&Border
   parent=Border_99
   widget=Border_95
   properties {
    background {
     bIsDynamicallyLoaded=false
     drawAs=Box
     imageType=NoImage
     mirroring=NoMirror
     resourceName=None
     resourceObject=null
     tiling=NoTile
     imageSize=<32 32>
     margin=<ltrb 0 0 0 0>
     outlineSettings {
      bUseBrushTransparency=false
      roundingType=HalfHeightRadius
      width=0
      color {colorUseRule=UseColor_Specified, specifiedColor=<rgba 0 0 0 0>}
      cornerRadii=<0 0 0 0>
     }
     tintColor {colorUseRule=UseColor_Specified, specifiedColor=<rgba 1 1 1 1>}
     uVRegion {bIsValid=false, max=<0 0>, min=<0 0>}
    }
    brushColor=<rgba 0.0467 0.0467 0.0497 1>
    padding=<ltrb 0 0 0 0>
   }
   slot {class=BorderSlot}
  }
  VerticalBox_6 {
   widgetClassPath=&VerticalBox
   parent=Border_95
   widget=VerticalBox_6
   slot {class=BorderSlot, properties {padding=<ltrb 0 0 0 0>}}
  }
  TextBlock_1 {
   widgetClassPath=&TextBlock
   parent=VerticalBox_6
   widget=TextBlock_1
   properties {
    text=Add to LevelSequence
    colorAndOpacity {colorUseRule=UseColor_Foreground, specifiedColor=<rgba 1 1 1 1>}
    font {
     bForceMonospaced=false
     bMaterialIsStencil=false
     fontMaterial=null
     fontObject=&Roboto
     letterSpacing=0
     monospacedWidth=1
     size=13.5
     skewAmount=0
     typefaceFontName=Bold
     outlineSettings {
      bApplyOutlineToDropShadows=false
      bMiteredCorners=false
      bSeparateFillAlpha=false
      outlineMaterial=null
      outlineSize=0
      outlineColor=<rgba 0 0 0 1>
     }
    }
   }
   slot {class=VerticalBoxSlot, properties {verticalAlignment=VAlign_Center, padding=<ltrb 5 5 5 5>}}
  }
  Border_96 {
   widgetClassPath=&Border
   parent=VerticalBox_6
   widget=Border_96
   properties {horizontalAlignment=HAlign_Left, verticalAlignment=VAlign_Center, brushColor=<rgba 0.0185 0.0185 0.0194 1>, padding=<ltrb 8 0 8 0>}
   slot {class=VerticalBoxSlot}
  }
  VerticalBox_1 {
   widgetClassPath=&VerticalBox
   parent=Border_96
   widget=VerticalBox_1
   properties {bOverride_Cursor=true}
   slot {class=BorderSlot, properties {horizontalAlignment=HAlign_Left, verticalAlignment=VAlign_Center, padding=<ltrb 8 0 8 0>}}
  }
  ImportPropCreateLinkBtn {
   widgetClassPath=&EditorUtilityButton
   parent=VerticalBox_1
   widget=ImportPropCreateLinkBtn
   properties {toolTipText=Will add prop and create a linked animation sequence}
   slot {class=VerticalBoxSlot, properties {verticalAlignment=VAlign_Center, padding=<ltrb 0 5 0 0>}}
  }
  ImportRigText {
   widgetClassPath=&TextBlock
   parent=ImportPropCreateLinkBtn
   widget=ImportRigText
   properties {
    text=Add and Link AnimSequence
    font {
     bForceMonospaced=false
     bMaterialIsStencil=false
     fontMaterial=null
     fontObject=&Roboto
     letterSpacing=0
     monospacedWidth=1
     size=12
     skewAmount=0
     typefaceFontName=Regular
     outlineSettings {
      bApplyOutlineToDropShadows=false
      bMiteredCorners=false
      bSeparateFillAlpha=false
      outlineMaterial=null
      outlineSize=0
      outlineColor=<rgba 0 0 0 1>
     }
    }
   }
   slot {class=ButtonSlot, properties {horizontalAlignment=HAlign_Left, verticalAlignment=VAlign_Fill, padding=<ltrb 0 0 0 0>}}
  }
  OnlyImportPropBtn {
   widgetClassPath=&EditorUtilityButton
   parent=VerticalBox_1
   widget=OnlyImportPropBtn
   properties {toolTipText=will only add the prop into the scene}
   slot {class=VerticalBoxSlot, properties {verticalAlignment=VAlign_Center, padding=<ltrb 0 3 0 5>}}
  }
  ImportRigText_1 {
   widgetClassPath=&TextBlock
   parent=OnlyImportPropBtn
   widget=ImportRigText_1
   properties {
    text="Just Add "
    font {
     bForceMonospaced=false
     bMaterialIsStencil=false
     fontMaterial=null
     fontObject=&Roboto
     letterSpacing=0
     monospacedWidth=1
     size=12
     skewAmount=0
     typefaceFontName=Regular
     outlineSettings {
      bApplyOutlineToDropShadows=false
      bMiteredCorners=false
      bSeparateFillAlpha=false
      outlineMaterial=null
      outlineSize=0
      outlineColor=<rgba 0 0 0 1>
     }
    }
   }
   slot {class=ButtonSlot, properties {horizontalAlignment=HAlign_Left, verticalAlignment=VAlign_Fill, padding=<ltrb 0 0 0 0>}}
  }
  PropControlSetup_Widget {
   widgetClassPath=/AnimInEdTools/PropBrowser/PropControl/EUW_PropControlSetup
   bIsVariable=true
   parent=VerticalBox_75
   widget=PropControlSetup_Widget
   slot {class=VerticalBoxSlot}
  }
  ExtraPanels {
   widgetClassPath=&CanvasPanel
   bIsVariable=true
   parent=VerticalBox_75
   widget=ExtraPanels
   slot {class=VerticalBoxSlot, properties {size {sizeRule=Fill, value=1}}}
  }
  ScaleBox_0 {
   widgetClassPath=&ScaleBox
   parent=ExtraPanels
   widget=ScaleBox_0
   slot {
    class=CanvasPanelSlot
    properties {layoutData {alignment=<0 0>, anchors {maximum=<0 0>, minimum=<0 0>}, offsets=<ltrb 0 0 698.999 571.3413>}}
   }
  }
  SizeBox_471 {
   widgetClassPath=&SizeBox
   parent=ScaleBox_0
   widget=SizeBox_471
   properties {bOverride_MinDesiredHeight=true, bOverride_MinDesiredWidth=true, maxDesiredWidth=51952.5625, minDesiredHeight=64.9271}
   slot {class=ScaleBoxSlot, properties {horizontalAlignment=HAlign_Fill, verticalAlignment=VAlign_Fill}}
  }
  ExtraSlot {widgetClassPath=&NamedSlot, bIsVariable=true, parent=SizeBox_471, widget=ExtraSlot, slot {class=SizeBoxSlot}}
  Overlay_173 {
   widgetClassPath=&Overlay
   parent=CanvasPanel_23
   widget=Overlay_173
   slot {
    class=CanvasPanelSlot
    properties {zOrder=1, layoutData {alignment=<0 0>, anchors {maximum=<1 1>, minimum=<0 0>}, offsets=<ltrb 0 0 1.3728 0>}}
   }
  }
  EUW_ContextMenu {
   widgetClassPath=/AnimInEdTools/ContextMenu/EUW_ContextMenu
   bIsVariable=true
   parent=Overlay_173
   widget=EUW_ContextMenu
   properties {visibility=Collapsed}
   slot {class=OverlaySlot}
  }
  ModalLayer {
   widgetClassPath=&NamedSlot
   bIsVariable=true
   parent=Overlay_173
   widget=ModalLayer
   properties {bExposeOnInstanceOnly=true}
   slot {class=OverlaySlot, properties {horizontalAlignment=HAlign_Fill, verticalAlignment=VAlign_Fill}}
  }
  ModalLayer_ErrorDialog {
   widgetClassPath=&NamedSlot
   bIsVariable=true
   parent=Overlay_173
   widget=ModalLayer_ErrorDialog
   properties {bExposeOnInstanceOnly=true}
   slot {class=OverlaySlot, properties {horizontalAlignment=HAlign_Fill, verticalAlignment=VAlign_Fill}}
  }
  ModalLayer_ErrorDialog_DriverCtrl {
   widgetClassPath=&NamedSlot
   bIsVariable=true
   parent=Overlay_173
   widget=ModalLayer_ErrorDialog_DriverCtrl
   properties {bExposeOnInstanceOnly=true}
   slot {class=OverlaySlot, properties {horizontalAlignment=HAlign_Fill, verticalAlignment=VAlign_Fill}}
  }
 ]
}
