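# Editor Utility list row for the Prop Browser, showing an animatable prop's name and type icon.
# On assignment, it derives the name from skeletal mesh, sequencer actor class or static mesh, in that priority.
# It maps the type string to an icon texture loaded by path, hiding the image when that fails.
# Selection recolours border and text; right-clicks pass Item Data, never set here, to the parent PropBrowser.
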WidgetBlueprint /AnimInEdTools/PropBrowser/EUW_PropBrowser_ListItem
graphs [
 AssetName {nodes [
   n1 = AssetName!() -> (then, AP File:Animatable Prop Object Reference) {is=title K2Node_FunctionEntry}
   n2 = Get Skeletal Mesh Asset!(self:Animatable Prop Object Reference=@n1) -> (Skeletal Mesh Asset:Skeletal Mesh Soft Object Reference)
   > n3 = Branch(Condition=IsValidSoftObjectReference(SoftObjectReference=@n2))
   n4 = GetSoftPtrAssetName!(Asset:Soft Object Reference=@n2) -> (ReturnValue:String)
   > n5 = Branch(Condition=StartsWith(SourceString=@n4, InPrefix=SK_))
   > Return Node!(execute, Asset Name:String=Split(SourceString=@n4, InStr=SK_, SearchDir=FromStart).RightS)
   Return Node(execute=@n5.else, Asset Name=@n4)
   n6 = Get Sequencer Actor!(self:Animatable Creature Object Reference=@n1) -> (Sequencer Actor:Object Soft Class Reference)
   n7 = Branch(execute=@n3.else, Condition=IsValidSoftClassReference!(SoftClassReference:"Object Soft Class Reference (by ref)"=@n6) -> (ReturnValue:bool))
   n8 = "To String (SoftClassReference)"!(SoftClassReference:"Object Soft Class Reference (by ref)"=@n6) -> (ReturnValue:String)
   > Return Node(Asset Name=Split(SourceString=Split(SourceString=@n8, InStr=., SearchDir=FromStart).RightS, InStr=_C, SearchDir=FromEnd))
   n9 = Get Static Mesh Asset!(self:Animatable Prop Object Reference=@n1) -> (Static Mesh Asset:Static Mesh Soft Object Reference)
   Branch(execute=@n7.else, Condition=IsValidSoftObjectReference(SoftObjectReference=@n9))
   > Return Node(Asset Name=GetSoftPtrAssetName(Asset=@n9))
  ]}
 Edit Text {nodes [Edit Text!() -> (then, Edit:bool=~) {is=title K2Node_FunctionEntry}]}
 EventGraph {nodes [
   n1 = Display Selected!() -> (OutputDelegate, then, Is Selected:bool=~) {is=title K2Node_CustomEvent}
   > SetBrushColor(
    self=Get Border() -> (Border:Border Object Reference)
    InBrushColor=SelectColor(A=(R=0,G=0.1608,B=0.7451,A=1), B=(R=0,G=0,B=0,A=0), bPickA=@n1.Is Selected)
   )
   > SetColorAndOpacity(
    self=Get Asset() -> (Asset:Text Object Reference)
    InColorAndOpacity=Select(
     Option 0:Slate Color Structure=Get Color And Opacity() -> (Color And Opacity:Slate Color Structure)
     Option 1:Slate Color Structure=Make Slate Color(SpecifiedColor=(R=0.0103,G=0.0103,B=0.0103,A=1))
     Index=@n1.Is Selected
    ) -> (ReturnValue:Slate Color Structure)
   )
   > n2 = Branch(Condition=@n1.Is Selected)
   Edit Text!(execute=@n2.else, self, Edit:bool=false) -> (then) {is=title K2Node_CallFunction}
   n3 = Event On Item Selection Changed!() -> (OutputDelegate, then, bIsSelected:bool) {is=title K2Node_Event}
   > Display Selected!(execute, self, Is Selected:bool=@n3.bIsSelected) -> (then) {is=title K2Node_CallFunction}
   n4 = Event OnListItemObjectSet()
   > n5 = Cast To Object_PropBrowser!(execute, Object=@n4.ListItemObject) -> (then, CastFailed, AsObject Prop Browser:Object Prop Browser Object Reference) {is=title K2Node_DynamicCast}
   n6 = Break Struct Prop Brower Item!(Struct_PropBrower_Item:"Struct Prop Brower Item Structure (by ref)"=Get Data!(self:Object Prop Browser Object Reference=@n5) -> (Data:Struct Prop Brower Item Structure)) -> (
    AP_53_#1:Animatable Prop Object Reference
    Season_AssetType_57_#2:String
    AP_Type_62_#3:String
   ) {is=title K2Node_BreakStruct}
   > n7 = AssetName!(execute, self, AP File:Animatable Prop Object Reference=@n6) -> (then, Asset Name:String) {is=title K2Node_CallFunction}
   > "SetText (Text)"(self=Get Asset(), InText="To Text (String)"(InString=@n7))
   > Icon Path!(execute, self, Ap_Type:String=@n6.AP_Type_62_#3) -> (then)
   > Display Selected(Is Selected=false)
  ]}
 Icon Path {nodes [
   n1 = Icon Path!() -> (then, Ap_Type:String) {is=title K2Node_FunctionEntry}
   > Set AP Type(AP Type:String=@n1) -> (Output_Get:String)
   > n2 = Branch(Condition=IsEmpty(InString=Get AP Type() -> (AP Type:String)))
   > Set Visibility(Visibility=Hidden, self=Get AssetTypeIcon() -> (AssetTypeIcon:Image Object Reference))
   > Return Node!(execute)
   n3 = Get AP Type()
   n4 = Get AP Type()
   n5 = SelectString(A=Drum, B=Get AP Type(), bPickA=OR Boolean(A="Equal Exactly (String)"(A=@n4, B=Drumkit), B="Equal Exactly (String)"(A=@n3, B=DrumStick)))
   n6 = SelectString(A=Prop, B=@n5, bPickA="Equal Exactly (String)"(A=@n5, B=Explosive))
   n7 = SelectString(A=Shotgun, B=@n6, bPickA="Equal Exactly (String)"(A=@n6, B=Gun))
   n8 = SelectString(A=heavy, B=@n7, bPickA="Equal Exactly (String)"(A=@n7, B=Launcher))
   n9 = SelectString(A=Consumable, B=@n8, bPickA="Equal Exactly (String)"(A=@n8, B=Consumable))
   n10 = ToLower!(SourceString:String=SelectString(A=gadgets, B=@n9, bPickA="Equal Exactly (String)"(A=@n9, B=Gadgets))) -> (ReturnValue:String)
   n11 = LoadAsset(execute=@n2.else, AssetPath=Append(A=&/Icons/, B=Append(A=@n10, B=_icon)))
   > n12 = Cast To Texture2D(Object=@n11)
   > Set Visibility(Visibility=Visible, self=Get AssetTypeIcon())
   > n13 = SetBrushFromTexture!(execute, self:Image Object Reference=Get AssetTypeIcon(), Texture:Texture 2D Object Reference=@n12, bMatchSize:bool=false) -> (then)
   Return Node(execute=[@n13.then, @n14.then])
   n14 = Set Visibility(execute=@n12.CastFailed, Visibility=Hidden, self=Get AssetTypeIcon())
  ]}
 OnMouseButtonDown {nodes [
   n1 = OnMouseButtonDown!() -> (then, MyGeometry:Geometry Structure, MouseEvent:"Pointer Event Structure (by ref)") {is=title K2Node_FunctionEntry}
   > n2 = Branch(Condition=Is Mouse Button Down!(Input:"Pointer Event Structure (by ref)"=@n1.MouseEvent, MouseButton:Key Structure=RightMouseButton) -> (ReturnValue:bool))
   > n3 = FindParentWidgetOfType!(
    execute
    StartingWidget:Widget Object Reference=Self-Reference()
    Type:Widget Class Reference=&/PropBrowser.PropBrowser_C
   ) -> (then, ReturnValue:Prop Browser Object Reference)
   > "On Item Right Clicked (List View)"!(
    execute
    self:Prop Browser Object Reference=@n3
    Item:Object Prop Browser Object Reference=Get Item Data() -> (Item Data:Object Prop Browser Object Reference)
    Mouse Event:Pointer Event Structure=@n1.MouseEvent
   ) -> (then)
   > Return Node!(execute, ReturnValue:Event Reply Structure=Handled())
   Return Node(execute=@n2.else, ReturnValue=Unhandled())
  ]}
]
variables [
 Color And Opacity {type=SlateColor, default {colorUseRule=UseColor_Specified, specifiedColor=<rgba 1 1 1 1>}}
 Item Data {type=Object_PropBrowser_C, default=null}
]
widgets {
 info {inheritedWidgetCount=0, parentClass=&EditorUtilityWidget, rootWidgetClass=&Border, widgetCount=6}
 widgets [
  Border {widgetClassPath=&Border, bIsVariable=true, widget=Border, properties {brushColor=<rgba 1 1 1 0>}}
  HorizontalBox_1 {widgetClassPath=&HorizontalBox, parent=Border, widget=HorizontalBox_1, slot {class=BorderSlot}}
  SizeBox_1 {
   widgetClassPath=&SizeBox
   parent=HorizontalBox_1
   widget=SizeBox_1
   properties {bOverride_MinDesiredWidth=true, minDesiredWidth=200}
   slot {class=HorizontalBoxSlot, properties {padding=<ltrb 3 0 0 0>}}
  }
  HorizontalBox_0 {
   widgetClassPath=&HorizontalBox
   parent=SizeBox_1
   widget=HorizontalBox_0
   slot {class=SizeBoxSlot, properties {verticalAlignment=VAlign_Center}}
  }
  AssetTypeIcon {
   widgetClassPath=&Image
   bIsVariable=true
   parent=HorizontalBox_0
   widget=AssetTypeIcon
   properties {brush {
     bIsDynamicallyLoaded=false
     drawAs=Image
     imageType=FullColor
     mirroring=NoMirror
     resourceName=None
     resourceObject=null
     tiling=NoTile
     imageSize=<28 28>
     margin=<ltrb 0 0 0 0>
     outlineSettings {
      bUseBrushTransparency=false
      roundingType=HalfHeightRadius
      width=0
      color {colorUseRule=UseColor_Specified, specifiedColor=<rgba 0 0 0 0>}
      cornerRadii=<0 0 0 0>
     }
     tintColor {colorUseRule=UseColor_Specified, specifiedColor=<rgba 1 1 1 1>}
     uVRegion {bIsValid=false, max=<0 0>, min=<0 0>}
    }}
   slot {class=HorizontalBoxSlot, properties {horizontalAlignment=HAlign_Center, verticalAlignment=VAlign_Center, padding=<ltrb 0 0 5 0>}}
  }
  Asset {
   widgetClassPath=&TextBlock
   bIsVariable=true
   parent=HorizontalBox_0
   widget=Asset
   properties {
    text="Asset "
    colorAndOpacity {colorUseRule=UseColor_Specified, specifiedColor=<rgba 0.5255 0.5255 0.5255 1>}
    font {
     bForceMonospaced=false
     bMaterialIsStencil=false
     fontMaterial=null
     fontObject=&Roboto
     letterSpacing=0
     monospacedWidth=1
     size=10
     skewAmount=0
     typefaceFontName=Regular
     outlineSettings {
      bApplyOutlineToDropShadows=false
      bMiteredCorners=false
      bSeparateFillAlpha=false
      outlineMaterial=null
      outlineSize=0
      outlineColor=<rgba 0 0 0 1>
     }
    }
   }
   slot {class=HorizontalBoxSlot, properties {verticalAlignment=VAlign_Center}}
  }
 ]
}
